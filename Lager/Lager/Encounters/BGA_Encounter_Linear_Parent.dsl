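; Parent actor for linear Living World encounters, where AI attackers and defenders fight across a chain of spawn locations.
; On the server, a trigger sphere starts it once and disables other encounters sharing its location ID.
; Defender deaths shortly after player damage count toward per-playlist targets that advance the fight, shifting point providers, minimap visuals and loot drops.
; It also relocates providers when spawn pads move and uses a no-spawn timeout to shut down stalled encounters.
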(event Custom|BindTrackSpawnVisualBlueprints
  ;[1 1 1] When our Spawn Pad Visual BP's base building gets destroyed, it will relocate itself.
  ;
  ; After that, the Encounter needs to resolve the spawning Point Provider.
  (Default|AssignSpawnVisualActorTeleported (Variables|InternalLogic|GetEncounterSpawnVisualBlueprints) (delegate Custom|SpawnVisualActorTeleported_Event)))

(event Custom|SpawnVisualActorTeleported_Event
  (params (SpawnVisualActor :type BGA%20Encounter%20Visual%20Spawn%20Object%20Reference))
  (bind _encounterspawnvisualblueprints (Variables|InternalLogic|GetEncounterSpawnVisualBlueprints))
  (bind _returnvalue (Utilities|Array|FindItem _encounterspawnvisualblueprints SpawnVisualActor))
  (bind _returnvalue_2 (Transformation|GetActorLocation SpawnVisualActor))
  ;[1 1 1] When our Spawn Pad Visual BP's base building gets destroyed, it will relocate itself.
  ;
  ; After that, the Encounter needs to resolve the spawning Point Provider.
  (if (Utilities|Array|IsValidIndex _encounterspawnvisualblueprints _returnvalue)
    (Utilities|FlowControl|Sequence
      (:then_0
        (if (Utilities|Array|IsValidIndex (Variables|InternalLogic|GetEncounterLocations) _returnvalue)
          (Utilities|Array|SetArrayElem (Variables|InternalLogic|GetEncounterLocations) _returnvalue (- _returnvalue_2 (Transformation|GetActorLocation)) :!pinned true) ; Update the location of the Encounter with the new Spawn Actor Visual BP's \nLocation since we will use this for Leashing behaviors.
        ))
      (:then_1
        (if (Utilities|Array|IsValidIndex (Variables|InternalLogic|GetSpawnedAttackerLivingWorldPointProviders) _returnvalue)
          (bind _result (ArrayMacros|SafeArrayGet (Variables|InternalLogic|GetSpawnedAttackerLivingWorldPointProviders) _returnvalue)
            (:%20%20
              (CallFunction|SetNewPointProviderLocationandResetifNecessary _returnvalue_2 _result)))))
      (:then_2
        (if (Utilities|Array|IsValidIndex (Variables|InternalLogic|GetSpawnedDefenderLivingWorldPointProviders) _returnvalue)
          (bind _result_2 (ArrayMacros|SafeArrayGet (Variables|InternalLogic|GetSpawnedDefenderLivingWorldPointProviders) _returnvalue)
            (:%20%20
              (CallFunction|SetNewPointProviderLocationandResetifNecessary _returnvalue_2 _result_2 :!pinned true) ; Relocate the Point Provider to match the Spawn Visual BP.\nIf Provider is Active, De-Activate and Re-Activate to kickstart fresh EQS.
            )))))))

(event EventBeginPlay
  (bind _returnvalue (Class|FortAthenaLivingWorldManager|GetLivingWorldManager))
  ;[1 1 1] Grab our Point Provider Volume Classes and Encounter Classes
  (Networking|SwitchHasAuthority
    (:Authority
      (bind _gamestate (Class|FortAsyncActionCurrentPlaylistReady|PlaylistDataReadyAsync)
        (:OnReady
          (Utilities|FlowControl|DelayUntilNextTick)
          (Variables|InternalLogic|SetCachedGameState _gamestate)
          (if (CallFunction|EncounterIsEnabled)
            (Utilities|FlowControl|Sequence
              (:then_0
                (Utilities|IsValid _returnvalue
                  (:Is%20Valid
                    (Class|FortAthenaLivingWorldManager|RegisterDensityReservoir _returnvalue self (Variables|EncounterSetup|GetLWActorDensityWeight) (Variables|EncounterSetup|GetLWActorDensityRange)))))
              (:then_1
                (bind _outactors (Actor|GetAllActorsOfClass "/Lager/Lager/Encounters/BGA_Encounter_Visual_POI.BGA_Encounter_Visual_POI_C"))
                (for _array_element _outactors
                  (if (== (Variables|/Lager/Lager/Encounters/BGA_Encounter_Visual_POI.BGA_Encounter_Visual_POI_C|GetEncounterLocationID _array_element) (Variables|EncounterSetup|GetLinearEncounterLocationID))
                    (Class|BGAEncounterVisualPOI|ColorizePOIName _array_element (Variables|EncounterSetup|GetPOITag) :!pinned true) ; Change on-map POI Text color to highlight this area.
                  )))
              (:then_2
                (bind _encountertriggersphere (Variables|Default|GetEncounterTriggerSphere)
                  (:then
                    (Variables|/Script/Engine.PrimitiveComponent|SetbGenerateOverlapEvents true _encountertriggersphere))
                  (:else))))
            (else
              (Default|DisableEncounter)
              (bind _outactors_2 (Actor|GetAllActorsOfClass "/Lager/Lager/Encounters/BGA_Encounter_Visual_POI.BGA_Encounter_Visual_POI_C"))
              (for _array_element_2 _outactors_2
                (if (== (Variables|/Lager/Lager/Encounters/BGA_Encounter_Visual_POI.BGA_Encounter_Visual_POI_C|GetEncounterLocationID _array_element_2) (Variables|EncounterSetup|GetLinearEncounterLocationID))
                  (Actor|SetLifeSpan _array_element_2 0.1))))))))))

(event Custom|BindAllEncounterEvents
  ;[1 1 1] Track Encounters Spawning Actors
  (bind _activeattackerencounter (Variables|InternalLogic|GetActiveAttackerEncounter)
    (:then
      (EventDispatchers|AssignOnActorSpawned _activeattackerencounter (delegate Custom|OnActorSpawned_Attackers)))
    (:else
      _))
  (bind _activedefenderencounter (Variables|InternalLogic|GetActiveDefenderEncounter)
    (:then
      (EventDispatchers|AssignOnActorSpawned _activedefenderencounter (delegate Custom|OnActorSpawned_Defenders)))
    (:else)))

(event Custom|SafetyCheckForNoSpawnCondition
  (bind _nospawnsafetytimeout_hf (Variables|EncounterSetup|GetNoSpawnSafetyTimeout_HF))
  (bind _returnvalue (ScalableFloat|GetValueatLevel _nospawnsafetytimeout_hf))
  ;[1 1 1] Safety check for no-spawn condition.
  ;
  ; EX: Encounter is active, spawn locations and UI drawing, but no one exists because the spawn locations are in the storm and everyone is dead.
  (if (and (== (Utilities|Array|Length (Variables|InternalLogic|GetSpawnedPawnsAttackers)) 0) (== (Utilities|Array|Length (Variables|InternalLogic|GetSpawnedPawnsDefenders)) 0))
    (if (> _returnvalue 0)
      (Utilities|Time|SetTimerbyFunctionName 0 "NoSpawnTimeout" _returnvalue) ; Set countdown timer looking for someone to spawn.
    )))

(event Custom|UnBindAllEncounterEvents
  ;[1 1 1] Track Encounters Spawning Actors
  (bind _activeattackerencounter (Variables|InternalLogic|GetActiveAttackerEncounter)
    (:then
      (EventDispatchers|UnbindEventfromOnActorSpawned _activeattackerencounter (delegate Custom|OnActorSpawned_Attackers)))
    (:else
      _))
  (bind _activedefenderencounter (Variables|InternalLogic|GetActiveDefenderEncounter)
    (:then
      (EventDispatchers|UnbindEventfromOnActorSpawned _activedefenderencounter (delegate Custom|OnActorSpawned_Defenders)))
    (:else)))

(event Custom|HidePOIMiniMapMarker
  ;[1 1 1] Initial Setup when Triggered
  (bind _outactors (Actor|GetAllActorsOfClass "/Lager/Lager/Encounters/BGA_Encounter_Visual_POI.BGA_Encounter_Visual_POI_C"))
  (for _array_element _outactors
    (if (== (Variables|/Lager/Lager/Encounters/BGA_Encounter_Visual_POI.BGA_Encounter_Visual_POI_C|GetEncounterLocationID _array_element) (Variables|EncounterSetup|GetLinearEncounterLocationID))
      (Components|Activation|Deactivate (Variables|/Lager/Lager/Encounters/BGA_Encounter_Visual_POI.BGA_Encounter_Visual_POI_C|GetAthenaSpecialActor_POI_MiniMap _array_element) :!pinned true) ; Hide POI Encounter MiniMap indicator
    )))

(event Custom|OnActorSpawned_Attackers
  (params (EncounterInstance :type Fort%20Athena%20Living%20World%20Encounter%20Instance%20Object%20Reference) (SpawnedActor :type Actor%20Object%20Reference) (EventData :type Fort%20Athena%20Living%20World%20Event%20Data%20Object%20Reference))
  ;[1 1 1] Track Encounters Spawning Actors
  (Utilities|FlowControl|Sequence
    (:then_0
      (bind _asfort_pawn (Utilities|Casting|CastToFortPawn SpawnedActor))
      (CallFunction|BindPawnDeathEventsAttackers _asfort_pawn)
      (if (CallFunction|AttackersDamagedbyPlayerRecently)
        (Default|SetSingleAIChallenged _asfort_pawn) ; Allows AI to drop Loot
      ))
    (:then_1
      (CallFunction|TryBuiltinPlayerSpawnFX SpawnedActor)
      (CallFunction|TrySpawnLocationBlueprintSpawnFX true SpawnedActor))))

(event Custom|NoSpawnTimeout
  ;[1 1 1] Safety check for no-spawn condition.
  ;
  ; EX: Encounter is active, spawn locations and UI drawing, but no one exists because the spawn locations are in the storm and everyone is dead.
  (Default|DisableEncounter) ; ...shut down the Encounter.
)

(event Custom|OnActorSpawned_Defenders
  (params (EncounterInstance :type Fort%20Athena%20Living%20World%20Encounter%20Instance%20Object%20Reference) (SpawnedActor :type Actor%20Object%20Reference) (EventData :type Fort%20Athena%20Living%20World%20Event%20Data%20Object%20Reference))
  ;[1 1 1] Track Encounters Spawning Actors
  (Utilities|FlowControl|Sequence
    (:then_0
      (bind _asfort_pawn (Utilities|Casting|CastToFortPawn SpawnedActor))
      (CallFunction|BindPawnDeathEventsDefenders _asfort_pawn)
      (if (CallFunction|DefendersDamagedbyPlayerRecently)
        (Default|SetSingleAIChallenged _asfort_pawn) ; Allows AI to drop Loot
      ))
    (:then_1
      (CallFunction|TryBuiltinPlayerSpawnFX SpawnedActor)
      (CallFunction|TrySpawnLocationBlueprintSpawnFX false SpawnedActor))))

(event Custom|ResetSafetyCheckForNoSpawnCondition
  ;[1 1 1] Safety check for no-spawn condition.
  ;
  ; EX: Encounter is active, spawn locations and UI drawing, but no one exists because the spawn locations are in the storm and everyone is dead.
  (Utilities|Time|ClearTimerbyFunctionName 0 "NoSpawnTimeout") ; Clear the no-spawn timeout because... someone spawned.
)

(event OnComponentBeginOverlap(EncounterTriggerSphere)
  (params (OverlappedComponent :type Primitive%20Component%20Object%20Reference) (OtherActor :type Actor%20Object%20Reference) (OtherComp :type Primitive%20Component%20Object%20Reference) (OtherBodyIndex :type Integer) (bFromSweep :type Boolean) (SweepResult :type Hit%20Result%20Structure%20%28by%20ref%29))
  (bind _linearencounterlocationid (Variables|EncounterSetup|GetLinearEncounterLocationID))
  ;[1 1 1] Initial Setup when Triggered
  (Networking|SwitchHasAuthority
    (:Authority
      (Class|FortAsyncActionCurrentPlaylistReady|PlaylistDataReadyAsync
        (:OnReady
          (if (and (CallFunction|EncounterIsEnabled) (not (Variables|EncounterSetup|GetAbortLinearEncounterStart)))
            (bind _asfort_pawn (Utilities|Casting|CastToFortPawn OtherActor))
            (Utilities|Casting|CastToPlayerController (Pawn|GetController _asfort_pawn))
            (Utilities|FlowControl|DoOnce false
              (:Completed
                (if (not (== (Variables|EncounterSetup|GetLinearEncounterLocationID) -1))
                  (bind _outactors (Actor|GetAllActorsOfClass "/Lager/Lager/Encounters/BGA_Encounter_Linear_Parent.BGA_Encounter_Linear_Parent_C")) ; Get all other Linear Encounters
                  (bind _array_element (Utilities|Array|ForEachLoopwithBreak _outactors)
                    (:LoopBody
                      (if (!= _array_element self)
                        (if (and (== (Variables|EncounterSetup|GetLinearEncounterLocationID _array_element) _linearencounterlocationid) (Variables|InternalLogic|GetLinearEncounterIsActive _array_element))
                          (Variables|EncounterSetup|SetAbortLinearEncounterStart true))
                        (if (== (Variables|EncounterSetup|GetLinearEncounterLocationID _array_element) _linearencounterlocationid)
                          (Utilities|Array|Add (Variables|InternalLogic|GetOtherLinearEncountersWithMatchingLocationIDs) _array_element) ; Add to list to we can pre-emptively shut down other Linear Encounter options, if needed, after we iterate.
                        )))
                    (:Completed
                      _)))
                (if (not (Variables|EncounterSetup|GetAbortLinearEncounterStart))
                  (for _array_element_2 (Variables|InternalLogic|GetOtherLinearEncountersWithMatchingLocationIDs)
                    (Default|DisableEncounter _array_element_2))
                  (Default|StartEncounter)))) ; Only kick-off Encounter once.
          ))))))

(event Custom|UnBindPawnDeathEvents_Attackers
  (params (FortPawn :type Fort%20Pawn%20Object%20Reference))
  ;[1 1 1] Track AI Deaths/Damage
  (Game|UnbindEventfromOnEndPlay FortPawn (delegate Custom|OnEndPlay_Attackers))
  (EventDispatchers|UnbindEventfromOnDied FortPawn (delegate Custom|OnDied_Attackers))
  (EventDispatchers|UnbindEventfromOnDamaged FortPawn (delegate Custom|OnDamaged_Attackers))
  (Utilities|Array|RemoveItem (Variables|InternalLogic|GetSpawnedPawnsAttackers) FortPawn)
  (CallFunction|SafetyCheckforNoSpawnCondition) ; Check for an "Encounter is Active but everyone is dead and no one can spawn" condition.
)

(event Custom|BindPawnDeathEvents_Attackers
  (params (FortPawn :type Fort%20Pawn%20Object%20Reference))
  ;[1 1 1] Track AI Deaths/Damage
  (Game|AssignOnEndPlay FortPawn (delegate Custom|OnEndPlay_Attackers))
  (EventDispatchers|AssignOnDied FortPawn (delegate Custom|OnDied_Attackers))
  (EventDispatchers|AssignOnDamaged FortPawn (delegate Custom|OnDamaged_Attackers))
  (Utilities|Array|Add (Variables|InternalLogic|GetSpawnedPawnsAttackers) FortPawn)
  (Default|MakeOpposingAIAwareOfEachOther FortPawn) ; Make AI aware of Opponents when they spawn.
  (CallFunction|LeashSingleAIAfterSpawn FortPawn true)
  (CallFunction|ResetSafetyCheckforNoSpawnCondition) ; Someone spawned, so cancel the no-spawn timer, if it is active.
)

(event Custom|OnEndPlay_Attackers
  (params (Actor :type Actor%20Object%20Reference) (EndPlayReason :type EEndPlayReason%20Enum :default "Destroyed"))
  ;[1 1 1] Track AI Deaths/Damage
  (bind _asfort_pawn (Utilities|Casting|CastToFortPawn Actor))
  (CallFunction|UnBindPawnDeathEventsAttackers _asfort_pawn))

(event Custom|OnDamaged_Attackers
  (params (DamagedActor :type Actor%20Object%20Reference) (Damage :type Float%20%28single-precision%29) (InstigatedBy :type Controller%20Object%20Reference) (DamageCauser :type Actor%20Object%20Reference) (HitLocation :type Vector) (FHitComponent :type Primitive%20Component%20Object%20Reference) (BoneName :type Name) (Momentum :type Vector))
  ;[1 1 1] Track AI Deaths/Damage
  (if (!= DamagedActor DamageCauser)
    (Utilities|Casting|CastToPlayerController InstigatedBy) ; Works with vehicle run-over kills, too.
  ))

(event Custom|OnDied_Attackers
  (params (DamagedActor :type Actor%20Object%20Reference) (Damage :type Float%20%28single-precision%29) (InstigatedBy :type Controller%20Object%20Reference) (DamageCauser :type Actor%20Object%20Reference) (HitLocation :type Vector) (FHitComponent :type Primitive%20Component%20Object%20Reference) (BoneName :type Name) (Momentum :type Vector))
  ;[1 1 1] Track AI Deaths/Damage
  (bind _asfort_pawn (Utilities|Casting|CastToFortPawn DamagedActor))
  (CallFunction|UnBindPawnDeathEventsAttackers _asfort_pawn)
  (Actor|SetLifeSpan _asfort_pawn 1.0) ; TODO: Should be in the spawn data?
)

(event Custom|AttackersChallengedTimer
  ;[1 1 1] Track AI Deaths/Damage
  (Utilities|Time|ClearTimerbyFunctionName 0 "AttackersChallengedTimer")
  (Default|ClearAllAttackersChallenged) ; AI no longer drop Loot
  (Default|ClearAllDefendersChallenged))

(event Custom|EnableAttackerSpawn
  (params (WarmUpTime :type Float%20%28double-precision%29) (ForceImmediateRespawn :type Boolean))
  (bind _currentattackerencounterlocationindex (Variables|InternalLogic|GetCurrentAttackerEncounterLocationIndex))
  (bind _spawnedattackerlivingworldpointproviders (Variables|InternalLogic|GetSpawnedAttackerLivingWorldPointProviders))
  (bind _returnvalue (Utilities|Array|IsValidIndex _spawnedattackerlivingworldpointproviders _currentattackerencounterlocationindex))
  (bind _currentattackerencounterlocationindex_2 (Variables|InternalLogic|GetCurrentAttackerEncounterLocationIndex))
  (bind _returnvalue_2 (- _currentattackerencounterlocationindex_2 1))
  (bind _currentattackerencounterlocationindex_3 (Variables|InternalLogic|GetCurrentAttackerEncounterLocationIndex))
  (bind _encounterspawnvisualblueprints (Variables|InternalLogic|GetEncounterSpawnVisualBlueprints))
  (bind _encounterspawnlocationdata (Variables|EncounterSetup|GetEncounterSpawnLocationData))
  (bind _currentattackerencounterlocationindex_4 (Variables|InternalLogic|GetCurrentAttackerEncounterLocationIndex))
  (bind _encountermidfieldvisualblueprints (Variables|InternalLogic|GetEncounterMidfieldVisualBlueprints))
  (bind _currentattackerencounterlocationindex_5 (Variables|InternalLogic|GetCurrentAttackerEncounterLocationIndex))
  (bind _currentattackerencounterlocationindex_6 (Variables|InternalLogic|GetCurrentAttackerEncounterLocationIndex))
  (bind _encounterspawnvisualblueprints_2 (Variables|InternalLogic|GetEncounterSpawnVisualBlueprints))
  (bind _encounterspawnvisualblueprints_3 (Variables|InternalLogic|GetEncounterSpawnVisualBlueprints))
  (bind _currentattackerencounterlocationindex_7 (Variables|InternalLogic|GetCurrentAttackerEncounterLocationIndex))
  ;[1 1 1] Update Attacker and Defender Point Provider Spawn Locations as the fight progresses.
  (if (> WarmUpTime 0)
    (if _returnvalue
      (Utilities|Time|SetTimerbyFunctionName 0 "AttackerSpawnWarmupComplete" WarmUpTime) ; Delay before activating spawns so we can play Sounds, etc. on the spawn pad.
      (if (Utilities|Array|IsValidIndex _encounterspawnvisualblueprints _currentattackerencounterlocationindex_6)
        (bind _result (ArrayMacros|SafeArrayGet _encounterspawnvisualblueprints _currentattackerencounterlocationindex_6)
          (:%20%20
            (Utilities|IsValid _result
              (:Is%20Valid
                (Variables|/Lager/Lager/Encounters/BGA_Encounter_Visual_Spawn.BGA_Encounter_Visual_Spawn_C|SetEncounterSpawnerOwnershipState "NewEnumerator1" _result) ; Attacker Warm-Up Sounds and maybe VFX
              ))))))
    (elif _returnvalue
      (bind _result_2 (ArrayMacros|SafeArrayGet _spawnedattackerlivingworldpointproviders _currentattackerencounterlocationindex_3)
        (:%20%20
          (Utilities|FlowControl|Sequence
            (:then_0
              (if (Utilities|Array|IsValidIndex _encounterspawnlocationdata _currentattackerencounterlocationindex_4)
                (bind _result_3 (ArrayMacros|SafeArrayGet _encounterspawnlocationdata _currentattackerencounterlocationindex_4)
                  (:%20%20
                    (if (ScalableFloat|GetValueasBool (.AdvanceAttackerEncounterStage_55_F634112F4124E9C4CFAD7485CC6A8F1E (Utilities|Struct|BreakStructEncounterLinearData _result_3)))
                      (bind _activeattackerencounter (Variables|InternalLogic|GetActiveAttackerEncounter)
                        (:then
                          (Class|FortAthenaLivingWorldEncounterInstance|AdvanceEncounterStage _activeattackerencounter) ; Advance to the next Encounter Stage means we can change who/what spawns.\n\nBosses appear? Current variants expire and stop spawning? Whatever we need.
                        )
                        (:else)))))))
            (:then_1
              (CallFunction|EncounterDebugString (Utilities|String|BuildString(Integer) "" "ATTACKERS: Enable Encounter ID: " _currentattackerencounterlocationindex_3) false (Math|Color|Desaturate(LinearColor) (Variables|EncounterSetup|GetAttackerDebugColor) 0.1))
              (Class|FortAthenaLivingWorldPointProviderInterface|EnablePointProvider_DEPRECATED _result_2) ; Enable Attacker PP
              (Utilities|FlowControl|Sequence
                (:then_0
                  (if (Utilities|Array|IsValidIndex _encounterspawnvisualblueprints_3 _currentattackerencounterlocationindex_7)
                    (bind _result_4 (ArrayMacros|SafeArrayGet _encounterspawnvisualblueprints_3 _currentattackerencounterlocationindex_7)
                      (:%20%20
                        (Utilities|IsValid _result_4
                          (:Is%20Valid
                            (Variables|/Lager/Lager/Encounters/BGA_Encounter_Visual_Spawn.BGA_Encounter_Visual_Spawn_C|SetEncounterSpawnerOwnershipState "NewEnumerator2" _result_4) ; Show New Attacker Spawn MiniMap marker
                          ))))))
                (:then_1
                  (if (Variables|EncounterSetup|GetUseMidfieldUIVisualIndicators)
                    (if (Utilities|Array|IsValidIndex _encountermidfieldvisualblueprints _returnvalue_2)
                      (bind _result_5 (ArrayMacros|SafeArrayGet _encountermidfieldvisualblueprints _returnvalue_2)
                        (:%20%20
                          (Utilities|IsValid _result_5
                            (:Is%20Valid
                              (Class|BGAEncounterVisualMidfield|MiniMapHideAll _result_5) ; Hide previous Midfield MiniMap marker
                            )))))))
                (:then_2
                  (if (!= (Variables|InternalLogic|GetCurrentAttackerEncounterLocationIndex) (Variables|InternalLogic|GetAttackerStartLocationIndex))
                    (if (Utilities|Array|IsValidIndex _encounterspawnvisualblueprints_2 _currentattackerencounterlocationindex_5)
                      (bind _result_6 (ArrayMacros|SafeArrayGet _encounterspawnvisualblueprints_2 _currentattackerencounterlocationindex_5)
                        (:%20%20
                          (CallFunction|DelayandDropLootforAttackers _result_6 (Variables|InternalLogic|GetAttackersHaveClaimedFinalSpawnLocation) :!pinned true) ; If Attackers have Claimed Final Spawn, drop big loot.\n\nOtherwise drop in-progress loot to help sustain the fight.
                          (CallFunction|CESpawnShutDownFX (+ (Transformation|GetActorLocation _result_6) (Variables|EncounterSetup|GetAttackerLootDropOffset)) _result_6))))))
                (:then_3
                  (if (Variables|EncounterSetup|GetDebugEncounter)
                    (Debug|DebugSphereSendtoAll (Transformation|GetActorLocation _result_2) 1e+03 8 (Math|Conversions|ToColor(LinearColor) (Variables|EncounterSetup|GetAttackerDebugColor)) false 3e+01 0 1e+01)))))
            (:then_2
              (if ForceImmediateRespawn
                (bind _activeattackerencounter_2 (Variables|InternalLogic|GetActiveAttackerEncounter)
                  (:then
                    (Class|FortAthenaLivingWorldEncounterInstance|RequestEventGeneration _activeattackerencounter_2) ; Forcing Immediate Respawn
                  )
                  (:else))))))))))

(event Custom|ResetEncounterProgressForNewLocation
  (bind _currentattackerencounterlocationindex (Variables|InternalLogic|GetCurrentAttackerEncounterLocationIndex))
  (bind _encounterspawnlocationdata (Variables|EncounterSetup|GetEncounterSpawnLocationData))
  (bind _output (Utilities|Array|Get(aref) _encounterspawnlocationdata _currentattackerencounterlocationindex))
  (bind (_encountertransform_8_fa43d3a14fa1b6c70d41f2b179384291 _encounterleashinnerradius_49_1991dbb840ba5a888e831984e7f9cff7 _encounterleashouterradius_51_c9b619f34f3a7ec7176fdca869ce044e _defendereliminationtargetsbyplaylist_59_d59f32d1495a9055bd6e01905b904b3e _defenderfallbackdirectlytonextmidfield_46_85b3c06e410046cd33d2b2aaaaa3c554) (Utilities|Struct|BreakStructEncounterLinearData _output))
  ;[1 1 1] Reset Encounter Progress between Locations
  (if (Utilities|Array|IsValidIndex _encounterspawnlocationdata _currentattackerencounterlocationindex)
    (Variables|InternalLogic|GetCachedGameState
      (:then
        (bind _cachedgamestate (Variables|InternalLogic|GetCachedGameState)
          (:then
            (bind _array_element (Utilities|Array|ForEachLoopwithBreak _defendereliminationtargetsbyplaylist_59_d59f32d1495a9055bd6e01905b904b3e)
              (:LoopBody
                (if (or (GameplayTags|IsTagQueryEmpty (.PlaylistTags_6_2B3FACD54D6098DDA7D22BAF827D368A (Utilities|Struct|BreakStructEncounterElimThresholdData _array_element))) (GameplayTags|DoesContainerMatchTagQuery (Class|FortGameStateAthena|GetAthenaPlaylistContextTags _cachedgamestate) (.PlaylistTags_6_2B3FACD54D6098DDA7D22BAF827D368A (Utilities|Struct|BreakStructEncounterElimThresholdData _array_element))))
                  (Variables|InternalLogic|SetCurrentDefenderEliminationTarget (ScalableFloat|GetValueasInteger (Utilities|Struct|BreakStructEncounterElimThresholdData _array_element))) ; Set new Elim Target that players must help Attackers Eliminate.
                ))
              (:Completed
                _)) ; Find the first bit of data that matches.
            (Variables|InternalLogic|SetCurrentDefenderEliminationCount 0) ; Reset Elim Count to start again
            (Utilities|Time|ClearTimerbyFunctionName 0 "DefendersChallengedTimer") ; Reset Player Participation check so player must damage Defender AI periodically to proceed.
            (Variables|InternalLogic|SetDefendersFallbackDirectlyToNextMidfield (ScalableFloat|GetValueasBool _defenderfallbackdirectlytonextmidfield_46_85b3c06e410046cd33d2b2aaaaa3c554)) ; Defenders will fallback directly from Midfield to Midfield\ninstead of passing through their Spawn first.
            (CallFunction|ClientNotifyStateChangeEvent (Variables|InternalLogic|GetCurrentAttackerEncounterLocationIndex) (Utilities|Array|Length (Variables|EncounterSetup|GetEncounterSpawnLocationData)) (Variables|InternalLogic|GetCurrentDefenderEliminationCount) (Variables|InternalLogic|GetCurrentDefenderEliminationTarget)))
          (:else)))
      (:else))))

(event Custom|AttackerSpawnWarmupComplete
  (bind _currentattackerencounterlocationindex (Variables|InternalLogic|GetCurrentAttackerEncounterLocationIndex))
  (bind _returnvalue (- _currentattackerencounterlocationindex 1))
  (bind _currentattackerencounterlocationindex_2 (Variables|InternalLogic|GetCurrentAttackerEncounterLocationIndex))
  (bind _encounterspawnlocationdata (Variables|EncounterSetup|GetEncounterSpawnLocationData))
  (bind _currentattackerencounterlocationindex_3 (Variables|InternalLogic|GetCurrentAttackerEncounterLocationIndex))
  (bind _encountermidfieldvisualblueprints (Variables|InternalLogic|GetEncounterMidfieldVisualBlueprints))
  (bind _currentattackerencounterlocationindex_4 (Variables|InternalLogic|GetCurrentAttackerEncounterLocationIndex))
  (bind _spawnedattackerlivingworldpointproviders (Variables|InternalLogic|GetSpawnedAttackerLivingWorldPointProviders))
  (bind _encounterspawnvisualblueprints (Variables|InternalLogic|GetEncounterSpawnVisualBlueprints))
  (bind _encounterspawnvisualblueprints_2 (Variables|InternalLogic|GetEncounterSpawnVisualBlueprints))
  (bind _currentattackerencounterlocationindex_5 (Variables|InternalLogic|GetCurrentAttackerEncounterLocationIndex))
  ;[1 1 1] Update Attacker and Defender Point Provider Spawn Locations as the fight progresses.
  (if (Utilities|Array|IsValidIndex _spawnedattackerlivingworldpointproviders (Variables|InternalLogic|GetCurrentAttackerEncounterLocationIndex))
    (bind _result (ArrayMacros|SafeArrayGet _spawnedattackerlivingworldpointproviders _currentattackerencounterlocationindex_2)
      (:%20%20
        (Utilities|FlowControl|Sequence
          (:then_0
            (if (Utilities|Array|IsValidIndex _encounterspawnlocationdata _currentattackerencounterlocationindex_3)
              (bind _result_2 (ArrayMacros|SafeArrayGet _encounterspawnlocationdata _currentattackerencounterlocationindex_3)
                (:%20%20
                  (if (ScalableFloat|GetValueasBool (.AdvanceAttackerEncounterStage_55_F634112F4124E9C4CFAD7485CC6A8F1E (Utilities|Struct|BreakStructEncounterLinearData _result_2)))
                    (bind _activeattackerencounter (Variables|InternalLogic|GetActiveAttackerEncounter)
                      (:then
                        (Class|FortAthenaLivingWorldEncounterInstance|AdvanceEncounterStage _activeattackerencounter) ; Advance to the next Encounter Stage means we can change who/what spawns.\n\nBosses appear? Current variants expire and stop spawning? Whatever we need.
                      )
                      (:else)))))))
          (:then_1
            (CallFunction|EncounterDebugString (Utilities|String|BuildString(Integer) "" "ATTACKERS: Enable Encounter ID: " _currentattackerencounterlocationindex_2) false (Math|Color|Desaturate(LinearColor) (Variables|EncounterSetup|GetAttackerDebugColor) 0.1))
            (Class|FortAthenaLivingWorldPointProviderInterface|EnablePointProvider_DEPRECATED _result) ; Enable Attacker PP
            (Utilities|FlowControl|Sequence
              (:then_0
                (if (Utilities|Array|IsValidIndex _encounterspawnvisualblueprints_2 _currentattackerencounterlocationindex_5)
                  (bind _result_3 (ArrayMacros|SafeArrayGet _encounterspawnvisualblueprints_2 _currentattackerencounterlocationindex_5)
                    (:%20%20
                      (Utilities|IsValid _result_3
                        (:Is%20Valid
                          (Variables|/Lager/Lager/Encounters/BGA_Encounter_Visual_Spawn.BGA_Encounter_Visual_Spawn_C|SetEncounterSpawnerOwnershipState "NewEnumerator2" _result_3) ; Show New Attacker Spawn MiniMap marker
                        ))))))
              (:then_1
                (if (Variables|EncounterSetup|GetUseMidfieldUIVisualIndicators)
                  (if (Utilities|Array|IsValidIndex _encountermidfieldvisualblueprints _returnvalue)
                    (bind _result_4 (ArrayMacros|SafeArrayGet _encountermidfieldvisualblueprints _returnvalue)
                      (:%20%20
                        (Utilities|IsValid _result_4
                          (:Is%20Valid
                            (Class|BGAEncounterVisualMidfield|MiniMapHideAll _result_4) ; Hide previous Midfield MiniMap marker
                          )))))))
              (:then_2
                (if (!= (Variables|InternalLogic|GetCurrentAttackerEncounterLocationIndex) (Variables|InternalLogic|GetAttackerStartLocationIndex))
                  (if (Utilities|Array|IsValidIndex _encounterspawnvisualblueprints _currentattackerencounterlocationindex_4)
                    (bind _result_5 (ArrayMacros|SafeArrayGet _encounterspawnvisualblueprints _currentattackerencounterlocationindex_4)
                      (:%20%20
                        (CallFunction|DelayandDropLootforAttackers _result_5 (Variables|InternalLogic|GetAttackersHaveClaimedFinalSpawnLocation) :!pinned true) ; If Attackers have Claimed Final Spawn, drop big loot.\n\nOtherwise drop in-progress loot to help sustain the fight.
                        (CallFunction|CESpawnShutDownFX (+ (Transformation|GetActorLocation _result_5) (Variables|EncounterSetup|GetAttackerLootDropOffset)) _result_5))))))
              (:then_3
                (if (Variables|EncounterSetup|GetDebugEncounter)
                  (Debug|DebugSphereSendtoAll (Transformation|GetActorLocation _result) 1e+03 8 (Math|Conversions|ToColor(LinearColor) (Variables|EncounterSetup|GetAttackerDebugColor)) false 3e+01 0 1e+01)))))
          (:then_2
            (if (stale Custom|EnableAttackerSpawn ForceImmediateRespawn)
              (bind _activeattackerencounter_2 (Variables|InternalLogic|GetActiveAttackerEncounter)
                (:then
                  (Class|FortAthenaLivingWorldEncounterInstance|RequestEventGeneration _activeattackerencounter_2) ; Forcing Immediate Respawn
                )
                (:else)))))))))

(event Custom|UnBindPawnDeathEvents_Defenders
  (params (FortPawn :type Fort%20Pawn%20Object%20Reference))
  ;[1 1 1] Track AI Deaths/Damage
  (Game|UnbindEventfromOnEndPlay FortPawn (delegate Custom|OnEndPlay_Defenders))
  (EventDispatchers|UnbindEventfromOnDied FortPawn (delegate Custom|OnDied_Defenders))
  (EventDispatchers|UnbindEventfromOnDamaged FortPawn (delegate Custom|OnDamaged_Defenders))
  (Utilities|Array|RemoveItem (Variables|InternalLogic|GetSpawnedPawnsDefenders) FortPawn)
  (CallFunction|SafetyCheckforNoSpawnCondition) ; Check for an "Encounter is Active but everyone is dead and no one can spawn" condition.
)

(event Custom|BindPawnDeathEvents_Defenders
  (params (FortPawn :type Fort%20Pawn%20Object%20Reference))
  ;[1 1 1] Track AI Deaths/Damage
  (Game|AssignOnEndPlay FortPawn (delegate Custom|OnEndPlay_Defenders))
  (EventDispatchers|AssignOnDied FortPawn (delegate Custom|OnDied_Defenders))
  (EventDispatchers|AssignOnDamaged FortPawn (delegate Custom|OnDamaged_Defenders))
  (Utilities|Array|Add (Variables|InternalLogic|GetSpawnedPawnsDefenders) FortPawn)
  (Default|MakeOpposingAIAwareOfEachOther FortPawn) ; Make AI aware of Opponents when they spawn.
  (CallFunction|LeashSingleAIAfterSpawn FortPawn) ; Defenders choose Midfield location based on whether they are holding ground or retreating
  (CallFunction|ResetSafetyCheckforNoSpawnCondition) ; Someone spawned, so cancel the no-spawn timer, if it is active.
)

(event Custom|CE_SpawnShutDownFX
  (params (FX%20Location :type Vector) (VisualSpawnBP :type BGA%20Encounter%20Visual%20Spawn%20Object%20Reference))
  ;[1 1 1] Update Attacker and Defender Point Provider Spawn Locations as the fight progresses.
  (CallFunction|ClientSpawnShutDownFX FX%20Location VisualSpawnBP))

(event Custom|OnEndPlay_Defenders
  (params (Actor :type Actor%20Object%20Reference) (EndPlayReason :type EEndPlayReason%20Enum :default "Destroyed"))
  ;[1 1 1] Track AI Deaths/Damage
  (bind _asfort_pawn (Utilities|Casting|CastToFortPawn Actor))
  (CallFunction|UnBindPawnDeathEventsDefenders _asfort_pawn))

(event Custom|OnDamaged_Defenders
  (params (DamagedActor :type Actor%20Object%20Reference) (Damage :type Float%20%28single-precision%29) (InstigatedBy :type Controller%20Object%20Reference) (DamageCauser :type Actor%20Object%20Reference) (HitLocation :type Vector) (FHitComponent :type Primitive%20Component%20Object%20Reference) (BoneName :type Name) (Momentum :type Vector))
  ;[1 1 1] Track AI Deaths/Damage
  (if (!= DamagedActor DamageCauser)
    (Utilities|Casting|CastToPlayerController InstigatedBy) ; Works with vehicle run-over kills, too.
    (Utilities|Time|SetTimerbyFunctionName 0 "DefendersChallengedTimer" (ScalableFloat|GetValueatLevel (Variables|EncounterSetup|GetPlayerParticipationTimeAfterDamagingDefenders_HF))) ; Set a countdown timer for the last time a Player helped damage Defenders
    (CallFunction|SetAllDefendersChallenged) ; Allows AI to drop Loot
    (CallFunction|SetAllAttackersChallenged)))

(event Custom|DefendersChallengedTimer
  ;[1 1 1] Track AI Deaths/Damage
  (Utilities|Time|ClearTimerbyFunctionName 0 "DefendersChallengedTimer")
  (Default|ClearAllDefendersChallenged) ; AI no longer drop Loot
  (Default|ClearAllAttackersChallenged))

(event Custom|OnDied_Defenders
  (params (DamagedActor :type Actor%20Object%20Reference) (Damage :type Float%20%28single-precision%29) (InstigatedBy :type Controller%20Object%20Reference) (DamageCauser :type Actor%20Object%20Reference) (HitLocation :type Vector) (FHitComponent :type Primitive%20Component%20Object%20Reference) (BoneName :type Name) (Momentum :type Vector))
  (bind _attackerdebugcolor (Variables|EncounterSetup|GetAttackerDebugColor))
  (bind _returnvalue (Math|Color|Desaturate(LinearColor) _attackerdebugcolor 0.1))
  (bind _currentdefendereliminationcount (Variables|InternalLogic|GetCurrentDefenderEliminationCount))
  ;[1 1 1] Track AI Deaths/Damage
  (bind _asfort_pawn (Utilities|Casting|CastToFortPawn DamagedActor))
  (CallFunction|UnBindPawnDeathEventsDefenders _asfort_pawn)
  (Actor|SetLifeSpan _asfort_pawn 1.0) ; TODO: Should be in the spawn data?
  (if (!= DamagedActor DamageCauser)
    (if (CallFunction|DefendersDamagedbyPlayerRecently)
      (Math|Integer|IncrementInt _currentdefendereliminationcount :!pinned true
        (:%20%20
          (CallFunction|EncounterDebugString (Utilities|String|BuildString(Integer) (Utilities|String|BuildString(Integer) "ATTACKERS: " "Valid Elims: " _currentdefendereliminationcount "/") "" (Variables|InternalLogic|GetCurrentDefenderEliminationTarget)) false (Variables|EncounterSetup|GetElimAssistedStartColor) 5.0)
          (if (Variables|EncounterSetup|GetDebugEncounter)
            (Debug|DebugSphereSendtoAll (Transformation|GetActorLocation _asfort_pawn) (Math|Float|SelectFloat 2e+02 1e+02 (<= (Variables|InternalLogic|GetCurrentDefenderEliminationCount) (Variables|InternalLogic|GetCurrentDefenderEliminationTarget))) (Math|Integer|SelectInt 8 5 (<= (Variables|InternalLogic|GetCurrentDefenderEliminationCount) (Variables|InternalLogic|GetCurrentDefenderEliminationTarget))) (select (<= (Variables|InternalLogic|GetCurrentDefenderEliminationCount) (Variables|InternalLogic|GetCurrentDefenderEliminationTarget)) (Math|Conversions|ToColor(LinearColor) (Math|Color|Lerp(LinearColor) (Variables|EncounterSetup|GetElimAssistedStartColor) (Variables|EncounterSetup|GetElimAssistedEndColor) (/ (- (ScalableFloat|GetValueatLevel (Variables|EncounterSetup|GetPlayerParticipationTimeAfterDamagingDefenders_HF)) (Utilities|Time|GetTimerRemainingTimebyFunctionName 0 "DefendersChallengedTimer")) (ScalableFloat|GetValueatLevel (Variables|EncounterSetup|GetPlayerParticipationTimeAfterDamagingDefenders_HF))))) (Math|Conversions|ToColor(LinearColor) (Variables|EncounterSetup|GetElimAssistedOverkillColor))) false (Math|Float|SelectFloat 3.0 1.5 (<= (Variables|InternalLogic|GetCurrentDefenderEliminationCount) (Variables|InternalLogic|GetCurrentDefenderEliminationTarget))) 0 4.0)
            (else
              _))
          (if (>= (Variables|InternalLogic|GetCurrentDefenderEliminationCount) (Variables|InternalLogic|GetCurrentDefenderEliminationTarget))
            (CallFunction|LinearEncounterAdvance)
            (else
              _))
          (CallFunction|ClientNotifyStateChangeEvent (Variables|InternalLogic|GetCurrentAttackerEncounterLocationIndex) (Utilities|Array|Length (Variables|EncounterSetup|GetEncounterSpawnLocationData)) (Variables|InternalLogic|GetCurrentDefenderEliminationCount) (Variables|InternalLogic|GetCurrentDefenderEliminationTarget)))) ; Count Valid Eliminations of Defenders
      (else
        (if (Variables|EncounterSetup|GetDebugEncounter)
          (Debug|DebugSphereSendtoAll (Transformation|GetActorLocation _asfort_pawn) 1e+02 5 (Math|Conversions|ToColor(LinearColor) (Variables|EncounterSetup|GetElimUnassistedColor)) false 1.5 0 4.0)
          (else
            _))
        (if (Variables|InternalLogic|GetHoldingForDefenderCleanup)
          (if (CallFunction|AllDefendersDefeated)
            (Utilities|FlowControl|DoOnce false :!pinned true
              (:Completed
                (CallFunction|EncounterDebugString "ATTACKERS: Victory! Cleanup of Defenders DONE." false _returnvalue)
                (CallFunction|AttackersSetNewSpawnandAdvancetoMidfield))) ; Attackers Claim Victory
            (else
              (Variables|InternalLogic|SetHoldingForDefenderCleanup true) ; Track Cleanup Phase so Attacker AI can \nfinish things up themselves and progress.
              (CallFunction|EncounterDebugString "ATTACKERS: Cleanup remaining Defenders." false _returnvalue))))))))

(event Custom|DefenderSpawnWarmupComplete
  (bind _encounterspawnvisualblueprints (Variables|InternalLogic|GetEncounterSpawnVisualBlueprints))
  (bind _defenderencounterlocationtoenable (Variables|InternalLogic|GetDefenderEncounterLocationToEnable))
  (bind _encounterspawnlocationdata (Variables|EncounterSetup|GetEncounterSpawnLocationData))
  (bind _defenderencounterlocationtoenable_2 (Variables|InternalLogic|GetDefenderEncounterLocationToEnable))
  ;[1 1 1] Update Attacker and Defender Point Provider Spawn Locations as the fight progresses.
  (if (Utilities|Array|IsValidIndex (Variables|InternalLogic|GetSpawnedDefenderLivingWorldPointProviders) (Variables|InternalLogic|GetDefenderEncounterLocationToEnable))
    (bind _result (ArrayMacros|SafeArrayGet (Variables|InternalLogic|GetSpawnedDefenderLivingWorldPointProviders) _defenderencounterlocationtoenable)
      (:%20%20
        (Utilities|FlowControl|Sequence
          (:then_0
            (if (Utilities|Array|IsValidIndex _encounterspawnlocationdata _defenderencounterlocationtoenable_2)
              (bind _result_2 (ArrayMacros|SafeArrayGet _encounterspawnlocationdata _defenderencounterlocationtoenable_2)
                (:%20%20
                  (if (ScalableFloat|GetValueasBool (.AdvanceDefenderEncounterStage_52_F23CF75741E54E23495FC9AF65A80450 (Utilities|Struct|BreakStructEncounterLinearData _result_2)))
                    (bind _activedefenderencounter (Variables|InternalLogic|GetActiveDefenderEncounter)
                      (:then
                        (Class|FortAthenaLivingWorldEncounterInstance|AdvanceEncounterStage _activedefenderencounter) ; Advane to the next Encounter Stage means we can change who/what spawns.\n\nBosses appear? Current variants expire and stop spawning? Whatever we need.
                      )
                      (:else)))))))
          (:then_1
            (CallFunction|EncounterDebugString (Utilities|String|BuildString(Integer) "" "DEFENDERS: Enable Encounter ID: " _defenderencounterlocationtoenable) false (Math|Color|Desaturate(LinearColor) (Variables|EncounterSetup|GetDefenderDebugColor) 0.1))
            (Class|FortAthenaLivingWorldPointProviderInterface|EnablePointProvider_DEPRECATED _result) ; Enable Defender PP
            (Utilities|FlowControl|Sequence
              (:then_0
                (if (Utilities|Array|IsValidIndex _encounterspawnvisualblueprints _defenderencounterlocationtoenable)
                  (bind _result_3 (ArrayMacros|SafeArrayGet _encounterspawnvisualblueprints _defenderencounterlocationtoenable)
                    (:%20%20
                      (Utilities|IsValid _result_3
                        (:Is%20Valid
                          (Variables|/Lager/Lager/Encounters/BGA_Encounter_Visual_Spawn.BGA_Encounter_Visual_Spawn_C|SetEncounterSpawnerOwnershipState "NewEnumerator5" _result_3) ; Show Defender Spawn MiniMap marker
                        ))))))
              (:then_1
                (if (Variables|EncounterSetup|GetDebugEncounter)
                  (Debug|DebugSphereSendtoAll (Transformation|GetActorLocation _result) 9.5e+02 8 (Math|Conversions|ToColor(LinearColor) (Variables|EncounterSetup|GetDefenderDebugColor)) false 3e+01 0 1e+01)))))
          (:then_2
            (if (stale Custom|EnableDefenderSpawn ForceImmediateRespawn)
              (bind _activedefenderencounter_2 (Variables|InternalLogic|GetActiveDefenderEncounter)
                (:then
                  (Class|FortAthenaLivingWorldEncounterInstance|RequestEventGeneration _activedefenderencounter_2) ; Forcing Immediate Respawn
                )
                (:else)))))))))

(event Custom|LinearEncounterAdvance
  (bind _attackerdebugcolor (Variables|EncounterSetup|GetAttackerDebugColor))
  (bind _returnvalue (Math|Color|Desaturate(LinearColor) _attackerdebugcolor 0.1))
  ;[1 1 1] Track AI Deaths/Damage
  (if (Variables|InternalLogic|GetLinearEncounterIsActive)
    (CallFunction|DefendersFallbacktoSpawn :!pinned true) ; Defenders Fallback to Spawn
    (if (CallFunction|EncounterStillHasValidCombatEncountersBeyondCurrentOne)
      _
      (elif (CallFunction|AllDefendersDefeated)
        (Utilities|FlowControl|DoOnce false :!pinned true
          (:Completed
            (CallFunction|EncounterDebugString "ATTACKERS: Victory! Cleanup of Defenders DONE." false _returnvalue)
            (CallFunction|AttackersSetNewSpawnandAdvancetoMidfield))) ; Attackers Claim Victory
        (else
          (Variables|InternalLogic|SetHoldingForDefenderCleanup true) ; Track Cleanup Phase so Attacker AI can \nfinish things up themselves and progress.
          (CallFunction|EncounterDebugString "ATTACKERS: Cleanup remaining Defenders." false _returnvalue))))))

(event Custom|EnableDefenderSpawn
  (params (WarmUpTime :type Float%20%28double-precision%29) (ForceImmediateRespawn :type Boolean) (AdditionalLocationOffset :type Integer))
  (bind _returnvalue (CallFunction|DefendersinFallbackPhase))
  (bind _returnvalue_2 (Math|Integer|SelectInt 2 1 _returnvalue))
  (bind _currentattackerencounterlocationindex (Variables|InternalLogic|GetCurrentAttackerEncounterLocationIndex))
  (bind _returnvalue_3 (Utilities|Operators|Add AdditionalLocationOffset _currentattackerencounterlocationindex _returnvalue_2))
  (bind _encounterspawnvisualblueprints (Variables|InternalLogic|GetEncounterSpawnVisualBlueprints))
  (bind _encounterspawnvisualblueprints_2 (Variables|InternalLogic|GetEncounterSpawnVisualBlueprints))
  (bind _defenderencounterlocationtoenable (Variables|InternalLogic|GetDefenderEncounterLocationToEnable))
  (bind _encounterspawnlocationdata (Variables|EncounterSetup|GetEncounterSpawnLocationData))
  (bind _defenderencounterlocationtoenable_2 (Variables|InternalLogic|GetDefenderEncounterLocationToEnable))
  ;[1 1 1] Update Attacker and Defender Point Provider Spawn Locations as the fight progresses.
  (if (Variables|InternalLogic|GetLastDefenderSpawnDefeated)
    _
    (elif (Utilities|Array|IsValidIndex (Variables|InternalLogic|GetSpawnedDefenderLivingWorldPointProviders) _returnvalue_3)
      (Variables|InternalLogic|SetDefenderEncounterLocationToEnable _returnvalue_3)
      (if (> WarmUpTime 0)
        (Utilities|Time|SetTimerbyFunctionName 0 "DefenderSpawnWarmupComplete" WarmUpTime) ; Delay before activating spawns so we can play Sounds, etc. on the spawn pad.
        (if (Utilities|Array|IsValidIndex _encounterspawnvisualblueprints_2 _defenderencounterlocationtoenable)
          (bind _result (ArrayMacros|SafeArrayGet _encounterspawnvisualblueprints_2 _defenderencounterlocationtoenable)
            (:%20%20
              (Utilities|IsValid _result
                (:Is%20Valid
                  (Variables|/Lager/Lager/Encounters/BGA_Encounter_Visual_Spawn.BGA_Encounter_Visual_Spawn_C|SetEncounterSpawnerOwnershipState "NewEnumerator4" _result) ; Defender Warm-Up Sounds and maybe VFX
                )))))
        (elif (Utilities|Array|IsValidIndex (Variables|InternalLogic|GetSpawnedDefenderLivingWorldPointProviders) (Variables|InternalLogic|GetDefenderEncounterLocationToEnable))
          (bind _result_2 (ArrayMacros|SafeArrayGet (Variables|InternalLogic|GetSpawnedDefenderLivingWorldPointProviders) _defenderencounterlocationtoenable)
            (:%20%20
              (Utilities|FlowControl|Sequence
                (:then_0
                  (if (Utilities|Array|IsValidIndex _encounterspawnlocationdata _defenderencounterlocationtoenable_2)
                    (bind _result_3 (ArrayMacros|SafeArrayGet _encounterspawnlocationdata _defenderencounterlocationtoenable_2)
                      (:%20%20
                        (if (ScalableFloat|GetValueasBool (.AdvanceDefenderEncounterStage_52_F23CF75741E54E23495FC9AF65A80450 (Utilities|Struct|BreakStructEncounterLinearData _result_3)))
                          (bind _activedefenderencounter (Variables|InternalLogic|GetActiveDefenderEncounter)
                            (:then
                              (Class|FortAthenaLivingWorldEncounterInstance|AdvanceEncounterStage _activedefenderencounter) ; Advane to the next Encounter Stage means we can change who/what spawns.\n\nBosses appear? Current variants expire and stop spawning? Whatever we need.
                            )
                            (:else)))))))
                (:then_1
                  (CallFunction|EncounterDebugString (Utilities|String|BuildString(Integer) "" "DEFENDERS: Enable Encounter ID: " _defenderencounterlocationtoenable) false (Math|Color|Desaturate(LinearColor) (Variables|EncounterSetup|GetDefenderDebugColor) 0.1))
                  (Class|FortAthenaLivingWorldPointProviderInterface|EnablePointProvider_DEPRECATED _result_2) ; Enable Defender PP
                  (Utilities|FlowControl|Sequence
                    (:then_0
                      (if (Utilities|Array|IsValidIndex _encounterspawnvisualblueprints _defenderencounterlocationtoenable)
                        (bind _result_4 (ArrayMacros|SafeArrayGet _encounterspawnvisualblueprints _defenderencounterlocationtoenable)
                          (:%20%20
                            (Utilities|IsValid _result_4
                              (:Is%20Valid
                                (Variables|/Lager/Lager/Encounters/BGA_Encounter_Visual_Spawn.BGA_Encounter_Visual_Spawn_C|SetEncounterSpawnerOwnershipState "NewEnumerator5" _result_4) ; Show Defender Spawn MiniMap marker
                              ))))))
                    (:then_1
                      (if (Variables|EncounterSetup|GetDebugEncounter)
                        (Debug|DebugSphereSendtoAll (Transformation|GetActorLocation _result_2) 9.5e+02 8 (Math|Conversions|ToColor(LinearColor) (Variables|EncounterSetup|GetDefenderDebugColor)) false 3e+01 0 1e+01)))))
                (:then_2
                  (if ForceImmediateRespawn
                    (bind _activedefenderencounter_2 (Variables|InternalLogic|GetActiveDefenderEncounter)
                      (:then
                        (Class|FortAthenaLivingWorldEncounterInstance|RequestEventGeneration _activedefenderencounter_2) ; Forcing Immediate Respawn
                      )
                      (:else))))))))))))

(event Custom|ClientNotifyStateChangeEvent
  (params (LocationIndex :type Integer) (LocationCount :type Integer) (EliminationCount :type Integer) (EliminationTarget :type Integer))
  ;[1 1 1] Send data update to clients for UI
  (Networking|SwitchHasAuthority
    (:Remote
      (Default|ClientNotifyStateChange LocationIndex LocationCount EliminationCount EliminationTarget))))

(event Custom|UpdateMidfieldToEncounterDebug
  (params (Index :type Integer))
  (bind _midfieldtodefenseencounterarrows (Variables|InternalLogic|GetMidfieldToDefenseEncounterArrows))
  ;[1 1 1] Defenders defeated at current location. Defender Spawn location drops back one spot, and the Defender AI begin to Fallback to their newly disabled Spawn.
  ; Newly spawned Defenders spawn at the new location and will press forward to the new Midfield and wait for their existing Defender friends to arrive.
  ;[BEGIN DELAY]
  (if (Utilities|Array|IsValidIndex _midfieldtodefenseencounterarrows Index)
    (bind _result (ArrayMacros|SafeArrayGet _midfieldtodefenseencounterarrows Index)
      (:%20%20
        (Components|Arrow|SetArrowColor(LinearColor) _result (Variables|EncounterSetup|GetAttackerDebugColor)) ; Update Arrow Color from Defender to Attacker
      ))))

(event Custom|DefendersFallbackToSpawn
  (bind _defendersfallbacktospawntimer_hf (Variables|EncounterSetup|GetDefendersFallbackToSpawnTimer_HF))
  (bind _returnvalue (ScalableFloat|GetValueatLevel _defendersfallbacktospawntimer_hf))
  (bind _currentattackerencounterlocationindex (Variables|InternalLogic|GetCurrentAttackerEncounterLocationIndex))
  (bind _returnvalue_2 (+ _currentattackerencounterlocationindex 2))
  (bind _currentattackerencounterlocationindex_2 (Variables|InternalLogic|GetCurrentAttackerEncounterLocationIndex))
  (bind _returnvalue_3 (+ _currentattackerencounterlocationindex_2 1))
  (bind _currentattackerencounterlocationindex_3 (Variables|InternalLogic|GetCurrentAttackerEncounterLocationIndex))
  (bind _returnvalue_4 (+ _currentattackerencounterlocationindex_3 1))
  (bind _encounterspawnvisualblueprints (Variables|InternalLogic|GetEncounterSpawnVisualBlueprints))
  (bind _encounterspawnvisualblueprints_2 (Variables|InternalLogic|GetEncounterSpawnVisualBlueprints))
  (bind _spawneddefenderlivingworldpointproviders (Variables|InternalLogic|GetSpawnedDefenderLivingWorldPointProviders))
  ;[1 1 1] Defenders defeated at current location. Defender Spawn location drops back one spot, and the Defender AI begin to Fallback to their newly disabled Spawn.
  ; Newly spawned Defenders spawn at the new location and will press forward to the new Midfield and wait for their existing Defender friends to arrive.
  ;[BEGIN DELAY]
  (if (CallFunction|DefendersinFallbackPhase)
    _
    (elif (Utilities|Array|IsValidIndex _spawneddefenderlivingworldpointproviders _returnvalue_4)
      (Variables|InternalLogic|SetLastDefenderSpawnDefeated (== (Utilities|Array|LastIndex _spawneddefenderlivingworldpointproviders) _returnvalue_4))
      (if (Variables|InternalLogic|GetDefendersFallbackDirectlyToNextMidfield)
        _
        (elif (and (Variables|InternalLogic|GetLastDefenderSpawnDefeated) (not (ScalableFloat|GetValueasBool (Variables|EncounterSetup|GetAttackersShouldDoFinalCleanupPushToDefenderSpawn_HF))))
          _
          (elif (> _returnvalue 0)
            (Utilities|Time|SetTimerbyFunctionName 0 "DefendersFallbackToSpawnTimer" _returnvalue) ; Defenders Fallback for a while before Attackers will follow\nFallback from current Midfield to Spawn we are about to disable
            (bind _result (ArrayMacros|SafeArrayGet (Variables|InternalLogic|GetSpawnedDefenderLivingWorldPointProviders) _returnvalue_3)
              (:%20%20
                (Class|FortAthenaLivingWorldVolume|DisablePointProvider _result) ; Disable current Defender Spawn
                (Utilities|FlowControl|Sequence
                  (:then_0
                    (if (Variables|EncounterSetup|GetDebugEncounter)
                      (CallFunction|UpdateMidfieldtoEncounterDebug (Variables|InternalLogic|GetCurrentAttackerEncounterLocationIndex))))
                  (:then_1
                    (if (and (Variables|InternalLogic|GetLastDefenderSpawnDefeated) (not (ScalableFloat|GetValueasBool (Variables|EncounterSetup|GetAttackersShouldDoFinalCleanupPushToDefenderSpawn_HF))))
                      _
                      (elif (Variables|InternalLogic|GetDefendersFallbackDirectlyToNextMidfield)
                        (CallFunction|DefendersFallbacktoMidfield) ; Defenders skip Fallback to Spawn, and go directly to next Midfield.
                        (if (ScalableFloat|GetValueasBool (Variables|EncounterSetup|GetDefenderSpawnUpdatesImmediatelyOnFallback_HF))
                          (CallFunction|EnableDefenderSpawn (Variables|EncounterSounds|GetDefenderSpawnWarmUpTime) true) ; Enable next Defender Spawn
                        )
                        (else
                          (CallFunction|EncounterDebugString "DEFENDERS: Fallback to Spawn." false (Math|Color|Desaturate(LinearColor) (Variables|EncounterSetup|GetDefenderDebugColor) 0.1))
                          (CallFunction|LeashAllDefenderstoSpawn) ; Leash Defenders to their just-disabled Spawn as a fallback behavior.\nThey are pulling BACK from the current Midfield.
                        ))))
                  (:then_2
                    (if (Utilities|Array|IsValidIndex _encounterspawnvisualblueprints_2 _returnvalue_3)
                      (bind _result_2 (ArrayMacros|SafeArrayGet _encounterspawnvisualblueprints_2 _returnvalue_3)
                        (:%20%20
                          (Utilities|IsValid _result_2
                            (:Is%20Valid
                              (if (Variables|InternalLogic|GetLastDefenderSpawnDefeated)
                                (Variables|/Lager/Lager/Encounters/BGA_Encounter_Visual_Spawn.BGA_Encounter_Visual_Spawn_C|SetEncounterSpawnerOwnershipState "NewEnumerator6" _result_2) ; Disable (NOT Hide) Defender Spawn MiniMap marker
                                (else
                                  (Variables|/Lager/Lager/Encounters/BGA_Encounter_Visual_Spawn.BGA_Encounter_Visual_Spawn_C|SetEncounterSpawnerOwnershipState "NewEnumerator3" _result_2) ; Disable (NOT Hide) Defender Spawn MiniMap marker
                                ))))))))
                  (:then_3
                    (if (Utilities|Array|IsValidIndex _encounterspawnvisualblueprints _returnvalue_2)
                      (bind _result_3 (ArrayMacros|SafeArrayGet _encounterspawnvisualblueprints _returnvalue_2)
                        (:%20%20
                          (Utilities|IsValid _result_3
                            (:Is%20Valid
                              (Variables|/Lager/Lager/Encounters/BGA_Encounter_Visual_Spawn.BGA_Encounter_Visual_Spawn_C|SetEncounterSpawnerOwnershipState "NewEnumerator6" _result_3) ; Preview next Defender Spawn\n\nIn this case we immediately show a Disabled version of the NEXT Defender spawn, well before it is enabled.
                            )))))))))
            (else
              (CallFunction|DefendersFallbacktoSpawnTimer) ; If invalid, do resulting timer stuff anyway...
            )))))))

(event Custom|OnPlayerEnteredPoi
  (params (PoiTag :type Gameplay%20Tag%20Structure%20%28by%20ref%29) (PlayerController :type Fort%20Player%20Controller%20Gameplay%20Object%20Reference))
  ;[1 1 1] Send an update on current status - delay prevents spam and gives an extra moment for Player's POI tag to be replicated
  (if (not (Utilities|Time|IsTimerActivebyHandle (Variables|InternalLogic|GetOnEnterDelayTimer)))
    (bind _returnvalue (Utilities|Time|SetTimerbyEvent (delegate Custom|OnPlayerEnteredPoiDelayed) 2.0))
    (Variables|InternalLogic|SetOnEnterDelayTimer _returnvalue)))

(event Custom|OnPlayerEnteredPoiDelayed
  ;[1 1 1] Send an update on current status - delay prevents spam and gives an extra moment for Player's POI tag to be replicated
  (CallFunction|ClientNotifyStateChangeEvent (Variables|InternalLogic|GetCurrentAttackerEncounterLocationIndex) (Utilities|Array|Length (Variables|EncounterSetup|GetEncounterSpawnLocationData)) (Variables|InternalLogic|GetCurrentDefenderEliminationCount) (Variables|InternalLogic|GetCurrentDefenderEliminationTarget)))

(event Custom|EndDefenderEncounter
  (bind _returnvalue (Game|GetGameState))
  (bind _returnvalue_2 (Actor|GetComponentbyClass _returnvalue "/Script/LagerRuntime.FortAthenaLivingWorldManager"))
  (bind _finaldefenderencountershutdowndelay_hf (Variables|EncounterSetup|GetFinalDefenderEncounterShutdownDelay_HF))
  (bind _returnvalue_3 (ScalableFloat|GetValueatLevel _finaldefenderencountershutdowndelay_hf))
  ;[1 1 1] Shut down Encounters when we are done with them.
  (if (> _returnvalue_3 0)
    (if (not (Utilities|Time|IsTimerActivebyFunctionName 0 "DefenderEncounterShutdownDelay"))
      (Utilities|Time|SetTimerbyFunctionName 0 "DefenderEncounterShutdownDelay" _returnvalue_3) ; This timer exists so that the FINAL Defender Elimination doesn't immediately pop out of existence.\n\nShutting down the Encounter will forcibly clean up the pawns without allowing their deaths to play. \nSo this time will allow deaths to play out and finish.
    )
    (else
      (Utilities|IsValid _returnvalue
        (:Is%20Valid
          (Utilities|IsValid _returnvalue_2
            (:Is%20Valid
              (bind _activedefenderencounter (Variables|InternalLogic|GetActiveDefenderEncounter)
                (:then
                  (Class|FortAthenaLivingWorldManager|EndEncounter _returnvalue_2 _activedefenderencounter)
                  (CallFunction|ClientNotifyStateChangeEvent (Variables|InternalLogic|GetCurrentAttackerEncounterLocationIndex) (Utilities|Array|Length (Variables|EncounterSetup|GetEncounterSpawnLocationData)) (Variables|InternalLogic|GetCurrentDefenderEliminationCount) (Variables|InternalLogic|GetCurrentDefenderEliminationTarget)))
                (:else)))))))))

(event Custom|DefenderEncounterShutdownDelay
  (bind _returnvalue (Game|GetGameState))
  (bind _returnvalue_2 (Actor|GetComponentbyClass _returnvalue "/Script/LagerRuntime.FortAthenaLivingWorldManager"))
  ;[1 1 1] Shut down Encounters when we are done with them.
  (Utilities|IsValid _returnvalue
    (:Is%20Valid
      (Utilities|IsValid _returnvalue_2
        (:Is%20Valid
          (bind _activedefenderencounter (Variables|InternalLogic|GetActiveDefenderEncounter)
            (:then
              (Class|FortAthenaLivingWorldManager|EndEncounter _returnvalue_2 _activedefenderencounter)
              (CallFunction|ClientNotifyStateChangeEvent (Variables|InternalLogic|GetCurrentAttackerEncounterLocationIndex) (Utilities|Array|Length (Variables|EncounterSetup|GetEncounterSpawnLocationData)) (Variables|InternalLogic|GetCurrentDefenderEliminationCount) (Variables|InternalLogic|GetCurrentDefenderEliminationTarget)))
            (:else)))))))

(event Custom|DefendersFallbackToMidfield
  (bind _defendersfallbacktomidfieldtimer_hf (Variables|EncounterSetup|GetDefendersFallbackToMidfieldTimer_HF))
  (bind _returnvalue (ScalableFloat|GetValueatLevel _defendersfallbacktomidfieldtimer_hf))
  (bind _currentattackerencounterlocationindex (Variables|InternalLogic|GetCurrentAttackerEncounterLocationIndex))
  (bind _returnvalue_2 (+ _currentattackerencounterlocationindex 1))
  (bind _currentattackerencounterlocationindex_2 (Variables|InternalLogic|GetCurrentAttackerEncounterLocationIndex))
  (bind _encountermidfieldvisualblueprints (Variables|InternalLogic|GetEncounterMidfieldVisualBlueprints))
  (bind _encountermidfieldvisualblueprints_2 (Variables|InternalLogic|GetEncounterMidfieldVisualBlueprints))
  ;[1 1 1] [AFTER DELAY]
  ; Defender Fallback to Spawn complete.
  ; Defenders Fallback farther to new Midfield, Attackers begin to push to the newly disabled Defender spawn that they will soon claim as their own.
  ;[BEGIN DELAY]
  (if (not (CallFunction|DefendersinFallbackPhase))
    (Utilities|FlowControl|Sequence
      (:then_0
        (if (Variables|EncounterSetup|GetUseMidfieldUIVisualIndicators)
          (if (Utilities|Array|IsValidIndex _encountermidfieldvisualblueprints _currentattackerencounterlocationindex_2)
            (bind _result (ArrayMacros|SafeArrayGet _encountermidfieldvisualblueprints _currentattackerencounterlocationindex_2)
              (:%20%20
                (Utilities|IsValid _result
                  (:Is%20Valid
                    (Class|BGAEncounterVisualMidfield|MiniMapDeactivate _result) ; Dim previous Midfield MiniMap marker
                  )))))))
      (:then_1
        (if (Variables|EncounterSetup|GetUseMidfieldUIVisualIndicators)
          (if (Utilities|Array|IsValidIndex _encountermidfieldvisualblueprints_2 _returnvalue_2)
            (bind _result_2 (ArrayMacros|SafeArrayGet _encountermidfieldvisualblueprints_2 _returnvalue_2)
              (:%20%20
                (Utilities|IsValid _result_2
                  (:Is%20Valid
                    (Class|BGAEncounterVisualMidfield|MiniMapActivate _result_2) ; Show new Midfield MiniMap marker
                  )))))))
      (:then_2
        (if (CallFunction|AttackersCanAdvanceorHaveWon)
          (if (> _returnvalue 0)
            (Utilities|Time|SetTimerbyFunctionName 0 "DefendersFallbackToMidfieldTimer" _returnvalue) ; Defenders Fallback for a while before Attackers will follow\nFallback from old Spawn to new Midfield
            (else
              (CallFunction|DefendersFallbacktoMidfieldTimer) ; Skip timer if invalid, do resulting stuff anyway.
            ))
          (CallFunction|EncounterDebugString "DEFENDERS: Fallback to Midfield." false (Math|Color|Desaturate(LinearColor) (Variables|EncounterSetup|GetDefenderDebugColor) 0.1))
          (CallFunction|LeashAllDefenderstoMidfield)
          (else
            (if (ScalableFloat|GetValueasBool (Variables|EncounterSetup|GetAttackersShouldDoFinalCleanupPushToDefenderSpawn_HF))
              (Variables|InternalLogic|SetAttackerFinalCleanupPushIsActive true) ; Set Final Push so we can leash newly spawned Attackers to the final Defender spawn.\nHelps them aid in the Cleanup process to mop-up the final Defenders.
              (else
                _))))
        (if (not (Variables|InternalLogic|GetDefendersFallbackDirectlyToNextMidfield))
          (CallFunction|AttackersAdvancetoSpawn) ; Now Attackers begin to Advance!
        ))
      (:then_3
        (if (Variables|InternalLogic|GetDefendersFallbackDirectlyToNextMidfield)
          _
          (elif (ScalableFloat|GetValueasBool (Variables|EncounterSetup|GetDefenderSpawnUpdatesImmediatelyOnFallback_HF))
            _
            (else
              (CallFunction|EnableDefenderSpawn (Variables|EncounterSounds|GetDefenderSpawnWarmUpTime) true) ; Enable next Defender Spawn
            )))))))

(event Custom|DefendersFallbackToSpawnTimer
  ;[1 1 1] [AFTER DELAY]
  ; Defender Fallback to Spawn complete.
  ; Defenders Fallback farther to new Midfield, Attackers begin to push to the newly disabled Defender spawn that they will soon claim as their own.
  ;[BEGIN DELAY]
  (Utilities|Time|ClearTimerbyFunctionName 0 "DefendersFallbackToSpawnTimer")
  (CallFunction|EncounterDebugString "DEFENDERS: Fallback to Spawn DONE." false (Math|Color|Desaturate(LinearColor) (Variables|EncounterSetup|GetDefenderDebugColor) 0.1))
  (CallFunction|DefendersFallbacktoMidfield) ; Now Defenders Fallback to Midfield
)

(event Custom|EndAttackerEncounter
  (bind _returnvalue (Game|GetGameState))
  (bind _returnvalue_2 (Actor|GetComponentbyClass _returnvalue "/Script/LagerRuntime.FortAthenaLivingWorldManager"))
  ;[1 1 1] Shut down Encounters when we are done with them.
  (Utilities|IsValid _returnvalue
    (:Is%20Valid
      (Utilities|IsValid _returnvalue_2
        (:Is%20Valid
          (bind _activeattackerencounter (Variables|InternalLogic|GetActiveAttackerEncounter)
            (:then
              (Class|FortAthenaLivingWorldManager|EndEncounter _returnvalue_2 _activeattackerencounter)
              (CallFunction|ClientNotifyStateChangeEvent (Variables|InternalLogic|GetCurrentAttackerEncounterLocationIndex) (Utilities|Array|Length (Variables|EncounterSetup|GetEncounterSpawnLocationData)) (Variables|InternalLogic|GetCurrentDefenderEliminationCount) (Variables|InternalLogic|GetCurrentDefenderEliminationTarget)))
            (:else)))))))

(event Custom|AttackersAdvanceToSpawn
  (bind _currentattackerencounterlocationindex (Variables|InternalLogic|GetCurrentAttackerEncounterLocationIndex))
  (bind _returnvalue (+ _currentattackerencounterlocationindex 1))
  (bind _encounterspawnvisualblueprints (Variables|InternalLogic|GetEncounterSpawnVisualBlueprints))
  ;[1 1 1] [AFTER DELAY]
  ; Defender Fallback to Spawn complete.
  ; Defenders Fallback farther to new Midfield, Attackers begin to push to the newly disabled Defender spawn that they will soon claim as their own.
  ;[BEGIN DELAY]
  (CallFunction|EncounterDebugString "ATTACKERS: Advance to Defender Spawn." false (Math|Color|Desaturate(LinearColor) (Variables|EncounterSetup|GetAttackerDebugColor) 0.1))
  (CallFunction|LeashAllAttackerstoDefenderSpawn) ; Leash Attackers to Defender Spawn that was just recently abandoned.\nAttackers will not spawn there yet.
  (Utilities|FlowControl|Sequence
    (:then_0
      (if (and (not (Variables|InternalLogic|GetAttackerFinalCleanupPushIsActive)) (Utilities|Array|IsValidIndex _encounterspawnvisualblueprints _returnvalue))
        (bind _result (ArrayMacros|SafeArrayGet _encounterspawnvisualblueprints _returnvalue)
          (:%20%20
            (Utilities|IsValid _result
              (:Is%20Valid
                (Variables|/Lager/Lager/Encounters/BGA_Encounter_Visual_Spawn.BGA_Encounter_Visual_Spawn_C|SetEncounterSpawnerOwnershipState "NewEnumerator3" _result) ; Preview next Attacker Spawn
              ))))))
    (:then_1
      (if (Variables|EncounterSetup|GetDebugEncounter)
        (CallFunction|DebugDisablePreviousMidfield (Variables|InternalLogic|GetCurrentAttackerEncounterLocationIndex))))))

(event Custom|DebugDisablePreviousMidfield
  (params (Index :type Integer))
  (bind _encountermidfieldcenterarrows (Variables|InternalLogic|GetEncounterMidfieldCenterArrows))
  ;[1 1 1] [AFTER DELAY]
  ; Defender Fallback to Spawn complete.
  ; Defenders Fallback farther to new Midfield, Attackers begin to push to the newly disabled Defender spawn that they will soon claim as their own.
  ;[BEGIN DELAY]
  (if (Utilities|Array|IsValidIndex _encountermidfieldcenterarrows Index)
    (bind _result (ArrayMacros|SafeArrayGet _encountermidfieldcenterarrows Index)
      (:%20%20
        (CallFunction|DebugShoworHideArrowandChildren _result true) ; Hide previous Midfield Debug Info
      ))))

(event Custom|StopLinearEncounter
  (Default|DisableEncounter))

(event Custom|Debug_StartLinearEncounter
  (bind _outactors (Actor|GetAllActorsOfClass "/Lager/Lager/Encounters/BGA_Encounter_Linear_Parent.BGA_Encounter_Linear_Parent_C")) ; Get all other Linear Encounters
  (bind _array_element (Utilities|Array|ForEachLoopwithBreak _outactors)
    (:LoopBody
      (Default|DisableEncounter _array_element))
    (:Completed
      (Default|StartEncounter))))

(event Custom|DoAttackerEmotes
  (bind (_outitem _outindex) (Utilities|Array|RandomArrayItem (Variables|EncounterSetup|GetAttackerVictoryEmotes)))
  ;[1 1 1] Do Attacker Victory Emotes
  (for _array_element (Variables|InternalLogic|GetSpawnedPawnsAttackers)
    (Utilities|IsValid _array_element
      (:Is%20Valid
        (Utilities|Array|AddUnique (Variables|InternalLogic|GetAttackersWhoShouldDoEmotes) _array_element) ; List of all existing AI to pull from.\n\nWe will pare this list down, which is why we're duplicating it.
      )))
  (Utilities|Array|Shuffle (Variables|InternalLogic|GetAttackersWhoShouldDoEmotes)) ; Mix up guys so they aren't in spawn order.
  (bind _result (ArrayMacros|SafeArrayGet (Variables|InternalLogic|GetAttackersWhoShouldDoEmotes))
    (:%20%20
      (Utilities|IsValid _result
        (:Is%20Valid
          (Utilities|FlowControl|Delay (Math|Random|RandomFloatinRange 0.75 1.5)) ; Randomly delay so they don't all start at once.
          (bind _asfort_athena_aibot_controller (Utilities|Casting|CastToFortAthenaAIBotController (Pawn|GetController _result)))
          (bind _result_2 (ArrayMacros|SafeArrayGet (Variables|EncounterSetup|GetAttackerVictoryEmotes) _outindex)
            (:%20%20
              (if (AssetManager|IsValidPrimaryAssetId _result_2)
                (Class|FortAthenaAIBotController|RequestEmote _asfort_athena_aibot_controller _result_2 -1.0)
                (Utilities|Array|RemoveIndex (Variables|InternalLogic|GetAttackersWhoShouldDoEmotes)) ; 0th guys just danced, remove 0th guy from list.
              ))) ; Get a random Emote
        )
        (:Is%20Not%20Valid
          _))
      (Utilities|FlowControl|Delay (Math|Random|RandomFloatinRange 1e+01 2e+01)) ; After we do Emotes, delay, then do Emotes again.\n\nLoop forever.
      (CallFunction|DoAttackerEmotes))))

(event Custom|DebugAttackerDisableSpawn
  (params (Index :type Integer))
  (bind _returnvalue (+ Index 1))
  (bind _attackerencountertomidfieldarrows (Variables|InternalLogic|GetAttackerEncounterToMidfieldArrows))
  (bind _midfieldtodefenseencounterarrows (Variables|InternalLogic|GetMidfieldToDefenseEncounterArrows))
  (bind _encounterdebugarrows (Variables|InternalLogic|GetEncounterDebugArrows))
  (bind _attackerencountertomidfieldarrows_2 (Variables|InternalLogic|GetAttackerEncounterToMidfieldArrows))
  (bind _attackerdebugcolor (Variables|EncounterSetup|GetAttackerDebugColor))
  (bind _encounterdebugarrows_2 (Variables|InternalLogic|GetEncounterDebugArrows))
  ;[1 1 1] [AFTER DELAY]
  ; Attackers claim new Spawn Location and, if still fighting, begin to press forward to the new Midfield.
  ; If Attackers have claimed Victory, allow Attackers to spawn and respawn at the final Spawn Location for a short time before disabling it.
  (if (Utilities|Array|IsValidIndex _encounterdebugarrows_2 Index)
    (bind _result (ArrayMacros|SafeArrayGet _encounterdebugarrows_2 Index)
      (:%20%20
        (CallFunction|DebugShoworHideArrowandChildren _result true) ; Hide previous Attacker Debug Spawn Info
        (if (and (Utilities|Array|IsValidIndex _attackerencountertomidfieldarrows Index) (Utilities|Array|IsValidIndex _midfieldtodefenseencounterarrows Index))
          (bind _result_2 (ArrayMacros|SafeArrayGet _attackerencountertomidfieldarrows Index)
            (:%20%20
              (bind _result_3 (ArrayMacros|SafeArrayGet _midfieldtodefenseencounterarrows Index)
                (:%20%20
                  (Development|SetHiddeninGame _result_2 true false) ; Hide previous arrow segments for fights no longer happening.
                  (if (Utilities|Array|IsValidIndex _encounterdebugarrows _returnvalue)
                    (bind _result_4 (ArrayMacros|SafeArrayGet _encounterdebugarrows _returnvalue)
                      (:%20%20
                        (Components|Arrow|SetArrowColor(LinearColor) _result_4 _attackerdebugcolor) ; Set current Arrow colors to Attacker
                        (if (Utilities|Array|IsValidIndex _attackerencountertomidfieldarrows_2 _returnvalue)
                          (bind _result_5 (ArrayMacros|SafeArrayGet _attackerencountertomidfieldarrows_2 _returnvalue)
                            (:%20%20
                              (Components|Arrow|SetArrowColor(LinearColor) _result_5 _attackerdebugcolor))))))))))))))))

(event Custom|DefendersFallbackToMidfieldTimer
  ;[1 1 1] [AFTER DELAY]
  ; Attackers claim new Spawn Location and, if still fighting, begin to press forward to the new Midfield.
  ; If Attackers have claimed Victory, allow Attackers to spawn and respawn at the final Spawn Location for a short time before disabling it.
  (Utilities|Time|ClearTimerbyFunctionName 0 "DefendersFallbackToMidfieldTimer")
  (CallFunction|EncounterDebugString "DEFENDERS: Fallback to Midfield DONE." false (Math|Color|Desaturate(LinearColor) (Variables|EncounterSetup|GetDefenderDebugColor) 0.1))
  (CallFunction|AttackersSetNewSpawnandAdvancetoMidfield) ; Attackers claim new Spawn Location
)

(event Custom|AttackersSetNewSpawnAndAdvanceToMidfield
  (bind _attackervictoryemotedelay_hf (Variables|EncounterSetup|GetAttackerVictoryEmoteDelay_HF))
  (bind _returnvalue (ScalableFloat|GetValueatLevel _attackervictoryemotedelay_hf))
  (bind _attackerrespawntimerafterwinning_hf (Variables|EncounterSetup|GetAttackerRespawnTimerAfterWinning_HF))
  (bind _returnvalue_2 (ScalableFloat|GetValueatLevel _attackerrespawntimerafterwinning_hf))
  (bind _encounterspawnvisualblueprints (Variables|InternalLogic|GetEncounterSpawnVisualBlueprints))
  (bind _currentattackerencounterlocationindex (Variables|InternalLogic|GetCurrentAttackerEncounterLocationIndex))
  (bind _spawnedattackerlivingworldpointproviders (Variables|InternalLogic|GetSpawnedAttackerLivingWorldPointProviders))
  (bind _currentattackerencounterlocationindex_2 (Variables|InternalLogic|GetCurrentAttackerEncounterLocationIndex))
  ;[1 1 1] [AFTER DELAY]
  ; Attackers claim new Spawn Location and, if still fighting, begin to press forward to the new Midfield.
  ; If Attackers have claimed Victory, allow Attackers to spawn and respawn at the final Spawn Location for a short time before disabling it.
  (if (Variables|InternalLogic|GetAttackersHaveClaimedFinalSpawnLocation)
    _
    (elif (Utilities|Array|IsValidIndex _spawnedattackerlivingworldpointproviders _currentattackerencounterlocationindex_2)
      (bind _result (ArrayMacros|SafeArrayGet _spawnedattackerlivingworldpointproviders _currentattackerencounterlocationindex_2)
        (:%20%20
          (Class|FortAthenaLivingWorldVolume|DisablePointProvider _result) ; Disable current Attacker Spawn
          (Utilities|FlowControl|Sequence
            (:then_0
              (if (Variables|InternalLogic|GetDefendersFallbackDirectlyToNextMidfield)
                (if (not (ScalableFloat|GetValueasBool (Variables|EncounterSetup|GetDefenderSpawnUpdatesImmediatelyOnFallback_HF)))
                  (CallFunction|EnableDefenderSpawn (Variables|EncounterSounds|GetDefenderSpawnWarmUpTime) true 1) ; Enable next Defender Spawn
                )))
            (:then_1
              (if (Utilities|Array|IsValidIndex _encounterspawnvisualblueprints _currentattackerencounterlocationindex)
                (bind _result_2 (ArrayMacros|SafeArrayGet _encounterspawnvisualblueprints _currentattackerencounterlocationindex)
                  (:%20%20
                    (Utilities|IsValid _result_2
                      (:Is%20Valid
                        (if (ScalableFloat|GetValueasBool (Variables|EncounterSetup|GetAttackerMiniMapIconsHideAlongTheWay_HF))
                          (Variables|/Lager/Lager/Encounters/BGA_Encounter_Visual_Spawn.BGA_Encounter_Visual_Spawn_C|SetEncounterSpawnerOwnershipState "NewEnumerator0" _result_2) ; Hide Attacker Spawn MiniMap marker
                          (else
                            (Variables|/Lager/Lager/Encounters/BGA_Encounter_Visual_Spawn.BGA_Encounter_Visual_Spawn_C|SetEncounterSpawnerOwnershipState "NewEnumerator3" _result_2) ; Disable (NOT Hide) Attacker Spawn MiniMap marker
                          ))))))))
            (:then_2
              (if (Variables|EncounterSetup|GetDebugEncounter)
                (CallFunction|DebugAttackerDisableSpawn (Variables|InternalLogic|GetCurrentAttackerEncounterLocationIndex))))
            (:then_3
              (if (Variables|InternalLogic|GetDefendersFallbackDirectlyToNextMidfield)
                (if (Variables|EncounterSetup|GetDebugEncounter)
                  (CallFunction|DebugDisablePreviousMidfield (Variables|InternalLogic|GetCurrentAttackerEncounterLocationIndex)))))
            (:then_4
              (if (Utilities|Array|IsValidIndex (Variables|EncounterSetup|GetEncounterSpawnLocationData) (+ (Variables|InternalLogic|GetCurrentAttackerEncounterLocationIndex) 1))
                (bind _result_3 (Math|Integer|IncrementInt (Variables|InternalLogic|GetCurrentAttackerEncounterLocationIndex) :!pinned true)
                  (:%20%20
                    (CallFunction|EncounterDebugString (Utilities|String|BuildString(Integer) "" "ATTACKERS: New Location ID: " _result_3) false (Math|Color|Desaturate(LinearColor) (Variables|EncounterSetup|GetAttackerDebugColor) 0.1))
                    (Variables|InternalLogic|SetAttackersHaveClaimedFinalSpawnLocation (== (Variables|InternalLogic|GetCurrentAttackerEncounterLocationIndex) (Utilities|Array|LastIndex (Variables|EncounterSetup|GetEncounterSpawnLocationData))) :!pinned true) ; Attackers have claimed victory?
                    (CallFunction|ResetEncounterProgressforNewLocation)
                    (CallFunction|EnableAttackerSpawn (Variables|EncounterSounds|GetAttackerSpawnWarmUpTime) true)
                    (if (Variables|InternalLogic|GetAttackersHaveClaimedFinalSpawnLocation)
                      (CallFunction|LeashAllAttackerstoDefenderSpawn)
                      (if (> _returnvalue_2 0)
                        (Utilities|Time|SetTimerbyFunctionName 0 "AttackerRespawnTimerAfterWinning" _returnvalue_2) ; If this is the Attackers' last spawn location. Shut off the spawn after a while so they can clean up naturally.
                        (else
                          (CallFunction|AttackerRespawnTimerAfterWinning) ; Invalid timer? Do immediately...
                        ))
                      (CallFunction|EndDefenderEncounter)
                      (if (> _returnvalue 0)
                        (Utilities|FlowControl|Delay _returnvalue) ; Delay before doing victory dances.
                      )
                      (CallFunction|DoAttackerEmotes)
                      (else
                        (Default|LeashAllAttackerstoMidfield))))) ; Update Attacker Location
              ))))))))

(event Custom|DelayAndDropLootForAttackers
  (params (LootActor :type Actor%20Object%20Reference) (FinalWinningDrop :type Boolean))
  ;[1 1 1] Delay and Drop Loot for Attackers
  (if (> (Variables|EncounterSetup|GetLootDropDelay) 0)
    (Utilities|FlowControl|Delay (Variables|EncounterSetup|GetLootDropDelay))
    (else
      _))
  (CallFunction|DropLootforAttackers LootActor (Variables|EncounterSetup|GetAttackerLootDropOffset) (select FinalWinningDrop (Variables|EncounterSetup|GetLootTierAttackersWin) (Variables|EncounterSetup|GetLootTierAttackersProgress))) ; Drop Loot for Attackers
  (CallFunction|AttackerLootDropSound LootActor)
  (if FinalWinningDrop
    (Class|FortAthenaLinearEncounterAnalyticComponent|OnEncounterEnd (Variables|Default|GetFortAthenaLinearEncounterAnalytic))))

(event Custom|AttackerLootDropSound
  (params (LootActor :type Actor%20Object%20Reference))
  ;[1 1 1] Loot Drop Sounds
  (Utilities|IsValid LootActor
    (:Is%20Valid
      (Networking|CanExecuteCosmeticEvents
        (:True
          (Audio|PlaySoundatLocation (Variables|EncounterSounds|GetSoundAttackerLootDrop) (Transformation|GetActorLocation LootActor)))) ; Only play for Clients
    )))

(event Custom|AttackerRespawnTimerAfterWinning
  (bind _currentattackerencounterlocationindex (Variables|InternalLogic|GetCurrentAttackerEncounterLocationIndex))
  (bind _returnvalue (- _currentattackerencounterlocationindex 1))
  (bind _spawnedattackerlivingworldpointproviders (Variables|InternalLogic|GetSpawnedAttackerLivingWorldPointProviders))
  (bind _encounterspawnvisualblueprints (Variables|InternalLogic|GetEncounterSpawnVisualBlueprints))
  (bind _currentattackerencounterlocationindex_2 (Variables|InternalLogic|GetCurrentAttackerEncounterLocationIndex))
  (bind _encountermidfieldvisualblueprints (Variables|InternalLogic|GetEncounterMidfieldVisualBlueprints))
  (bind _currentattackerencounterlocationindex_3 (Variables|InternalLogic|GetCurrentAttackerEncounterLocationIndex))
  ;[1 1 1] [AFTER DELAY]
  ; Attackers claim new Spawn Location and, if still fighting, begin to press forward to the new Midfield.
  ; If Attackers have claimed Victory, allow Attackers to spawn and respawn at the final Spawn Location for a short time before disabling it.
  (if (Utilities|Array|IsValidIndex _spawnedattackerlivingworldpointproviders _currentattackerencounterlocationindex_3)
    (Utilities|FlowControl|Sequence
      (:then_0
        (bind _result (ArrayMacros|SafeArrayGet _spawnedattackerlivingworldpointproviders _currentattackerencounterlocationindex_3)
          (:%20%20
            (Class|FortAthenaLivingWorldVolume|DisablePointProvider _result :!pinned true) ; Shut down final Attacker Spawn Location
            (if (ScalableFloat|GetValueasBool (Variables|EncounterSetup|GetAttackersDespawnAfterWinning_HF))
              (CallFunction|EndAttackerEncounter) ; Despawns Attackers
              (else
                _))
            (Variables|InternalLogic|SetLinearEncounterIsActive false)
            (if (Variables|EncounterSetup|GetDebugEncounter)
              (CallFunction|DebugAttackerDisableSpawn (Variables|InternalLogic|GetCurrentAttackerEncounterLocationIndex))))))
      (:then_1
        (if (Utilities|Array|IsValidIndex _encounterspawnvisualblueprints _currentattackerencounterlocationindex_2)
          (bind _result_2 (ArrayMacros|SafeArrayGet _encounterspawnvisualblueprints _currentattackerencounterlocationindex_2)
            (:%20%20
              (Utilities|IsValid _result_2
                (:Is%20Valid
                  (if (ScalableFloat|GetValueasBool (Variables|EncounterSetup|GetAttackerMiniMapIconsHideAlongTheWay_HF))
                    (Variables|/Lager/Lager/Encounters/BGA_Encounter_Visual_Spawn.BGA_Encounter_Visual_Spawn_C|SetEncounterSpawnerOwnershipState "NewEnumerator0" _result_2) ; Hide Attacker Spawn MiniMap marker
                    (else
                      (Variables|/Lager/Lager/Encounters/BGA_Encounter_Visual_Spawn.BGA_Encounter_Visual_Spawn_C|SetEncounterSpawnerOwnershipState "NewEnumerator3" _result_2) ; This will disable the "I am activated and can spawn guys" visuals first.
                    ))))))))
      (:then_2
        (if (Variables|EncounterSetup|GetUseMidfieldUIVisualIndicators)
          (if (Utilities|Array|IsValidIndex _encountermidfieldvisualblueprints _returnvalue)
            (bind _result_3 (ArrayMacros|SafeArrayGet _encountermidfieldvisualblueprints _returnvalue)
              (:%20%20
                (Utilities|IsValid _result_3
                  (:Is%20Valid
                    (Class|BGAEncounterVisualMidfield|MiniMapHideAll _result_3) ; Hide previous Midfield MiniMap marker
                  )))))))
      (:then_3
        (Utilities|FlowControl|Delay 6e+01)
        (CallFunction|SetEncounterActorsDormancy "DORM_DormantAll")))))

(fn AllDefendersDefeated
  (return (and (Utilities|Array|IsEmpty (Variables|InternalLogic|GetSpawnedPawnsDefenders)) (Variables|InternalLogic|GetLastDefenderSpawnDefeated))))

(fn AttackersCanAdvanceOrHaveWon
  (return (or (CallFunction|EncounterStillHasValidCombatEncountersBeyondCurrentOne) (and (not (CallFunction|EncounterStillHasValidCombatEncountersBeyondCurrentOne)) (CallFunction|AllDefendersDefeated)))))

(fn AttackersDamagedByPlayerRecently
  (return (Utilities|Time|IsTimerActivebyFunctionName 0 "AttackersChallengedTimer")))

(fn ClearAllAttackersChallenged
  (for _array_element (Variables|InternalLogic|GetSpawnedPawnsAttackers)
    (Default|ClearSingleAIChallenged _array_element)))

(fn ClearAllDefendersChallenged
  (for _array_element (Variables|InternalLogic|GetSpawnedPawnsDefenders)
    (Default|ClearSingleAIChallenged _array_element)))

(fn ClearSingleAIChallenged
  (params (AI :type Fort%20Pawn%20Object%20Reference))
  (bind _tc_playerencounterengagementtags (Variables|EncounterSetup|GetTC_PlayerEncounterEngagementTags))
  (bind _returnvalue (GameplayTags|MakeLiteralGameplayTagContainer _tc_playerencounterengagementtags))
  (if (GameplayTags|HasAnyMatchingGameplayTags AI _returnvalue)
    (Ability|GameplayEffect|RemoveLooseGameplayTags AI _returnvalue)))

(fn ClientNotifyStateChange
  (params (LocationIndex :type Integer) (LocationCount :type Integer) (EliminationCount :type Integer) (EliminationTarget :type Integer)))

(fn ClientSpawnShutDownFX
  (params (FX%20Location :type Vector) (VisualSpawnBP :type BGA%20Encounter%20Visual%20Spawn%20Object%20Reference))
  (Networking|CanExecuteCosmeticEvents
    (:True
      (Utilities|IsValid VisualSpawnBP
        (:Is%20Valid
          (Effects|Components|ParticleSystem|SpawnFXAtLocation (Variables|EncounterVFX|GetVFX_SpawnerLootSpray) FX%20Location) ; Loot Spray VFX when Attackers take new Spawn
        )))))

(fn CreateSpawnPointProvidersAtSpawnLocationIndex
  (params (EncounterSpawnLocationDataIndex :type Integer))
  (bind _returnvalue (Transformation|GetActorTransform))
  (bind _currentencounterspawnlocationdataindex (LocalVariables|GetCurrentEncounterSpawnLocationDataIndex))
  (bind _returnvalue_2 (- _currentencounterspawnlocationdataindex 1))
  (bind _currentencounterspawnlocationdataindex_2 (LocalVariables|GetCurrentEncounterSpawnLocationDataIndex))
  (bind _defendermidfieldlocations (Variables|EncounterSetup|GetDefenderMidfieldLocations))
  (bind _pointproviderclass (Variables|InternalLogic|GetPointProviderClass))
  (bind _attackermidfieldlocations (Variables|EncounterSetup|GetAttackerMidfieldLocations))
  (LocalVariables|SetCurrentEncounterSpawnLocationDataIndex EncounterSpawnLocationDataIndex)
  (if (Utilities|Array|IsValidIndex (Variables|EncounterSetup|GetEncounterSpawnLocationData) (LocalVariables|GetCurrentEncounterSpawnLocationDataIndex))
    (bind _result (ArrayMacros|SafeArrayGet (Variables|EncounterSetup|GetEncounterSpawnLocationData) (LocalVariables|GetCurrentEncounterSpawnLocationDataIndex))
      (:%20%20
        (Utilities|FlowControl|Sequence :!pinned true
          (:then_0
            (bind _returnvalue_3 (Game|SpawnActorfromClass _pointproviderclass (Math|Transform|MakeTransform (.location (Math|Transform|ComposeTransforms (Utilities|Struct|BreakStructEncounterLinearData _result) _returnvalue)) (.rotation (Math|Transform|ComposeTransforms (Utilities|Struct|BreakStructEncounterLinearData _result) _returnvalue))) "AlwaysSpawn" "SelectDefaultAtRuntime" 0 0 :!pinned true)) ; Spawn Attacker Point Providers
            (Utilities|Array|SetArrayElem (Variables|InternalLogic|GetSpawnedAttackerLivingWorldPointProviders) (LocalVariables|GetCurrentEncounterSpawnLocationDataIndex) _returnvalue_3 true)
            (Class|FortAthenaLivingWorldVolume|SetEQSQuery _returnvalue_3 (Variables|InternalLogic|GetAttackerPointProviderEnvQuery))
            (Class|FortAthenaLivingWorldVolume|SetFiltersTags _returnvalue_3 (Variables|InternalLogic|GetAttackerPointProviderTags))
            (Class|FortAthenaLivingWorldVolume|DisablePointProvider _returnvalue_3)
            (Utilities|Array|SetArrayElem (Variables|InternalLogic|GetAttackerEncounterPointProviders) (LocalVariables|GetCurrentEncounterSpawnLocationDataIndex) (Utilities|Casting|CastToFortAthenaLivingWorldPointProviderInterface _returnvalue_3) true) ; Feed all of these through the Interface cast, otherwise \nwe got a crash using the Point Providers directly
            (if (Utilities|Array|IsValidIndex _attackermidfieldlocations _currentencounterspawnlocationdataindex_2)
              (Class|FortAthenaLivingWorldVolume|SetSpawnRotation _returnvalue_3 (Math|Rotator|MakeRotator 0.0 0.0 (.yaw (Math|Rotator|FindLookatRotation (.location (Math|Transform|MakeTransform (.location (Math|Transform|ComposeTransforms (Utilities|Struct|BreakStructEncounterLinearData _result) _returnvalue)) (.rotation (Math|Transform|ComposeTransforms (Utilities|Struct|BreakStructEncounterLinearData _result) _returnvalue)))) (.location (Math|Transform|ComposeTransforms (Math|Conversions|ToTransform(Vector) (Utilities|Struct|BreakStructEncounterLeashData (Utilities|Array|Get(acopy) _attackermidfieldlocations _currentencounterspawnlocationdataindex_2))) (Transformation|GetActorTransform)))))))))
          (:then_1
            (bind _returnvalue_4 (Game|SpawnActorfromClass _pointproviderclass (Math|Transform|MakeTransform (.location (Math|Transform|ComposeTransforms (Utilities|Struct|BreakStructEncounterLinearData _result) _returnvalue)) (.rotation (Math|Transform|ComposeTransforms (Utilities|Struct|BreakStructEncounterLinearData _result) _returnvalue))) "AlwaysSpawn" "SelectDefaultAtRuntime" 0 0 :!pinned true)) ; Spawn Defender Point Providers
            (Utilities|Array|SetArrayElem (Variables|InternalLogic|GetSpawnedDefenderLivingWorldPointProviders) (LocalVariables|GetCurrentEncounterSpawnLocationDataIndex) _returnvalue_4 true)
            (Class|FortAthenaLivingWorldVolume|SetEQSQuery _returnvalue_4 (Variables|InternalLogic|GetDefenderPointProviderEnvQuery))
            (Class|FortAthenaLivingWorldVolume|SetFiltersTags _returnvalue_4 (Variables|InternalLogic|GetDefenderPointProviderTags))
            (Class|FortAthenaLivingWorldVolume|DisablePointProvider _returnvalue_4)
            (Utilities|Array|SetArrayElem (Variables|InternalLogic|GetDefenderEncounterPointProviders) (LocalVariables|GetCurrentEncounterSpawnLocationDataIndex) (Utilities|Casting|CastToFortAthenaLivingWorldPointProviderInterface _returnvalue_4) true) ; Feed all of these through the Interface cast, otherwise \nwe got a crash using the Point Providers directly
            (if (Utilities|Array|IsValidIndex _defendermidfieldlocations _returnvalue_2)
              (Class|FortAthenaLivingWorldVolume|SetSpawnRotation _returnvalue_4 (Math|Rotator|MakeRotator 0.0 0.0 (.yaw (Math|Rotator|FindLookatRotation (.location (Math|Transform|MakeTransform (.location (Math|Transform|ComposeTransforms (Utilities|Struct|BreakStructEncounterLinearData _result) _returnvalue)) (.rotation (Math|Transform|ComposeTransforms (Utilities|Struct|BreakStructEncounterLinearData _result) _returnvalue)))) (.location (Math|Transform|ComposeTransforms (Math|Conversions|ToTransform(Vector) (Utilities|Struct|BreakStructEncounterLeashData (Utilities|Array|Get(acopy) _defendermidfieldlocations _returnvalue_2))) (Transformation|GetActorTransform)))))))))) ; We create BOTH Attacker and Defender Point Providers at each location.
      )) ; Get the index.
  ))

(fn DebugElementShouldBeHidden
  (if (Editor|InEditororPIE)
    (if (CallFunction|EncounterIsEnabled)
      (if (Variables|EncounterSetup|GetDebugEncounter)
        (return false)
        (else
          _))
      (else
        _))
    (return true)
    (else
      _)))

(fn DebugShowOrHideArrowAndChildren
  (params (Arrow :type Arrow%20Component%20Object%20Reference) (MakeHidden :type Boolean) (OnlyShowOrHideChildren :type Boolean))
  (if (Variables|EncounterSetup|GetDebugEncounter)
    (if OnlyShowOrHideChildren
      (for _array_element (Components|GetChildrenComponents Arrow true)
        (Development|SetHiddeninGame _array_element MakeHidden true))
      (else
        (Development|SetHiddeninGame Arrow MakeHidden true)))))

(fn DefendersDamagedByPlayerRecently
  (return (Utilities|Time|IsTimerActivebyFunctionName 0 "DefendersChallengedTimer")))

(fn DefendersInFallbackPhase
  (return (or (Utilities|Time|IsTimerActivebyFunctionName 0 "DefendersFallbackToSpawnTimer") (Utilities|Time|IsTimerActivebyFunctionName 0 "DefendersFallbackToMidfieldTimer"))))

(fn DisableEncounter
  (bind _returnvalue (Game|GetGameState))
  (bind _returnvalue_2 (Actor|GetComponentbyClass _returnvalue "/Script/LagerRuntime.FortAthenaLivingWorldManager"))
  (bind _array (Utilities|Array|MakeArray))
  (Utilities|FlowControl|Sequence
    (:then_0
      ;[1 1 1] Hide encounter's visuals
      (Rendering|SetActorHiddenInGame true)
      (bind _encountertriggersphere (Variables|Default|GetEncounterTriggerSphere)
        (:then
          (Variables|/Script/Engine.PrimitiveComponent|SetbGenerateOverlapEvents false _encountertriggersphere)
          (Collision|SetCollisionEnabled _encountertriggersphere "NoCollision" :!pinned true) ; Shut off collision on trigger since we no longer need it.
          (Development|SetHiddeninGame _encountertriggersphere true false) ; Hide Trigger (if debug had it shown)
        )
        (:else
          _))
      (for _array_element (Variables|InternalLogic|GetEncounterSpawnVisualBlueprints)
        (Rendering|SetActorHiddenInGame _array_element true)
        (Class|BGAEncounterVisualSpawn|MiniMapSetHidden _array_element)
        (Variables|/Lager/Lager/Encounters/BGA_Encounter_Visual_Spawn.BGA_Encounter_Visual_Spawn_C|SetEncounterSpawnerOwnershipState "NewEnumerator0" _array_element)
        (Collision|SetCollisionEnabled (Variables|/Lager/Lager/Encounters/BGA_Encounter_Visual_Spawn.BGA_Encounter_Visual_Spawn_C|GetStaticMesh _array_element) "NoCollision" :!pinned true) ; Disable collision on spawner
      )
      (if (Variables|EncounterSetup|GetUseMidfieldUIVisualIndicators)
        (for _array_element_2 (Variables|InternalLogic|GetEncounterMidfieldVisualBlueprints)
          (Utilities|IsValid _array_element_2
            (:Is%20Valid
              (Class|BGAEncounterVisualMidfield|MiniMapHideAll _array_element_2))))))
    (:then_1
      ;[1 1 1] End LWM Encounters
      (Utilities|IsValid _returnvalue
        (:Is%20Valid
          (Utilities|IsValid _returnvalue_2
            (:Is%20Valid
              (Class|FortAthenaLivingWorldManager|EndEncounter _returnvalue_2 (Variables|InternalLogic|GetActiveAttackerEncounter))
              (Class|FortAthenaLivingWorldManager|EndEncounter _returnvalue_2 (Variables|InternalLogic|GetActiveDefenderEncounter))
              (Variables|InternalLogic|SetActiveAttackerEncounter 0)
              (Variables|InternalLogic|SetActiveDefenderEncounter 0)
              (Class|FortAthenaLivingWorldManager|UnregisterDensityReservoir _returnvalue_2 self))))))
    (:then_2
      ;[1 1 1] Reset Variables
      (Variables|InternalLogic|SetCurrentAttackerEncounterLocationIndex 0)
      (Variables|InternalLogic|SetAttackersHaveClaimedFinalSpawnLocation false)
      (Variables|InternalLogic|SetSpawnedPawnsAttackers _array)
      (Variables|InternalLogic|SetSpawnedPawnsDefenders _array)
      (Variables|InternalLogic|SetLastPlayerParticipationTimeDamagingDefenders 0.0)
      (Variables|InternalLogic|SetAttackerFinalCleanupPushIsActive false) ; Set Final Push so we can leash newly spawned Attackers to the final Defender spawn.\nHelps them aid in the Cleanup process to mop- the final Defenders.
      (Variables|InternalLogic|SetHoldingForDefenderCleanup false)
      (Variables|InternalLogic|SetLastDefenderSpawnDefeated false)
      (Variables|InternalLogic|SetLinearEncounterIsActive false)
      (Variables|InternalLogic|SetDisabled true))
    (:then_3
      ;[1 1 1] Disable Point Providers
      (Class|FortAthenaLivingWorldPointProviderInterface|DisablePointProvider (Variables|InternalLogic|GetAttackerEncounterPointProviders))
      (Class|FortAthenaLivingWorldPointProviderInterface|DisablePointProvider (Variables|InternalLogic|GetDefenderEncounterPointProviders)))
    (:then_4
      ;[1 1 1] Clear all timers
      (Utilities|Time|ClearTimerbyFunctionName 0 "DefendersFallbackToSpawnTimer")
      (Utilities|Time|ClearTimerbyFunctionName 0 "DefendersFallbackToMidfieldTimer")
      (Utilities|Time|ClearTimerbyFunctionName 0 "AttackerRespawnTimerAfterWinning")
      (Utilities|Time|ClearTimerbyFunctionName 0 "DefendersChallengedTimer")
      (Utilities|Time|ClearTimerbyFunctionName 0 "AttackersChallengedTimer")
      (Utilities|Time|ClearTimerbyFunctionName 0 "AttackerSpawnWarmupComplete")
      (Utilities|Time|ClearTimerbyFunctionName 0 "DefenderSpawnWarmupComplete")
      (Utilities|Time|ClearTimerbyFunctionName 0 "NoSpawnTimeout"))
    (:then_5
      (CallFunction|SetEncounterActorsDormancy "DORM_DormantAll"))
    (:then_6
      ;[1 1 1] Unbind POIVolume Overlap Events
      (bind _asfort_game_state (Utilities|Casting|CastToFortGameState (Game|GetGameState)))
      (Class|FortPoiManager|UnbindonPoiEntered (GameState|GetPoiManager _asfort_game_state) (Variables|EncounterSetup|GetPOITag) (EventDispatchers|CreateEvent :!event OnPlayerEnteredPoi))
      (Utilities|Time|ClearandInvalidateTimerbyHandle (Variables|InternalLogic|GetOnEnterDelayTimer)))))

(fn DropLootForAttackers
  (params (LootActor :type Actor%20Object%20Reference) (LootActorLocationOffset :type Vector) (LootTierToDrop :type Name))
  (if (!= LootTierToDrop "")
    (Utilities|IsValid LootActor
      (:Is%20Valid
        (bind _itempickupsspawned (Class|AthenaFunctionLibrary|SpawnLootfromTable LootTierToDrop 1 -1 (Utilities|Operators|Add :A (Transformation|GetActorLocation LootActor) :B LootActorLocationOffset) (Transformation|GetActorForwardVector LootActor) 0 true true))
        (Class|FortAthenaLinearEncounterAnalyticComponent|OnStageCompleted (Variables|Default|GetFortAthenaLinearEncounterAnalytic) _itempickupsspawned)))))

(fn EncounterDebugString
  (params (String :type String) (ForcePrint :type Boolean) (TextColor :type Linear%20Color%20Structure :default "(R=1.000000,G=1.000000,B=1.000000,A=1.000000)") (Duration :type Float%20%28double-precision%29 :default 2.0))
  (if (or ForcePrint (Variables|EncounterSetup|GetDebugEncounter))
    (Development|PrintString (Utilities|String|Append "===== " "Linear Encounter = " (Utilities|GetObjectName self) " | " String " | " (Utilities|String|BuildString(Integer) "" "Attacker Location: " (Variables|InternalLogic|GetCurrentAttackerEncounterLocationIndex) " ") "=====") true true TextColor Duration)))

(fn EncounterIsEnabled
  (bind _returnvalue (Class|FortAthenaLivingWorldManager|GetLivingWorldManager))
  (Utilities|IsValid _returnvalue
    (:Is%20Valid
      (return (Math|Boolean|ANDBoolean (ScalableFloat|GetValueasBool (Variables|EncounterSetup|GetLinearEncounterEnabled_HF)) (Class|FortAthenaLivingWorldManager|IsEncounterCategoryActive _returnvalue (Variables|EncounterSetup|GetAttackerEncounterClass)) (Class|FortAthenaLivingWorldManager|IsEncounterCategoryActive _returnvalue (Variables|EncounterSetup|GetDefenderEncounterClass)) (not (Variables|InternalLogic|GetDisabled)))))
    (:Is%20Not%20Valid
      (return false))))

(fn EncounterStillHasValidCombatEncountersBeyondCurrentOne
  (return (Utilities|Array|IsValidIndex (Variables|EncounterSetup|GetEncounterSpawnLocationData) (+ (Variables|InternalLogic|GetCurrentAttackerEncounterLocationIndex) 2))))

(fn LeashAllAttackersToDefenderSpawn
  (bind _attackerdebugcolor (Variables|EncounterSetup|GetAttackerDebugColor))
  (bind _returnvalue (Math|Color|Desaturate(LinearColor) _attackerdebugcolor 0.5))
  (bind _returnvalue_2 (Transformation|GetActorTransform))
  (bind _returnvalue_3 (Math|Conversions|ToColor(LinearColor) _returnvalue))
  (bind _currentattackerencounterlocationindex (Variables|InternalLogic|GetCurrentAttackerEncounterLocationIndex))
  (bind _returnvalue_4 (+ _currentattackerencounterlocationindex 1))
  (bind _encounterlocations (Variables|InternalLogic|GetEncounterLocations))
  (if (Utilities|Array|IsValidIndex _encounterlocations _returnvalue_4)
    (LocalVariables|SetDefenderSpawnIndexToLeashTo _returnvalue_4)
    (else
      (LocalVariables|SetDefenderSpawnIndexToLeashTo (Utilities|Array|LastIndex _encounterlocations)) ; For Attackers pushing to the final Defender spawn that was just defeated.
    ))
  (for _array_element (Variables|InternalLogic|GetSpawnedPawnsAttackers)
    (bind _result (ArrayMacros|SafeArrayGet (Variables|InternalLogic|GetEncounterLocations) _returnvalue_4)
      (:%20%20
        (bind _result_2 (ArrayMacros|SafeArrayGet (Variables|InternalLogic|GetEncounterOuterRadii) _returnvalue_4)
          (:%20%20
            (bind _result_3 (ArrayMacros|SafeArrayGet (Variables|InternalLogic|GetEncounterInnerRadii) _returnvalue_4)
              (:%20%20
                (Default|LeashSingleAIToSpecificLocation _array_element _result _result_3 _result_2 :!pinned true) ; Leash Attacker to Defender Spawn
                (if (Variables|EncounterSetup|GetDebugEncounter)
                  (Debug|DebugSphereSendtoAll (Transformation|GetActorLocation _array_element) 1e+02 5 _returnvalue_3 false 3.0 0 4.0)
                  (Debug|DebugLineSendtoAll (+ (.location (Math|Transform|ComposeTransforms (Math|Transform|MakeTransform _result) _returnvalue_2)) "0,0,200.000000") (Transformation|GetActorLocation _array_element) _returnvalue_3 false 3.0 0 4.0)))))))))
  (if (Variables|EncounterSetup|GetDebugEncounter)
    (Debug|DebugSphereSendtoAll (.location (Math|Transform|ComposeTransforms (Math|Transform|MakeTransform (stale _result)) _returnvalue_2)) 5e+02 8 _returnvalue_3 false 15.0 0 4.0)))

(fn LeashAllAttackersToMidfield
  (bind _attackerdebugcolor (Variables|EncounterSetup|GetAttackerDebugColor))
  (bind _returnvalue (Math|Color|Desaturate(LinearColor) _attackerdebugcolor 0.5))
  (bind _returnvalue_2 (Transformation|GetActorTransform))
  (bind _returnvalue_3 (Math|Conversions|ToColor(LinearColor) _returnvalue))
  (bind _currentattackerencounterlocationindex (Variables|InternalLogic|GetCurrentAttackerEncounterLocationIndex))
  (if (Utilities|Array|IsValidIndex (Variables|InternalLogic|GetEncounterMidfieldAttackerLocations) _currentattackerencounterlocationindex)
    (for _array_element (Variables|InternalLogic|GetSpawnedPawnsAttackers)
      (bind _result (ArrayMacros|SafeArrayGet (Variables|InternalLogic|GetEncounterMidfieldAttackerLocations) _currentattackerencounterlocationindex)
        (:%20%20
          (bind _result_2 (ArrayMacros|SafeArrayGet (Variables|InternalLogic|GetEncounterMidfieldAttackerOuterRadii) _currentattackerencounterlocationindex)
            (:%20%20
              (bind _result_3 (ArrayMacros|SafeArrayGet (Variables|InternalLogic|GetEncounterMidfieldAttackerInnerRadii) _currentattackerencounterlocationindex)
                (:%20%20
                  (Default|LeashSingleAIToSpecificLocation _array_element _result _result_3 _result_2 :!pinned true) ; Leash Attacker to Midfield
                  (if (Variables|EncounterSetup|GetDebugEncounter)
                    (Debug|DebugSphereSendtoAll (Transformation|GetActorLocation _array_element) 1e+02 5 _returnvalue_3 false 3.0 0 4.0)
                    (Debug|DebugLineSendtoAll (+ (.location (Math|Transform|ComposeTransforms (Math|Transform|MakeTransform _result) _returnvalue_2)) "0,0,200.000000") (Transformation|GetActorLocation _array_element) _returnvalue_3 false 3.0 0 4.0)))))))))
    (if (Variables|EncounterSetup|GetDebugEncounter)
      (Debug|DebugSphereSendtoAll (.location (Math|Transform|ComposeTransforms (Math|Transform|MakeTransform (stale _result)) _returnvalue_2)) 5e+02 8 _returnvalue_3 false 15.0 0 4.0))))

(fn LeashAllDefendersToMidfield
  (bind _returnvalue (CallFunction|DefendersinFallbackPhase))
  (bind _returnvalue_2 (Math|Integer|SelectInt 1 0 _returnvalue))
  (bind _defenderdebugcolor (Variables|EncounterSetup|GetDefenderDebugColor))
  (bind _returnvalue_3 (Math|Color|Desaturate(LinearColor) _defenderdebugcolor 0.5))
  (bind _returnvalue_4 (Math|Conversions|ToColor(LinearColor) _returnvalue_3))
  (bind _returnvalue_5 (Transformation|GetActorTransform))
  (bind _currentattackerencounterlocationindex (Variables|InternalLogic|GetCurrentAttackerEncounterLocationIndex))
  (bind _returnvalue_6 (+ _currentattackerencounterlocationindex _returnvalue_2))
  (if (Utilities|Array|IsValidIndex (Variables|InternalLogic|GetEncounterMidfieldDefenderLocations) _returnvalue_6)
    (for _array_element (Variables|InternalLogic|GetSpawnedPawnsDefenders)
      (bind _result (ArrayMacros|SafeArrayGet (Variables|InternalLogic|GetEncounterMidfieldDefenderLocations) _returnvalue_6)
        (:%20%20
          (bind _result_2 (ArrayMacros|SafeArrayGet (Variables|InternalLogic|GetEncounterMidfieldAttackerOuterRadii) _returnvalue_6)
            (:%20%20
              (bind _result_3 (ArrayMacros|SafeArrayGet (Variables|InternalLogic|GetEncounterMidfieldAttackerInnerRadii) _returnvalue_6)
                (:%20%20
                  (Default|LeashSingleAIToSpecificLocation _array_element _result _result_3 _result_2 :!pinned true) ; Leash Defender to Midfield
                  (if (Variables|EncounterSetup|GetDebugEncounter)
                    (Debug|DebugSphereSendtoAll (Transformation|GetActorLocation _array_element) 1e+02 5 _returnvalue_4 false 3.0 0 4.0)
                    (Debug|DebugLineSendtoAll (+ (.location (Math|Transform|ComposeTransforms (Math|Transform|MakeTransform _result) _returnvalue_5)) "0,0,200.000000") (Transformation|GetActorLocation _array_element) _returnvalue_4 false 3.0 0 4.0)))))))))
    (if (Variables|EncounterSetup|GetDebugEncounter)
      (Debug|DebugSphereSendtoAll (.location (Math|Transform|ComposeTransforms (Math|Transform|MakeTransform (stale _result)) _returnvalue_5)) 4.9e+02 8 _returnvalue_4 false 15.0 0 4.0))))

(fn LeashAllDefendersToSpawn
  (bind _defenderdebugcolor (Variables|EncounterSetup|GetDefenderDebugColor))
  (bind _returnvalue (Math|Color|Desaturate(LinearColor) _defenderdebugcolor 0.5))
  (bind _returnvalue_2 (Math|Conversions|ToColor(LinearColor) _returnvalue))
  (bind _returnvalue_3 (Transformation|GetActorTransform))
  (bind _currentattackerencounterlocationindex (Variables|InternalLogic|GetCurrentAttackerEncounterLocationIndex))
  (bind _returnvalue_4 (+ _currentattackerencounterlocationindex 1))
  (if (Utilities|Array|IsValidIndex (Variables|InternalLogic|GetEncounterLocations) _returnvalue_4)
    (for _array_element (Variables|InternalLogic|GetSpawnedPawnsDefenders)
      (bind _result (ArrayMacros|SafeArrayGet (Variables|InternalLogic|GetEncounterLocations) _returnvalue_4)
        (:%20%20
          (bind _result_2 (ArrayMacros|SafeArrayGet (Variables|InternalLogic|GetEncounterOuterRadii) _returnvalue_4)
            (:%20%20
              (bind _result_3 (ArrayMacros|SafeArrayGet (Variables|InternalLogic|GetEncounterInnerRadii) _returnvalue_4)
                (:%20%20
                  (Default|LeashSingleAIToSpecificLocation _array_element _result _result_3 _result_2 :!pinned true) ; Leash Defender to Spawn
                  (if (Variables|EncounterSetup|GetDebugEncounter)
                    (Debug|DebugSphereSendtoAll (Transformation|GetActorLocation _array_element) 1e+02 5 _returnvalue_2 false 3.0 0 4.0)
                    (Debug|DebugLineSendtoAll (+ (.location (Math|Transform|ComposeTransforms (Math|Transform|MakeTransform _result) _returnvalue_3)) "0,0,200.000000") (Transformation|GetActorLocation _array_element) _returnvalue_2 false 3.0 0 4.0)))))))))
    (if (Variables|EncounterSetup|GetDebugEncounter)
      (Debug|DebugSphereSendtoAll (.location (Math|Transform|ComposeTransforms (Math|Transform|MakeTransform (stale _result)) _returnvalue_3)) 4.9e+02 8 _returnvalue_2 false 15.0 0 4.0))))

(fn LeashSingleAIAfterSpawn
  (params (AI :type Pawn%20Object%20Reference) (Attacker :type Boolean))
  (bind _returnvalue (CallFunction|DefendersinFallbackPhase))
  (bind _returnvalue_2 (Math|Integer|SelectInt 1 0 _returnvalue))
  (bind _returnvalue_3 0.75)
  (bind _defenderdebugcolor (Variables|EncounterSetup|GetDefenderDebugColor))
  (bind _returnvalue_4 (Math|Color|Desaturate(LinearColor) _defenderdebugcolor _returnvalue_3))
  (bind _attackerdebugcolor (Variables|EncounterSetup|GetAttackerDebugColor))
  (bind _returnvalue_5 (Math|Color|Desaturate(LinearColor) _attackerdebugcolor _returnvalue_3))
  (bind _returnvalue_6 (Math|Conversions|ToColor(LinearColor) _returnvalue_4))
  (bind _aibeingleashed (LocalVariables|GetAIBeingLeashed))
  (bind _returnvalue_7 (Transformation|GetActorLocation _aibeingleashed))
  (bind _returnvalue_8 (Transformation|GetActorTransform))
  (bind _returnvalue_9 (Math|Conversions|ToColor(LinearColor) _returnvalue_5))
  (bind _leashoffset (LocalVariables|GetLeashOffset))
  (bind _returnvalue_10 (Math|Transform|MakeTransform _leashoffset))
  (bind _returnvalue_11 (Math|Transform|ComposeTransforms _returnvalue_10 _returnvalue_8))
  (bind _location (.location _returnvalue_11))
  (bind _currentattackerencounterlocationindex (Variables|InternalLogic|GetCurrentAttackerEncounterLocationIndex))
  (bind _returnvalue_12 (+ _currentattackerencounterlocationindex 1))
  (bind _currentattackerencounterlocationindex_2 (Variables|InternalLogic|GetCurrentAttackerEncounterLocationIndex))
  (bind _returnvalue_13 (+ _currentattackerencounterlocationindex_2 _returnvalue_2))
  (bind _isattacker (LocalVariables|GetIsAttacker))
  (bind _returnvalue_14 (select _isattacker _returnvalue_9 _returnvalue_6))
  (bind _currentattackerencounterlocationindex_3 (Variables|InternalLogic|GetCurrentAttackerEncounterLocationIndex))
  (bind _encountermidfielddefenderlocations (Variables|InternalLogic|GetEncounterMidfieldDefenderLocations))
  (LocalVariables|SetAIBeingLeashed AI)
  (LocalVariables|SetIsAttacker Attacker)
  (bind _asfort_athena_aibot_controller (Utilities|Casting|CastToFortAthenaAIBotController (Pawn|GetController AI)))
  (LocalVariables|SetAIController _asfort_athena_aibot_controller)
  (if (LocalVariables|GetIsAttacker)
    (if (or (Variables|InternalLogic|GetAttackerFinalCleanupPushIsActive) (Variables|InternalLogic|GetAttackersHaveClaimedFinalSpawnLocation))
      (if (Utilities|Array|IsValidIndex (Variables|InternalLogic|GetEncounterLocations) _returnvalue_12)
        (bind _result (ArrayMacros|SafeArrayGet (Variables|InternalLogic|GetEncounterLocations) _returnvalue_12)
          (:%20%20
            (LocalVariables|SetLeashOffset _result)
            (bind _result_2 (ArrayMacros|SafeArrayGet (Variables|InternalLogic|GetEncounterInnerRadii) _returnvalue_12)
              (:%20%20
                (LocalVariables|SetInnerLeash _result_2)
                (bind _result_3 (ArrayMacros|SafeArrayGet (Variables|InternalLogic|GetEncounterOuterRadii) _returnvalue_12)
                  (:%20%20
                    (LocalVariables|SetOuterLeash _result_3))))))))
      (else
        (bind _result_4 (ArrayMacros|SafeArrayGet (Variables|InternalLogic|GetEncounterMidfieldAttackerLocations) _currentattackerencounterlocationindex_3)
          (:%20%20
            (LocalVariables|SetLeashOffset _result_4)
            (bind _result_5 (ArrayMacros|SafeArrayGet (Variables|InternalLogic|GetEncounterMidfieldAttackerInnerRadii) _currentattackerencounterlocationindex_3)
              (:%20%20
                (LocalVariables|SetInnerLeash _result_5)
                (bind _result_6 (ArrayMacros|SafeArrayGet (Variables|InternalLogic|GetEncounterMidfieldAttackerOuterRadii) _currentattackerencounterlocationindex_3)
                  (:%20%20
                    (LocalVariables|SetOuterLeash _result_6)))))))))
    (elif (Utilities|Array|IsValidIndex _encountermidfielddefenderlocations _returnvalue_13)
      (bind _result_7 (ArrayMacros|SafeArrayGet _encountermidfielddefenderlocations _returnvalue_13)
        (:%20%20
          (LocalVariables|SetLeashOffset _result_7)
          (bind _result_8 (ArrayMacros|SafeArrayGet (Variables|InternalLogic|GetEncounterMidfieldAttackerInnerRadii) _returnvalue_13)
            (:%20%20
              (LocalVariables|SetInnerLeash _result_8)
              (bind _result_9 (ArrayMacros|SafeArrayGet (Variables|InternalLogic|GetEncounterMidfieldAttackerOuterRadii) _returnvalue_13)
                (:%20%20
                  (LocalVariables|SetOuterLeash _result_9)))))))))
  (Default|LeashSingleAIToSpecificLocation (LocalVariables|GetAIBeingLeashed) (LocalVariables|GetLeashOffset) (LocalVariables|GetInnerLeash) (LocalVariables|GetOuterLeash) :!pinned true) ; Leash spawned AI (Attacker or Defender) to where they should go.
  (if (Variables|EncounterSetup|GetDebugEncounter)
    (Debug|DebugSphereSendtoAll _location (Math|Float|SelectFloat 2.5e+02 2.4e+02 _isattacker) 8 _returnvalue_14 false 3.0 0 4.0)
    (Debug|DebugSphereSendtoAll _returnvalue_7 1e+02 5 _returnvalue_14 false 3.0 0 4.0)
    (Debug|DebugLineSendtoAll (+ _location "0,0,200.000000") _returnvalue_7 _returnvalue_14 false 3.0 0 4.0)))

(fn LeashSingleAIToSpecificLocation
  (params (AI :type Pawn%20Object%20Reference) (Location :type Vector) (InnerRadius :type Float%20%28double-precision%29) (OuterRadius :type Float%20%28double-precision%29))
  (bind _asfort_athena_aibot_controller (Utilities|Casting|CastToFortAthenaAIBotController (Pawn|GetController AI)))
  (Class|FortAthenaAIBotController|SetRadialLeashInnerRadius _asfort_athena_aibot_controller InnerRadius)
  (Class|FortAthenaAIBotController|SetRadialLeashOuterRadius _asfort_athena_aibot_controller OuterRadius)
  (Class|FortAthenaAIBotController|SetLeashReturnLocationMode _asfort_athena_aibot_controller "Random")
  (Class|FortAthenaAIBotController|SetLeashActortoFollow _asfort_athena_aibot_controller self 0.0 Location))

(fn MakeOpposingAIAwareOfEachOther
  (params (NewFortPawn :type Fort%20Pawn%20Object%20Reference))
  (AI|MakeBotGroupsTakeEachOtherasTargets (Utilities|Array|MakeArray NewFortPawn) (select (Utilities|Array|ContainsItem (Variables|InternalLogic|GetSpawnedPawnsAttackers) NewFortPawn) (Variables|InternalLogic|GetSpawnedPawnsDefenders) (Variables|InternalLogic|GetSpawnedPawnsAttackers)) "Threat_LKP" 1e+04 1e+06 :!pinned true) ; AI send knowledge of each other so they do not need to acquire each other as targets.\n\nJust spawn and start moving to one another to fight.
)

(fn OnRep_SpawnedPawnsAttackers
  (Class|LagerEncounterAudioController|UpdateAttackers (Variables|Default|GetEncounterAudioController) (Variables|InternalLogic|GetSpawnedPawnsAttackers)))

(fn OnRep_SpawnedPawnsDefenders
  (Class|LagerEncounterAudioController|UpdateDefenders (Variables|Default|GetEncounterAudioController) (Variables|InternalLogic|GetSpawnedPawnsDefenders)))

(fn SetAllAttackersChallenged
  (for _array_element (Variables|InternalLogic|GetSpawnedPawnsAttackers)
    (Default|SetSingleAIChallenged _array_element)))

(fn SetAllDefendersChallenged
  (for _array_element (Variables|InternalLogic|GetSpawnedPawnsDefenders)
    (Default|SetSingleAIChallenged _array_element)))

(fn SetEncounterActorsDormancy
  (params (NewDormancy :type ENetDormancy%20Enum :default "DORM_Never"))
  (Networking|SetNetDormancy NewDormancy)
  (for _array_element (Variables|InternalLogic|GetEncounterSpawnVisualBlueprints)
    (Utilities|IsValid _array_element
      (:Is%20Valid
        (Networking|SetNetDormancy _array_element NewDormancy)
        (bind _bga_portal_athena_encounter (Variables|/Lager/Lager/Encounters/BGA_Encounter_Visual_Spawn.BGA_Encounter_Visual_Spawn_C|GetBGA_Portal_Athena_Encounter _array_element)
          (:then
            (bind _childactor (Variables|/Script/Engine.ChildActorComponent|GetChildActor _bga_portal_athena_encounter)
              (:then
                (Networking|SetNetDormancy _childactor NewDormancy))
              (:else)))
          (:else))))))

(fn SetNewPointProviderLocationAndResetIfNecessary
  (params (NewLocation :type Vector) (PointProvider :type Fort%20Athena%20Living%20World%20Volume%20Object%20Reference))
  (bind _returnvalue (Transformation|GetActorLocation PointProvider))
  (bind _brushcomponent (Variables|/Script/Engine.Brush|GetBrushComponent PointProvider))
  (if (Variables|EncounterSetup|GetDebugEncounter)
    (Debug|DebugSphereSendtoAll _returnvalue 1.2e+02 8 (Utilities|Struct|MakeColor 0 255 0 255) false 3e+01 0 4.0) ; Debug where we started.
    (else
      _))
  (Transformation|SetMobility _brushcomponent "Movable") ; Make brush Movable so we can teleport it
  (Transformation|SetWorldLocation _brushcomponent NewLocation false true :!pinned true) ; Move Point Provider Brush to new Spawn Visual BP location
  (Transformation|SetMobility _brushcomponent "Stationary") ; Return brush to Static after teleporting
  (if (Variables|EncounterSetup|GetDebugEncounter)
    (Debug|DebugSphereSendtoAll _returnvalue 1.3e+02 8 (Utilities|Struct|MakeColor 0 0 255 255) false 3e+01 0 4.0) ; Debug where we went.
    (else
      _))
  (if (Class|FortAthenaLivingWorldVolume|IsPointProviderEnabled PointProvider)
    (Class|FortAthenaLivingWorldVolume|DisablePointProvider PointProvider)
    (Class|FortAthenaLivingWorldPointProviderInterface|EnablePointProvider_DEPRECATED PointProvider)))

(fn SetSingleAIChallenged
  (params (AI :type Fort%20Pawn%20Object%20Reference))
  (bind _tc_playerencounterengagementtags (Variables|EncounterSetup|GetTC_PlayerEncounterEngagementTags))
  (bind _returnvalue (GameplayTags|MakeLiteralGameplayTagContainer _tc_playerencounterengagementtags))
  (if (not (GameplayTags|HasAnyMatchingGameplayTags AI _returnvalue))
    (Ability|GameplayEffect|AddLooseGameplayTags AI _returnvalue)))

(fn SetupPointProviders
  (bind _attackerstartlocationindex (Variables|InternalLogic|GetAttackerStartLocationIndex))
  (bind _encounterspawnlocationdata (Variables|EncounterSetup|GetEncounterSpawnLocationData))
  (bind _output (Utilities|Array|Get(acopy) _encounterspawnlocationdata _attackerstartlocationindex))
  (bind (_encountertransform_8_fa43d3a14fa1b6c70d41f2b179384291 _encounterleashinnerradius_49_1991dbb840ba5a888e831984e7f9cff7 _encounterleashouterradius_51_c9b619f34f3a7ec7176fdca869ce044e _defendereliminationtargetsbyplaylist_59_d59f32d1495a9055bd6e01905b904b3e _defenderfallbackdirectlytonextmidfield_46_85b3c06e410046cd33d2b2aaaaa3c554 _advanceattackerencounterstage_55_f634112f4124e9c4cfad7485cc6a8f1e _advancedefenderencounterstage_52_f23cf75741e54e23495fc9af65a80450 _attackerpointprovidereqs_38_d775fb59468e6147c174118ee6768274 _attackerpointprovidertags_41_42b805564d2a69da3d87b3b5f030682a _defenderpointprovidereqs_37_0dae68b54ccc3e49c902b1a002f9223b _defenderpointprovidertags_43_634ec8e24a9d634b17bb6dbbbd4f647f) (Utilities|Struct|BreakStructEncounterLinearData _output))
  ;[1 1 1] Grab our Point Provider Volume Classes and Encounter Classes
  (if (Utilities|Array|IsValidIndex _encounterspawnlocationdata _attackerstartlocationindex)
    (Variables|InternalLogic|SetAttackerPointProviderEnvQuery _attackerpointprovidereqs_38_d775fb59468e6147c174118ee6768274)
    (Variables|InternalLogic|SetAttackerPointProviderTags _attackerpointprovidertags_41_42b805564d2a69da3d87b3b5f030682a)
    (Variables|InternalLogic|SetDefenderPointProviderEnvQuery _defenderpointprovidereqs_37_0dae68b54ccc3e49c902b1a002f9223b)
    (Variables|InternalLogic|SetDefenderPointProviderTags _defenderpointprovidertags_43_634ec8e24a9d634b17bb6dbbbd4f647f)
    (Utilities|FlowControl|Sequence
      (:then_0
        ;;[1 1 1] Spawn Living World Manager Point Providers and save-off references.
        (for (_ _array_index) (Variables|EncounterSetup|GetEncounterSpawnLocationData)
          (CallFunction|CreateSpawnPointProvidersatSpawnLocationIndex _array_index)))
      (:then_1
        (CallFunction|BindTrackSpawnVisualBlueprints :!pinned true) ; Watch our spawn pad BPs when their base gets destroyed.\nSpawn Pad will move downward and we will move our Point Providers along with them.
      )
      (:then_2
        (Variables|InternalLogic|SetHasCreatedPointProviders true)))))

(fn StartEncounter
  (bind _returnvalue (Game|GetGameState))
  (bind _returnvalue_2 (Actor|GetComponentbyClass _returnvalue "/Script/LagerRuntime.FortAthenaLivingWorldManager"))
  (bind _encountermidfieldvisualblueprints (Variables|InternalLogic|GetEncounterMidfieldVisualBlueprints))
  (bind _currentattackerencounterlocationindex (Variables|InternalLogic|GetCurrentAttackerEncounterLocationIndex))
  (if (not (Variables|InternalLogic|GetLinearEncounterIsActive))
    (Variables|InternalLogic|SetLinearEncounterIsActive true)
    (if (not (Variables|InternalLogic|GetHasCreatedPointProviders))
      (Default|SetupPointProviders))
    (Utilities|IsValid _returnvalue
      (:Is%20Valid
        (Utilities|IsValid _returnvalue_2
          (:Is%20Valid
            (Utilities|FlowControl|Sequence
              (:then_0
                (bind _outactors (Actor|GetAllActorsOfClass "/Lager/Lager/Encounters/BGA_Encounter_Visual_POI.BGA_Encounter_Visual_POI_C"))
                (for _array_element _outactors
                  (if (== (Variables|/Lager/Lager/Encounters/BGA_Encounter_Visual_POI.BGA_Encounter_Visual_POI_C|GetEncounterLocationID _array_element) (Variables|EncounterSetup|GetLinearEncounterLocationID))
                    (Components|Activation|Deactivate (Variables|/Lager/Lager/Encounters/BGA_Encounter_Visual_POI.BGA_Encounter_Visual_POI_C|GetAthenaSpecialActor_POI_MiniMap _array_element) :!pinned true) ; Hide POI Encounter MiniMap indicator
                  )))
              (:then_1
                (for _array_element_2 (Variables|InternalLogic|GetEncounterSpawnVisualBlueprints)
                  (Class|BGAEncounterVisualSpawn|FindBuildingActorBase _array_element_2 :!pinned true) ; Make sure all Spawn Visual BPs are all Based properly
                ))
              (:then_2
                (bind _returnvalue_3 (Class|FortAthenaLivingWorldManager|StartEncounter _returnvalue_2 (Variables|EncounterSetup|GetAttackerEncounterClass) (Variables|InternalLogic|GetAttackerEncounterPointProviders) self true :!pinned true)) ; Start Attacker Encounter
                (Variables|InternalLogic|SetActiveAttackerEncounter _returnvalue_3))
              (:then_3
                (bind _returnvalue_4 (Class|FortAthenaLivingWorldManager|StartEncounter _returnvalue_2 (Variables|EncounterSetup|GetDefenderEncounterClass) (Variables|InternalLogic|GetDefenderEncounterPointProviders) self true :!pinned true)) ; Start Defender Encounter
                (Variables|InternalLogic|SetActiveDefenderEncounter _returnvalue_4))
              (:then_4
                (CallFunction|BindAllEncounterEvents :!pinned true) ; Track Spawns/Deaths
                (CallFunction|ResetEncounterProgressforNewLocation :!pinned true) ; Reset Progress Variables 
                (CallFunction|EnableAttackerSpawn (Variables|EncounterSounds|GetAttackerSpawnWarmUpTime) true :!pinned true) ; Enable Attacker Spawn
                (CallFunction|EnableDefenderSpawn (Variables|EncounterSounds|GetDefenderSpawnWarmUpTime) true :!pinned true) ; Enable Defender Spawn
                (CallFunction|HidePOIMiniMapMarker :!pinned true) ; Hide the placeholder POI marker that exists before the encounter starts.\n\nThe spawn points should have their own markers from here on.
              )
              (:then_5
                (if (Variables|EncounterSetup|GetUseMidfieldUIVisualIndicators)
                  (if (Utilities|Array|IsValidIndex _encountermidfieldvisualblueprints _currentattackerencounterlocationindex)
                    (bind _result (ArrayMacros|SafeArrayGet _encountermidfieldvisualblueprints _currentattackerencounterlocationindex)
                      (:%20%20
                        (Utilities|IsValid _result
                          (:Is%20Valid
                            (Class|BGAEncounterVisualMidfield|MiniMapActivate _result :!pinned true) ; Show INITIAL Midfield MiniMap marker
                          )))))))
              (:then_6
                (for _array_element_3 (Variables|InternalLogic|GetEncounterSpawnVisualBlueprints)
                  (Rendering|SetActorHiddenInGame _array_element_3 false)
                  (Collision|SetCollisionEnabled (Variables|/Lager/Lager/Encounters/BGA_Encounter_Visual_Spawn.BGA_Encounter_Visual_Spawn_C|GetStaticMesh _array_element_3) "QueryAndPhysics" :!pinned true) ; Enable collision on spawner
                )
                (Rendering|SetActorHiddenInGame false))
              (:then_7
                (Variables|InternalLogic|SetDisabled false))
              (:then_8
                (CallFunction|ClientNotifyStateChangeEvent (Variables|InternalLogic|GetCurrentAttackerEncounterLocationIndex) (Utilities|Array|Length (Variables|EncounterSetup|GetEncounterSpawnLocationData)) (Variables|InternalLogic|GetCurrentDefenderEliminationCount) (Variables|InternalLogic|GetCurrentDefenderEliminationTarget)))
              (:then_9
                (bind _encountertriggersphere (Variables|Default|GetEncounterTriggerSphere)
                  (:then
                    (Variables|/Script/Engine.PrimitiveComponent|SetbGenerateOverlapEvents false _encountertriggersphere)
                    (Collision|SetCollisionEnabled _encountertriggersphere "NoCollision" :!pinned true) ; Shut off collision on trigger since we no longer need it.
                  )
                  (:else)))
              (:then_10
                (CallFunction|SetEncounterActorsDormancy "DORM_Awake"))
              (:then_11
                ;[1 1 1] Listen for players entering POI so they can get an update of the status of the encounter
                (bind _asfort_game_state (Utilities|Casting|CastToFortGameState (Game|GetGameState)))
                (Class|FortPoiManager|BindonPoiEntered (GameState|GetPoiManager _asfort_game_state) (Variables|EncounterSetup|GetPOITag) (EventDispatchers|CreateEvent :!event OnPlayerEnteredPoi)))
              (:then_12
                (Class|FortAthenaLinearEncounterAnalyticComponent|OnEncounterStart (Variables|Default|GetFortAthenaLinearEncounterAnalytic) (Variables|InternalLogic|GetActiveAttackerEncounter) (Variables|InternalLogic|GetActiveDefenderEncounter))))))))))

(fn TryBuiltInPlayerSpawnFX
  (params (SpawnedActor :type Object%20Reference))
  (bind _asplayer_pawn_athena (Utilities|Casting|CastToPlayerPawn_Athena SpawnedActor :!pinned true)) ; Will call the CreativeRespawnTeleportationGC variable tag as a GameplayCue along with SpawnImmunityTime (visuals)\n\nYou can change out the CreativeRespawnTeleportationGC and SpawnImmunityTime by making a child of the pawn.
  (Variables|/Game/Athena/PlayerPawn_Athena.PlayerPawn_Athena_C|SetPlayRespawnFXOnSpawn true _asplayer_pawn_athena))

(fn TrySpawnLocationBlueprintSpawnFX
  (params (Attacker :type Boolean) (Actor :type Actor%20Object%20Reference))
  (bind _returnvalue (CallFunction|DefendersinFallbackPhase))
  (bind _returnvalue_2 (Math|Integer|SelectInt 2 1 _returnvalue))
  (bind _currentattackerencounterlocationindex (Variables|InternalLogic|GetCurrentAttackerEncounterLocationIndex))
  (bind _returnvalue_3 (+ _currentattackerencounterlocationindex _returnvalue_2))
  (bind _returnvalue_4 (Math|Integer|SelectInt _currentattackerencounterlocationindex _returnvalue_3 Attacker))
  (bind _encounterspawnvisualblueprints (Variables|InternalLogic|GetEncounterSpawnVisualBlueprints))
  (Utilities|IsValid Actor
    (:Is%20Valid
      (if (Utilities|Array|IsValidIndex _encounterspawnvisualblueprints _returnvalue_4)
        (bind _result (ArrayMacros|SafeArrayGet _encounterspawnvisualblueprints _returnvalue_4)
          (:%20%20
            (bind _bga_portal_athena_encounter (Variables|/Lager/Lager/Encounters/BGA_Encounter_Visual_Spawn.BGA_Encounter_Visual_Spawn_C|GetBGA_Portal_Athena_Encounter _result)
              (:then
                (bind _childactor (Variables|/Script/Engine.ChildActorComponent|GetChildActor _bga_portal_athena_encounter)
                  (:then
                    (bind _asbga_portal_athena_encounter (Utilities|Casting|CastToBGA_Portal_Athena_Encounter _childactor))
                    (bind _asfort_player_pawn (Utilities|Casting|CastToFortPlayerPawn Actor))
                    (Class|BGAPortalAthena|PlayTeleportFX _asbga_portal_athena_encounter (Transformation|GetActorLocation _asfort_player_pawn) _asfort_player_pawn))
                  (:else)))
              (:else))))))))

(fn ConstructionScript
  (bind _currentencounterlocationindexcs (LocalVariables|GetCurrentEncounterLocationIndexCS))
  (bind _defendermidfieldlocations (Variables|EncounterSetup|GetDefenderMidfieldLocations))
  (bind _output (Utilities|Array|Get(acopy) _defendermidfieldlocations _currentencounterlocationindexcs))
  (bind (_leashlocation_2_cb573af946c84affa1334e8ebb85be3f _leashinnerradius_5_d92c837d4437de60c4b565b2a53a87c7 _leashouterradius_7_59d00b2542a352a7bcc40abebe64b15f) (Utilities|Struct|BreakStructEncounterLeashData _output))
  (bind _currentencounterlocationindexcs_2 (LocalVariables|GetCurrentEncounterLocationIndexCS))
  (bind _attackermidfieldlocations (Variables|EncounterSetup|GetAttackerMidfieldLocations))
  (bind _output_2 (Utilities|Array|Get(acopy) _attackermidfieldlocations _currentencounterlocationindexcs_2))
  (bind (_leashlocation_2_cb573af946c84affa1334e8ebb85be3f_2 _leashinnerradius_5_d92c837d4437de60c4b565b2a53a87c7_2 _leashouterradius_7_59d00b2542a352a7bcc40abebe64b15f_2) (Utilities|Struct|BreakStructEncounterLeashData _output_2))
  (bind _attackerstartlocationindex (Variables|InternalLogic|GetAttackerStartLocationIndex))
  (bind _returnvalue (+ _attackerstartlocationindex 1))
  (bind _currentencounterlocationindexcs_3 (LocalVariables|GetCurrentEncounterLocationIndexCS))
  (bind _returnvalue_2 (== _currentencounterlocationindexcs_3 _returnvalue))
  (bind _returnvalue_3 (Utilities|String|SelectString "Defender
Start" "" _returnvalue_2))
  (bind _lastindexcs (LocalVariables|GetLastIndexCS))
  (bind _currentmidfieldcenterlocationcs (LocalVariables|GetCurrentMidfieldCenterLocationCS))
  (bind _currentencounterlocationcs (LocalVariables|GetCurrentEncounterLocationCS))
  (bind _returnvalue_4 (- _currentencounterlocationcs _currentmidfieldcenterlocationcs))
  (bind _previousencounterlocationcs (LocalVariables|GetPreviousEncounterLocationCS))
  (bind _returnvalue_5 (- _previousencounterlocationcs _currentencounterlocationcs))
  (bind _returnvalue_6 (Math|Vector|SelectVector _returnvalue_5 _returnvalue_4 _lastindexcs))
  (bind _x (.x _returnvalue_6))
  (bind _y (.y _returnvalue_6))
  (bind _returnvalue_7 (Math|Vector|MakeVector _x _y))
  (bind _returnvalue_8 (Math|Conversions|RotationFromXVector _returnvalue_7))
  (bind _currentmidfielddefenderlocationcs (LocalVariables|GetCurrentMidfieldDefenderLocationCS))
  (bind _currentmidfieldcenterlocationcs_2 (LocalVariables|GetCurrentMidfieldCenterLocationCS))
  (bind _returnvalue_9 (- _currentmidfieldcenterlocationcs_2 _currentmidfielddefenderlocationcs))
  (bind _currentencounterlocationindexcs_4 (LocalVariables|GetCurrentEncounterLocationIndexCS))
  (bind _returnvalue_10 (+ _currentencounterlocationindexcs_4 1))
  (bind _currentencounterlocationindexcs_5 (LocalVariables|GetCurrentEncounterLocationIndexCS))
  (bind _returnvalue_11 (- _currentencounterlocationindexcs_5 1))
  (bind _currentmidfieldcenterlocationcs_3 (LocalVariables|GetCurrentMidfieldCenterLocationCS))
  (bind _nextmidfieldlocationcs (LocalVariables|GetNextMidfieldLocationCS))
  (bind _returnvalue_12 (- _nextmidfieldlocationcs _currentmidfieldcenterlocationcs_3))
  (bind _currentencounterlocationcs_2 (LocalVariables|GetCurrentEncounterLocationCS))
  (bind _returnvalue_13 (+ _currentencounterlocationcs_2 "0,0.000000,900.000000"))
  (bind _currentmidfieldattackerlocationcs (LocalVariables|GetCurrentMidfieldAttackerLocationCS))
  (bind _currentmidfieldcenterlocationcs_4 (LocalVariables|GetCurrentMidfieldCenterLocationCS))
  (bind _returnvalue_14 (- _currentmidfieldcenterlocationcs_4 _currentmidfieldattackerlocationcs))
  (bind _currentencounterlocationindexcs_6 (LocalVariables|GetCurrentEncounterLocationIndexCS))
  (bind _returnvalue_15 (+ _currentencounterlocationindexcs_6 1))
  (bind _debugoffsetfromspawnlocationactor (Variables|EncounterSetup|GetDebugOffsetFromSpawnLocationActor))
  (bind _currentencounterlocationcs_3 (LocalVariables|GetCurrentEncounterLocationCS))
  (bind _returnvalue_16 (+ _currentencounterlocationcs_3 _debugoffsetfromspawnlocationactor))
  (bind _currentmidfieldattackerlocationcs_2 (LocalVariables|GetCurrentMidfieldAttackerLocationCS))
  (bind _returnvalue_17 (- _currentmidfieldattackerlocationcs_2 _returnvalue_16))
  (bind _currentmidfielddefenderlocationcs_2 (LocalVariables|GetCurrentMidfieldDefenderLocationCS))
  (bind _debugoffsetfromspawnlocationactor_2 (Variables|EncounterSetup|GetDebugOffsetFromSpawnLocationActor))
  (bind _nextencounterlocationcs (LocalVariables|GetNextEncounterLocationCS))
  (bind _returnvalue_18 (+ _nextencounterlocationcs _debugoffsetfromspawnlocationactor_2))
  (bind _returnvalue_19 (- _returnvalue_18 _currentmidfielddefenderlocationcs_2))
  (bind _encounterinnerradii (Variables|InternalLogic|GetEncounterInnerRadii))
  (bind _encounterouterradii (Variables|InternalLogic|GetEncounterOuterRadii))
  (bind _currentencounterlocationindexcs_7 (LocalVariables|GetCurrentEncounterLocationIndexCS))
  (bind _encounterspawnlocationdata (Variables|EncounterSetup|GetEncounterSpawnLocationData))
  (bind _encounterspawnlocationdata_2 (Variables|EncounterSetup|GetEncounterSpawnLocationData))
  (bind _currentencounterarrowcs (LocalVariables|GetCurrentEncounterArrowCS))
  (bind _encountermidfieldcenterarrows (Variables|InternalLogic|GetEncounterMidfieldCenterArrows))
  (bind _currentmidfieldattackerlocationcs_3 (LocalVariables|GetCurrentMidfieldAttackerLocationCS))
  (bind _currentmidfielddefenderlocationcs_3 (LocalVariables|GetCurrentMidfieldDefenderLocationCS))
  (bind _encountermidfielddefenderlocations (Variables|InternalLogic|GetEncounterMidfieldDefenderLocations))
  (bind _encounterspawnlocationdata_3 (Variables|EncounterSetup|GetEncounterSpawnLocationData))
  (bind _currentencounterlocationindexcs_8 (LocalVariables|GetCurrentEncounterLocationIndexCS))
  (bind _encounterspawnlocationdata_4 (Variables|EncounterSetup|GetEncounterSpawnLocationData))
  (bind _currentencounterlocationindexcs_9 (LocalVariables|GetCurrentEncounterLocationIndexCS))
  (bind _currentencounterlocationindexcs_10 (LocalVariables|GetCurrentEncounterLocationIndexCS))
  (bind _currentencounterlocationindexcs_11 (LocalVariables|GetCurrentEncounterLocationIndexCS))
  ;[1 0 0] Gameplay Relevant
  (Transformation|AddLocalOffset (Variables|Default|GetEncounterTriggerSphere) (Variables|EncounterSetup|GetEncounterTriggerSphereLocation) false true :!pinned true) ; Placement of the Sphere Component Trigger \nthat activates the Encounter
  (Components|Sphere|SetSphereRadius (Variables|Default|GetEncounterTriggerSphere) (Utilities|ClampMinFloat (Variables|EncounterSetup|GetEncounterTriggerSphereRadius) 1e+02) false)
  (Utilities|Array|Clear (Variables|InternalLogic|GetEncounterMidfieldCenterLocations))
  (Utilities|Array|Clear (Variables|InternalLogic|GetEncounterMidfieldAttackerLocations))
  (Utilities|Array|Clear (Variables|InternalLogic|GetEncounterMidfieldDefenderLocations))
  (Utilities|Array|Clear (Variables|InternalLogic|GetEncounterMidfieldAttackerInnerRadii))
  (Utilities|Array|Clear (Variables|InternalLogic|GetEncounterMidfieldAttackerOuterRadii))
  (Utilities|Array|Clear (Variables|InternalLogic|GetEncounterLocations))
  (Utilities|Array|Clear (Variables|InternalLogic|GetEncounterInnerRadii))
  (Utilities|Array|Clear (Variables|InternalLogic|GetEncounterOuterRadii))
  (if (and (Variables|EncounterSetup|GetDebugEncounter) (CallFunction|EncounterIsEnabled))
    (Development|SetHiddeninGame (Variables|Default|GetEncounterTriggerSphere) (CallFunction|DebugElementShouldbeHidden) false :!pinned true) ; Show Trigger Sphere for Debug
    (else
      _))
  (for (_array_element _array_index) (Variables|EncounterSetup|GetEncounterSpawnLocationData)
    ;[1 1 1] Break Encounter Data for each array entry and figure out the
    ; data we need to  generate locations, spawn things, debug, etc.
    (bind _output_get (LocalVariables|SetFirstIndexCS (== (Variables|InternalLogic|GetAttackerStartLocationIndex) _array_index)))
    (if _output_get
      _
      (elif (Utilities|Array|IsValidIndex _encounterspawnlocationdata_3 (- _array_index 1))
        (LocalVariables|SetPreviousEncounterLocationCS (.location (Utilities|Struct|BreakStructEncounterLinearData (Utilities|Array|Get(acopy) _encounterspawnlocationdata_3 (- _array_index 1))))) ; Track our previous location.
      ))
    (LocalVariables|SetCurrentEncounterLocationIndexCS _array_index)
    (LocalVariables|SetLastIndexCS (== _array_index (Utilities|Array|LastIndex (Variables|EncounterSetup|GetEncounterSpawnLocationData))))
    (LocalVariables|SetDefenderFallbackToMidfieldCS (ScalableFloat|GetValueasBool (.DefenderFallbackDirectlyToNextMidfield_46_85B3C06E410046CD33D2B2AAAAA3C554 (Utilities|Struct|BreakStructEncounterLinearData _array_element))))
    ;[1 0 0] Gameplay Relevant
    (bind _output_get_2 (LocalVariables|SetCurrentEncounterLocationCS (.location (Utilities|Struct|BreakStructEncounterLinearData _array_element))))
    (Utilities|Array|Add (Variables|InternalLogic|GetEncounterLocations) _output_get_2)
    (LocalVariables|SetCurrentEncounterRotationCS (Math|Rotator|MakeRotator 0.0 0.0 (.yaw (.rotation (Utilities|Struct|BreakStructEncounterLinearData _array_element)))))
    (Utilities|Array|Add (Variables|InternalLogic|GetEncounterInnerRadii) (ScalableFloat|GetValueatLevel (.EncounterLeashInnerRadius_49_1991DBB840BA5A888E831984E7F9CFF7 (Utilities|Struct|BreakStructEncounterLinearData _array_element))))
    (Utilities|Array|Add (Variables|InternalLogic|GetEncounterOuterRadii) (ScalableFloat|GetValueatLevel (.EncounterLeashOuterRadius_51_C9B619F34F3A7EC7176FDCA869CE044E (Utilities|Struct|BreakStructEncounterLinearData _array_element))))
    (if (Utilities|Array|IsValidIndex _encounterspawnlocationdata_4 _currentencounterlocationindexcs_9)
      (bind _structout (Utilities|Struct|SetmembersinStructEncounterLinearData (Utilities|Array|Get(aref) _encounterspawnlocationdata_4 _currentencounterlocationindexcs_9) (Math|Transform|MakeTransform (.location (Utilities|Struct|BreakStructEncounterLinearData _array_element)) (Math|Rotator|MakeRotator 0.0 0.0 (.yaw (.rotation (Utilities|Struct|BreakStructEncounterLinearData _array_element)))) (.scale (Utilities|Struct|BreakStructEncounterLinearData _array_element)))))
      (Utilities|Array|SetArrayElem (Variables|EncounterSetup|GetEncounterSpawnLocationData) (LocalVariables|GetCurrentEncounterLocationIndexCS) _structout :!pinned true) ; Make sure we have no pitch or roll \non Encounter/Spawn. Only Yaw
      (else
        _))
    (Utilities|FlowControl|Sequence
      (:then_0
        ;[1 0 0] Gameplay Relevant
        (bind _returnvalue_20 (AddComponent|NODE_AddChildActorComponent-1 :RelativeTransform (Math|Transform|MakeTransform (LocalVariables|GetCurrentEncounterLocationCS))))
        (ChildActorComponent|SetChildActorClass _returnvalue_20 (Variables|InternalLogic|GetSpawnerVisualIndicatorClass) :!pinned true) ; Spawn Spawner (yo dawg) Blueprints
        (bind _asbga_encounter_visual_spawn (Utilities|Casting|CastToBGA_Encounter_Visual_Spawn (Variables|/Script/Engine.ChildActorComponent|GetChildActor _returnvalue_20)))
        (Utilities|Array|Add (Variables|InternalLogic|GetEncounterSpawnVisualBlueprints) _asbga_encounter_visual_spawn)
        (Utilities|FlowControl|Sequence
          (:then_0
            (if (Variables|EncounterSetup|GetAttackerOnlyShowsActiveUIVisualIndicatorsForFinalSpawnPoint)
              (Variables|/Lager/Lager/Encounters/BGA_Encounter_Visual_Spawn.BGA_Encounter_Visual_Spawn_C|SetAttackerShowUI false _asbga_encounter_visual_spawn) ; We DON'T want to show UI except the last one
              (else
                (Variables|/Lager/Lager/Encounters/BGA_Encounter_Visual_Spawn.BGA_Encounter_Visual_Spawn_C|SetAttackerShowUI true _asbga_encounter_visual_spawn) ; We want to ALWAYS show UI indicators.
              )))
          (:then_1
            (if (Variables|EncounterSetup|GetDefenderOnlyShowsDisabledUIVisualIndicatorsForFinalSpawnPoint)
              (if (== (Utilities|Array|LastIndex (Variables|EncounterSetup|GetEncounterSpawnLocationData)) (LocalVariables|GetCurrentEncounterLocationIndexCS))
                (Variables|/Lager/Lager/Encounters/BGA_Encounter_Visual_Spawn.BGA_Encounter_Visual_Spawn_C|SetDefenderShowUIForDisabledState true _asbga_encounter_visual_spawn) ; We ONLY show disabled UI on the LAST spawner
                (else
                  (Variables|/Lager/Lager/Encounters/BGA_Encounter_Visual_Spawn.BGA_Encounter_Visual_Spawn_C|SetDefenderShowUIForDisabledState false _asbga_encounter_visual_spawn) ; We DON'T want to show disabled UI except the last one
                ))
              (else
                (Variables|/Lager/Lager/Encounters/BGA_Encounter_Visual_Spawn.BGA_Encounter_Visual_Spawn_C|SetDefenderShowUIForDisabledState true _asbga_encounter_visual_spawn) ; We want to ALWAYS show disabled UI indicators.
              )))))
      (:then_1
        ;[1 1 1] Draw debug Arrows and Spheres for Encounter/Spawn locations
        (if (Editor|InEditororPIE)
          (bind _returnvalue_21 (AddComponent|NODE_AddArrowComponent-0 :RelativeTransform (Math|Transform|MakeTransform (+ (LocalVariables|GetCurrentEncounterLocationCS) (Variables|EncounterSetup|GetDebugOffsetFromSpawnLocationActor)) (Math|Rotator|MakeRotator 0.0 9e+01) "8.000000,8.000000,8.000000") :!pinned true)) ; Vertical Arrows for Encounter Locations
          (bind _output_get_3 (LocalVariables|SetCurrentEncounterArrowCS _returnvalue_21))
          (Utilities|Array|Add (Variables|InternalLogic|GetEncounterDebugArrows) _output_get_3)
          (Development|SetHiddeninGame _currentencounterarrowcs (CallFunction|DebugElementShouldbeHidden) false)
          (Components|Arrow|SetArrowColor(LinearColor) _currentencounterarrowcs (Math|Color|SelectColor (Variables|EncounterSetup|GetAttackerDebugColor) (Variables|EncounterSetup|GetDefenderDebugColor) (== (LocalVariables|GetCurrentEncounterLocationIndexCS) (Variables|InternalLogic|GetAttackerStartLocationIndex))))
          (if (Utilities|Array|IsValidIndex (Variables|EncounterSetup|GetEncounterSpawnLocationData) _returnvalue_11)
            (bind _result (ArrayMacros|SafeArrayGet (Variables|EncounterSetup|GetEncounterSpawnLocationData) _returnvalue_11)
              (:%20%20
                (if (not (ScalableFloat|GetValueasBool (.DefenderFallbackDirectlyToNextMidfield_46_85B3C06E410046CD33D2B2AAAAA3C554 (Utilities|Struct|BreakStructEncounterLinearData _result))))
                  (bind _returnvalue_22 (AddComponent|NODE_AddSphereComponent-2 :RelativeTransform (Math|Transform|MakeTransform (LocalVariables|GetCurrentEncounterLocationCS)) :!pinned true)) ; Draw Encounter/Spawn Leash Radii
                  (Transformation|AttachComponentToComponent _returnvalue_22 (LocalVariables|GetCurrentEncounterArrowCS) "None" "KeepWorld" "KeepWorld" "KeepWorld" false) ; Attach so we can Hide/Unhide without making a million arrays
                  (Collision|SetCollisionEnabled _returnvalue_22 "NoCollision")
                  (if (Utilities|Array|IsValidIndex _encounterinnerradii _currentencounterlocationindexcs_7)
                    (bind _result_2 (ArrayMacros|SafeArrayGet _encounterinnerradii _currentencounterlocationindexcs_7)
                      (:%20%20
                        (Components|Sphere|SetSphereRadius _returnvalue_22 _result_2 false :!pinned true) ; Inner Sphere Radius
                        (Development|SetHiddeninGame _returnvalue_22 (CallFunction|DebugElementShouldbeHidden) false)
                        (bind _returnvalue_23 (AddComponent|NODE_AddSphereComponent-4 :RelativeTransform (Math|Transform|MakeTransform (LocalVariables|GetCurrentEncounterLocationCS))))
                        (Transformation|AttachComponentToComponent _returnvalue_23 (LocalVariables|GetCurrentEncounterArrowCS) "None" "KeepWorld" "KeepWorld" "KeepWorld" false) ; Attach so we can Hide/Unhide without making a million arrays
                        (Collision|SetCollisionEnabled _returnvalue_23 "NoCollision")
                        (if (Utilities|Array|IsValidIndex _encounterouterradii _currentencounterlocationindexcs_10)
                          (bind _result_3 (ArrayMacros|SafeArrayGet _encounterouterradii _currentencounterlocationindexcs_10)
                            (:%20%20
                              (Components|Sphere|SetSphereRadius _returnvalue_23 _result_3 false :!pinned true) ; Outer Sphere Radius
                              (Development|SetHiddeninGame _returnvalue_23 (CallFunction|DebugElementShouldbeHidden) false)))))))))))))
      (:then_2
        (if (Utilities|Array|IsValidIndex _encounterspawnlocationdata _returnvalue_10)
          (LocalVariables|SetNextEncounterLocationCS (.location (Utilities|Struct|BreakStructEncounterLinearData (Utilities|Array|Get(acopy) _encounterspawnlocationdata _returnvalue_10))))
          (Utilities|FlowControl|Sequence
            (:then_0
              ;[1 1 1] Grab or create Attacker Midfield data
              (if (Utilities|Array|IsValidIndex (Variables|EncounterSetup|GetAttackerMidfieldLocations) (LocalVariables|GetCurrentEncounterLocationIndexCS))
                (LocalVariables|SetCurrentMidfieldAttackerLocationCS _leashlocation_2_cb573af946c84affa1334e8ebb85be3f_2) ; We have a hand-placed Midfield location to go along with our current Encounter Index
                (LocalVariables|SetCurrentMidfieldAttackerInnerRadiusCS (ScalableFloat|GetValueatLevel _leashinnerradius_5_d92c837d4437de60c4b565b2a53a87c7_2))
                (LocalVariables|SetCurrentMidfieldAttackerOuterRadiusCS (ScalableFloat|GetValueatLevel _leashouterradius_7_59d00b2542a352a7bcc40abebe64b15f_2))
                (else
                  (bind (_vectorpoints _successfullyfoundpoints) (Class|MissionBlueprintFunctionLibrary|PointLocationsBetweenTwoVectors (LocalVariables|GetNextEncounterLocationCS) (LocalVariables|GetCurrentEncounterLocationCS))) ; Auto-generate Midfield location if we do not have a hand-placed one.\nWill be at exact midpoint between Attacker and Defender spawns.
                  (if _successfullyfoundpoints
                    (LocalVariables|SetCurrentMidfieldAttackerLocationCS (Utilities|Array|Get(acopy) _vectorpoints))
                    (LocalVariables|SetCurrentMidfieldAttackerInnerRadiusCS 1e+03)
                    (LocalVariables|SetCurrentMidfieldAttackerOuterRadiusCS 1.2e+03)))))
            (:then_1
              ;[1 1 1] Grab or create Defender Midfield data
              (if (Utilities|Array|IsValidIndex (Variables|EncounterSetup|GetDefenderMidfieldLocations) (LocalVariables|GetCurrentEncounterLocationIndexCS))
                (LocalVariables|SetCurrentMidfieldDefenderLocationCS _leashlocation_2_cb573af946c84affa1334e8ebb85be3f) ; We have a hand-placed Midfield location to go along with our current Encounter Index
                (LocalVariables|SetCurrentMidfieldDefenderInnerRadiusCS (ScalableFloat|GetValueatLevel _leashinnerradius_5_d92c837d4437de60c4b565b2a53a87c7))
                (LocalVariables|SetCurrentMidfieldDefenderOuterRadiusCS (ScalableFloat|GetValueatLevel _leashouterradius_7_59d00b2542a352a7bcc40abebe64b15f))
                (else
                  (bind (_vectorpoints_2 _successfullyfoundpoints_2) (Class|MissionBlueprintFunctionLibrary|PointLocationsBetweenTwoVectors (LocalVariables|GetNextEncounterLocationCS) (LocalVariables|GetCurrentEncounterLocationCS))) ; Auto-generate Midfield location if we do not have a hand-placed one.\nWill be at exact midpoint between Attacker and Defender spawns.
                  (if _successfullyfoundpoints_2
                    (LocalVariables|SetCurrentMidfieldDefenderLocationCS (Utilities|Array|Get(acopy) _vectorpoints_2))
                    (LocalVariables|SetCurrentMidfieldDefenderInnerRadiusCS 1e+03)
                    (LocalVariables|SetCurrentMidfieldDefenderOuterRadiusCS 1.2e+03)))))
            (:then_2
              ;[1 1 1] Create Midfield Center data
              (bind (_vectorpoints_3 _successfullyfoundpoints_3) (Class|MissionBlueprintFunctionLibrary|PointLocationsBetweenTwoVectors (LocalVariables|GetCurrentMidfieldDefenderLocationCS) (LocalVariables|GetCurrentMidfieldAttackerLocationCS))) ; Auto-generate Midfield Center. \nWill be at exact midpoint between Attacker and Defender Midfields.
              (if _successfullyfoundpoints_3
                (LocalVariables|SetCurrentMidfieldCenterLocationCS (Utilities|Array|Get(acopy) _vectorpoints_3))))
            (:then_3
              ;[1 0 0] Gameplay Relevant
              (Utilities|Array|Add (Variables|InternalLogic|GetEncounterMidfieldCenterLocations) (LocalVariables|GetCurrentMidfieldCenterLocationCS))
              (Utilities|Array|Add (Variables|InternalLogic|GetEncounterMidfieldAttackerLocations) (LocalVariables|GetCurrentMidfieldAttackerLocationCS))
              (Utilities|Array|Add (Variables|InternalLogic|GetEncounterMidfieldDefenderLocations) (LocalVariables|GetCurrentMidfieldDefenderLocationCS))
              (Utilities|Array|Add (Variables|InternalLogic|GetEncounterMidfieldAttackerInnerRadii) (LocalVariables|GetCurrentMidfieldAttackerInnerRadiusCS))
              (Utilities|Array|Add (Variables|InternalLogic|GetEncounterMidfieldAttackerOuterRadii) (LocalVariables|GetCurrentMidfieldAttackerOuterRadiusCS))
              (Utilities|Array|Add (Variables|InternalLogic|GetEncounterMidfieldDefenderInnerRadii) (LocalVariables|GetCurrentMidfieldDefenderInnerRadiusCS))
              (Utilities|Array|Add (Variables|InternalLogic|GetEncounterMidfieldDefenderOuterRadii) (LocalVariables|GetCurrentMidfieldDefenderOuterRadiusCS))
              (LocalVariables|SetCurrentAverageMidfieldDirectionCS (Math|Vector|Normalize (Math|Vector|MakeVector (.x (- (LocalVariables|GetCurrentMidfieldDefenderLocationCS) (LocalVariables|GetCurrentMidfieldAttackerLocationCS))) (.y (- (LocalVariables|GetCurrentMidfieldDefenderLocationCS) (LocalVariables|GetCurrentMidfieldAttackerLocationCS)))))) ; Average direction from Attacker spawn through \nMidfield and towards Defender spawn.
              (if (Variables|EncounterSetup|GetUseMidfieldUIVisualIndicators)
                (bind _returnvalue_24 (AddComponent|NODE_AddChildActorComponent-2 :RelativeTransform (Math|Transform|MakeTransform (LocalVariables|GetCurrentMidfieldCenterLocationCS))))
                (ChildActorComponent|SetChildActorClass _returnvalue_24 (Variables|InternalLogic|GetMidfieldVisualIndicatorClass) :!pinned true) ; Spawn Midfield MiniMap blueprints
                (bind _asbga_encounter_visual_midfield (Utilities|Casting|CastToBGA_Encounter_Visual_Midfield (Variables|/Script/Engine.ChildActorComponent|GetChildActor _returnvalue_24)))
                (Utilities|Array|Add (Variables|InternalLogic|GetEncounterMidfieldVisualBlueprints) _asbga_encounter_visual_midfield)
                (else
                  _))
              (if (Editor|InEditororPIE)
                (for (_array_element_2 _array_index_2) (Utilities|Array|MakeArray (LocalVariables|GetCurrentMidfieldCenterLocationCS) (LocalVariables|GetCurrentMidfieldAttackerLocationCS) (LocalVariables|GetCurrentMidfieldDefenderLocationCS))
                  (bind _returnvalue_25 (AddComponent|NODE_AddArrowComponent-1 :RelativeTransform (Math|Transform|MakeTransform (+ _array_element_2 (Math|Vector|MakeVector 0.0 0.0 5e+02)) (Math|Rotator|MakeRotator 0.0 -9e+01) "6.000000,6.000000,6.000000") :!pinned true)) ; Vertical Arrows for Midfield Centers.
                  (if (== _array_index_2 0)
                    (bind _output_get_4 (LocalVariables|SetCurrentMidfieldCenterArrowCS _returnvalue_25))
                    (Utilities|Array|Add (Variables|InternalLogic|GetEncounterMidfieldCenterArrows) _output_get_4)
                    (else
                      (bind _currentmidfieldcenterarrowcs (LocalVariables|GetCurrentMidfieldCenterArrowCS)
                        (:then
                          (Transformation|AttachComponentToComponent _returnvalue_25 _currentmidfieldcenterarrowcs "None" "KeepWorld" "KeepWorld" "KeepWorld" false) ; Attach so we can Hide/Unhide without making a million arrays
                        )
                        (:else
                          _))))
                  (Development|SetHiddeninGame _returnvalue_25 (CallFunction|DebugElementShouldbeHidden) false)
                  (Components|Arrow|SetArrowColor(LinearColor) _returnvalue_25 (Variables|EncounterSetup|GetMidfieldDebugColor))
                  (if (not (== _array_index_2 0))
                    (bind _returnvalue_26 (AddComponent|NODE_AddSphereComponent-1 :RelativeTransform (Math|Transform|MakeTransform (Math|Vector|SelectVector _currentmidfieldattackerlocationcs_3 _currentmidfielddefenderlocationcs_3 (== _array_index_2 1))) :!pinned true)) ; Draw Midfield Leash Radii
                    (Transformation|AttachComponentToComponent _returnvalue_26 (LocalVariables|GetCurrentMidfieldCenterArrowCS) "None" "KeepWorld" "KeepWorld" "KeepWorld" false) ; Attach so we can Hide/Unhide without making a million arrays
                    (Collision|SetCollisionEnabled _returnvalue_26 "NoCollision")
                    (Components|Sphere|SetSphereRadius _returnvalue_26 (Math|Float|SelectFloat (LocalVariables|GetCurrentMidfieldAttackerInnerRadiusCS) (LocalVariables|GetCurrentMidfieldDefenderInnerRadiusCS) (== _array_index_2 1)) false)
                    (Development|SetHiddeninGame _returnvalue_26 (CallFunction|DebugElementShouldbeHidden) false)
                    (bind _returnvalue_27 (AddComponent|NODE_AddSphereComponent-3 :RelativeTransform (Math|Transform|MakeTransform (Math|Vector|SelectVector _currentmidfieldattackerlocationcs_3 _currentmidfielddefenderlocationcs_3 (== _array_index_2 1)))))
                    (Transformation|AttachComponentToComponent _returnvalue_27 (LocalVariables|GetCurrentMidfieldCenterArrowCS) "None" "KeepWorld" "KeepWorld" "KeepWorld" false) ; Attach so we can Hide/Unhide without making a million arrays
                    (Collision|SetCollisionEnabled _returnvalue_27 "NoCollision")
                    (Components|Sphere|SetSphereRadius _returnvalue_27 (Math|Float|SelectFloat (LocalVariables|GetCurrentMidfieldAttackerOuterRadiusCS) (LocalVariables|GetCurrentMidfieldDefenderOuterRadiusCS) (== _array_index_2 1)) false)
                    (Development|SetHiddeninGame _returnvalue_27 (CallFunction|DebugElementShouldbeHidden) false)))))
            (:then_4
              (if (Editor|InEditororPIE)
                (bind _returnvalue_28 (AddComponent|NODE_AddArrowComponent-2 :RelativeTransform (Math|Transform|MakeTransform _returnvalue_16 (Math|Conversions|RotationFromXVector _returnvalue_17) "8.000000,8.000000,8.000000") :!pinned true)) ; Team Colored Arrows between Encounters and Midfield locations to show flow\nCurrent Encounter to Attacker Midfield
                (Utilities|Array|Add (Variables|InternalLogic|GetAttackerEncounterToMidfieldArrows) _returnvalue_28)
                (Development|SetHiddeninGame _returnvalue_28 (CallFunction|DebugElementShouldbeHidden) false)
                (Components|Arrow|SetArrowColor(LinearColor) _returnvalue_28 (Math|Color|SelectColor (Variables|EncounterSetup|GetAttackerDebugColor) (Variables|EncounterSetup|GetDefenderDebugColor) (== (LocalVariables|GetCurrentEncounterLocationIndexCS) (Variables|InternalLogic|GetAttackerStartLocationIndex))))
                (Variables|/Script/Engine.ArrowComponent|SetArrowLength (/ (Math|Vector|VectorLength _returnvalue_17) 8.0) _returnvalue_28)
                (bind _returnvalue_29 (AddComponent|NODE_AddArrowComponent-3 :RelativeTransform (Math|Transform|MakeTransform _currentmidfielddefenderlocationcs_2 (Math|Conversions|RotationFromXVector _returnvalue_19) "8.000000,8.000000,8.000000") :!pinned true)) ; Defender Midfield to Next Encounter
                (Utilities|Array|Add (Variables|InternalLogic|GetMidfieldToDefenseEncounterArrows) _returnvalue_29)
                (Development|SetHiddeninGame _returnvalue_29 (CallFunction|DebugElementShouldbeHidden) false)
                (Components|Arrow|SetArrowColor(LinearColor) _returnvalue_29 (Variables|EncounterSetup|GetDefenderDebugColor))
                (Variables|/Script/Engine.ArrowComponent|SetArrowLength (/ (Math|Vector|VectorLength _returnvalue_19) 8.0) _returnvalue_29)
                (bind _returnvalue_30 (AddComponent|NODE_AddArrowComponent-5 :RelativeTransform (Math|Transform|MakeTransform _currentmidfieldattackerlocationcs (Math|Conversions|RotationFromXVector _returnvalue_14) "8.000000,8.000000,8.000000") :!pinned true)) ; Attacker Midfield to Center
                (Utilities|Array|Add (Variables|InternalLogic|GetAttackerEncounterToMidfieldArrows) _returnvalue_30)
                (Development|SetHiddeninGame _returnvalue_30 (CallFunction|DebugElementShouldbeHidden) false)
                (Components|Arrow|SetArrowColor(LinearColor) _returnvalue_30 (Variables|EncounterSetup|GetAttackerDebugColor))
                (Variables|/Script/Engine.ArrowComponent|SetArrowLength (/ (Math|Vector|VectorLength _returnvalue_14) 8.0) _returnvalue_30)
                (bind _returnvalue_31 (AddComponent|NODE_AddArrowComponent-6 :RelativeTransform (Math|Transform|MakeTransform _currentmidfielddefenderlocationcs (Math|Conversions|RotationFromXVector _returnvalue_9) "8.000000,8.000000,8.000000") :!pinned true)) ; Center to Defender Midfield
                (Utilities|Array|Add (Variables|InternalLogic|GetMidfieldToDefenseEncounterArrows) _returnvalue_31)
                (Development|SetHiddeninGame _returnvalue_31 (CallFunction|DebugElementShouldbeHidden) false)
                (Components|Arrow|SetArrowColor(LinearColor) _returnvalue_31 (Variables|EncounterSetup|GetDefenderDebugColor))
                (Variables|/Script/Engine.ArrowComponent|SetArrowLength (/ (Math|Vector|VectorLength _returnvalue_9) 8.0) _returnvalue_31))))
          (elif (Utilities|Array|IsValidIndex _encounterspawnlocationdata_2 _currentencounterlocationindexcs_8)
            (bind _result_4 (ArrayMacros|SafeArrayGet _encounterspawnlocationdata_2 _currentencounterlocationindexcs_8)
              (:%20%20
                ;[1 0 0] Gameplay Relevant
                (Utilities|Array|Add (Variables|InternalLogic|GetEncounterMidfieldCenterLocations) (LocalVariables|GetCurrentEncounterLocationCS) :!pinned true) ; Make our current (last) encounter location our last Leash Location for the Attackers.\nWhen they win, they will advance to the Defender spawns and hang out.
                (Utilities|Array|Add (Variables|InternalLogic|GetEncounterMidfieldAttackerInnerRadii) (ScalableFloat|GetValueatLevel (.EncounterLeashInnerRadius_49_1991DBB840BA5A888E831984E7F9CFF7 (Utilities|Struct|BreakStructEncounterLinearData _result_4))))
                (Utilities|Array|Add (Variables|InternalLogic|GetEncounterMidfieldAttackerOuterRadii) (ScalableFloat|GetValueatLevel (.EncounterLeashOuterRadius_51_C9B619F34F3A7EC7176FDCA869CE044E (Utilities|Struct|BreakStructEncounterLinearData _result_4)))))))))
      (:then_3
        (if (Editor|InEditororPIE)
          (bind _returnvalue_32 (AddComponent|NODE_AddTextRenderComponent-0 :RelativeTransform (Math|Transform|MakeTransform (+ (LocalVariables|GetCurrentEncounterLocationCS) "0,0,600.000000") (Math|Conversions|RotationFromXVector (Math|Vector|MakeVector (.x (Math|Vector|SelectVector (- (LocalVariables|GetPreviousEncounterLocationCS) (LocalVariables|GetCurrentEncounterLocationCS)) (- (LocalVariables|GetCurrentEncounterLocationCS) (LocalVariables|GetCurrentMidfieldCenterLocationCS)) (LocalVariables|GetLastIndexCS))) (.y (Math|Vector|SelectVector (- (LocalVariables|GetPreviousEncounterLocationCS) (LocalVariables|GetCurrentEncounterLocationCS)) (- (LocalVariables|GetCurrentEncounterLocationCS) (LocalVariables|GetCurrentMidfieldCenterLocationCS)) (LocalVariables|GetLastIndexCS))))) "6.000000,6.000000,6.000000") :!pinned true)) ; Debug Text above Encounter Location
          (Rendering|Components|TextRender|SetText _returnvalue_32 (Utilities|Text|ToText(String) (Utilities|String|BuildString(Integer) (Utilities|String|BuildString(Integer) "" "Location " (+ (LocalVariables|GetCurrentEncounterLocationIndexCS) 1)) " [" (LocalVariables|GetCurrentEncounterLocationIndexCS) "]")))
          (Development|SetHiddeninGame _returnvalue_32 (CallFunction|DebugElementShouldbeHidden) false)
          (Transformation|AttachComponentToComponent _returnvalue_32 (LocalVariables|GetCurrentEncounterArrowCS) "None" "KeepWorld" "KeepWorld" "KeepWorld" false) ; Attach so we can Hide/Unhide without making a million arrays
        ))
      (:then_4
        (if (Editor|InEditororPIE)
          (if (Math|Boolean|ORBoolean (== (LocalVariables|GetCurrentEncounterLocationIndexCS) (+ (Variables|InternalLogic|GetAttackerStartLocationIndex) 1)) (LocalVariables|GetFirstIndexCS) (LocalVariables|GetLastIndexCS))
            (bind _returnvalue_33 (AddComponent|NODE_AddTextRenderComponent-3 :RelativeTransform (Math|Transform|MakeTransform _returnvalue_13 _returnvalue_8 "10.000000,10.000000,10.000000") :!pinned true)) ; Team-Specific Debug Text
            (Rendering|Components|TextRender|SetText _returnvalue_33 (Utilities|Text|ToText(String) (Utilities|String|SelectString "Attacker
Destination" (Utilities|String|SelectString "Attacker
Start" _returnvalue_3 (LocalVariables|GetFirstIndexCS)) (LocalVariables|GetLastIndexCS))))
            (Development|SetHiddeninGame _returnvalue_33 (CallFunction|DebugElementShouldbeHidden) false)
            (Rendering|Components|TextRender|SetTextRenderColor _returnvalue_33 (select (Math|Boolean|ANDBoolean _returnvalue_2 (not (LocalVariables|GetLastIndexCS)) false) (Math|Conversions|ToColor(LinearColor) (Variables|EncounterSetup|GetDefenderDebugColor)) (Math|Conversions|ToColor(LinearColor) (Variables|EncounterSetup|GetAttackerDebugColor))))
            (Transformation|AttachComponentToComponent _returnvalue_33 (LocalVariables|GetCurrentEncounterArrowCS) "None" "KeepWorld" "KeepWorld" "KeepWorld" false) ; Attach so we can Hide/Unhide without making a million arrays
            (if (== (LocalVariables|GetLastIndexCS) (== (LocalVariables|GetCurrentEncounterLocationIndexCS) (+ (Variables|InternalLogic|GetAttackerStartLocationIndex) 1)))
              (bind _returnvalue_34 (AddComponent|NODE_AddTextRenderComponent-4 :RelativeTransform (Math|Transform|MakeTransform (+ (LocalVariables|GetCurrentEncounterLocationCS) "0,0,1100.000000") _returnvalue_8 "10.000000,10.000000,10.000000"))) ; If Defender Start and Attacker Destination are the same.
              (Rendering|Components|TextRender|SetText _returnvalue_34 (Utilities|Text|ToText(String) _returnvalue_3))
              (Development|SetHiddeninGame _returnvalue_34 (CallFunction|DebugElementShouldbeHidden) false)
              (Rendering|Components|TextRender|SetTextRenderColor _returnvalue_34 (Math|Conversions|ToColor(LinearColor) (Variables|EncounterSetup|GetDefenderDebugColor)))
              (Transformation|AttachComponentToComponent _returnvalue_34 (LocalVariables|GetCurrentEncounterArrowCS) "None" "KeepWorld" "KeepWorld" "KeepWorld" false) ; Attach so we can Hide/Unhide without making a million arrays
            )
            (else
              (bind _returnvalue_35 (AddComponent|NODE_AddTextRenderComponent-5 :RelativeTransform (Math|Transform|MakeTransform _returnvalue_13 _returnvalue_8 "10.000000,10.000000,10.000000") :!pinned true)) ; Generic Text Over Spawns/Encounters
              (Rendering|Components|TextRender|SetText _returnvalue_35 (Utilities|Text|ToText(String) "Defender
Location"))
              (Development|SetHiddeninGame _returnvalue_35 (CallFunction|DebugElementShouldbeHidden) false)
              (Rendering|Components|TextRender|SetTextRenderColor _returnvalue_35 (Math|Conversions|ToColor(LinearColor) (Variables|EncounterSetup|GetDefenderDebugColor)))
              (Transformation|AttachComponentToComponent _returnvalue_35 (LocalVariables|GetCurrentEncounterArrowCS) "None" "KeepWorld" "KeepWorld" "KeepWorld" false) ; Attach so we can Hide/Unhide without making a million arrays
            ))))
      (:then_5
        (if (Editor|InEditororPIE)
          (if (not (LocalVariables|GetLastIndexCS))
            (bind _returnvalue_36 (AddComponent|NODE_AddTextRenderComponent-1 :RelativeTransform (Math|Transform|MakeTransform (+ (LocalVariables|GetCurrentMidfieldCenterLocationCS) "0,0,450.000000") (Math|Conversions|RotationFromXVector (Math|Vector|MakeVector (.x (Math|Vector|SelectVector (- (LocalVariables|GetCurrentEncounterLocationCS) (LocalVariables|GetCurrentMidfieldCenterLocationCS)) (* (LocalVariables|GetCurrentAverageMidfieldDirectionCS) -1.0) (LocalVariables|GetLastIndexCS))) (.y (Math|Vector|SelectVector (- (LocalVariables|GetCurrentEncounterLocationCS) (LocalVariables|GetCurrentMidfieldCenterLocationCS)) (* (LocalVariables|GetCurrentAverageMidfieldDirectionCS) -1.0) (LocalVariables|GetLastIndexCS))))) "6.000000,6.000000,6.000000") :!pinned true)) ; Debug Text above Midfield Location
            (Rendering|Components|TextRender|SetText _returnvalue_36 (Utilities|Text|ToText(String) (Utilities|String|BuildString(Integer) (Utilities|String|BuildString(Integer) "" "Midfield
" (+ (LocalVariables|GetCurrentEncounterLocationIndexCS) 1) " >>> VS <<< ") "" (+ (LocalVariables|GetCurrentEncounterLocationIndexCS) 2))))
            (Development|SetHiddeninGame _returnvalue_36 (CallFunction|DebugElementShouldbeHidden) false)
            (Transformation|AttachComponentToComponent _returnvalue_36 (LocalVariables|GetCurrentMidfieldCenterArrowCS) "None" "KeepWorld" "KeepWorld" "KeepWorld" false) ; Attach so we can Hide/Unhide without making a million arrays
          )))))
  ;[1 1 1] Last pass, we draw debug info for when Defenders fallback directly from Midfield-to-Midfield, skipping falling back through their Spawn.
  (if (Editor|InEditororPIE)
    (for (_array_element_3 _array_index_3) (Variables|EncounterSetup|GetEncounterSpawnLocationData)
      (LocalVariables|SetCurrentEncounterLocationIndexCS _array_index_3)
      (LocalVariables|SetDefenderFallbackToMidfieldCS (ScalableFloat|GetValueasBool (.DefenderFallbackDirectlyToNextMidfield_46_85B3C06E410046CD33D2B2AAAAA3C554 (Utilities|Struct|BreakStructEncounterLinearData _array_element_3))))
      (if (LocalVariables|GetDefenderFallbackToMidfieldCS)
        (if (Utilities|Array|IsValidIndex _encountermidfielddefenderlocations _currentencounterlocationindexcs_6)
          (bind _result_5 (ArrayMacros|SafeArrayGet _encountermidfielddefenderlocations _currentencounterlocationindexcs_6)
            (:%20%20
              (LocalVariables|SetCurrentMidfieldCenterLocationCS _result_5)
              (if (Utilities|Array|IsValidIndex _encountermidfielddefenderlocations _returnvalue_15)
                (bind _result_6 (ArrayMacros|SafeArrayGet _encountermidfielddefenderlocations _returnvalue_15)
                  (:%20%20
                    (LocalVariables|SetNextMidfieldLocationCS _result_6)
                    (bind _returnvalue_37 (AddComponent|NODE_AddArrowComponent-4 :RelativeTransform (Math|Transform|MakeTransform _currentmidfieldcenterlocationcs_3 (Math|Conversions|RotationFromXVector _returnvalue_12) "8.000000,8.000000,8.000000") :!pinned true)) ; Debug arrow for direct Defender fallback
                    (Development|SetHiddeninGame _returnvalue_37 (CallFunction|DebugElementShouldbeHidden) false)
                    (Components|Arrow|SetArrowColor(LinearColor) _returnvalue_37 (Variables|EncounterSetup|GetMidfieldDebugColor))
                    (Variables|/Script/Engine.ArrowComponent|SetArrowLength (/ (Math|Vector|VectorLength _returnvalue_12) 8.0) _returnvalue_37)
                    (if (Utilities|Array|IsValidIndex _encountermidfieldcenterarrows _currentencounterlocationindexcs_11)
                      (bind _result_7 (ArrayMacros|SafeArrayGet _encountermidfieldcenterarrows _currentencounterlocationindexcs_11)
                        (:%20%20
                          (Utilities|IsValid _result_7
                            (:Is%20Valid
                              (Transformation|AttachComponentToComponent _returnvalue_37 _result_7 "None" "KeepWorld" "KeepWorld" "KeepWorld" false) ; Attach so we can Hide/Unhide without making a million arrays
                            )))))))))))))))
