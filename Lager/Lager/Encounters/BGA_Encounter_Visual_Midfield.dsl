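; Blueprint actor that handles the minimap visuals for a Midfield encounter, judging by its name and special-actor components.
; It toggles two Athena special-actor components, an enabled minimap marker and a disabled one.
; Activate and Deactivate functions swap which marker is active, while Hide All deactivates only the enabled marker and leaves the disabled one untouched.
; The construction script is empty.
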(fn MiniMapActivate
  (Components|Activation|Activate (Variables|Default|GetAthenaSpecialActor_Midfield_Enabled_MiniMap) false)
  (Components|Activation|Deactivate (Variables|Default|GetAthenaSpecialActor_Midfield_Disabled)))

(fn MiniMapDeactivate
  (Components|Activation|Deactivate (Variables|Default|GetAthenaSpecialActor_Midfield_Enabled_MiniMap))
  (Components|Activation|Activate (Variables|Default|GetAthenaSpecialActor_Midfield_Disabled) false))

(fn MiniMapHideAll
  (Components|Activation|Deactivate (Variables|Default|GetAthenaSpecialActor_Midfield_Enabled_MiniMap)))

(fn ConstructionScript)
