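; Encounter visual POI Blueprint that can recolor a point of interest's name using context tags.
; ColorizePOIName acts only when the given POI tag is valid and the text color override context tag container is not empty.
; It then adds a GSC_MiniMapDataOverride component and sets the POI text with those context tags.
; The construction script is empty.
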(fn ColorizePOIName
  (params (POI%20Tag :type Gameplay%20Tag%20Structure))
  (bind _poi_textcoloroverridecontexttags (Variables|Default|GetPOI%20TextColorOverrideContextTags))
  (if (and (GameplayTags|IsGameplayTagValid POI%20Tag) (Utilities|GameplayTagContainerNotEmpty _poi_textcoloroverridecontexttags))
    (AddComponent|NODE_AddGSC_MiniMapDataOverride-0)
    (Default|SetPOITextwithContextTag POI%20Tag _poi_textcoloroverridecontexttags)))

(fn ConstructionScript)
