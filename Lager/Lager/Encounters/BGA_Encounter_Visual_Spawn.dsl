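; Visual marker for an encounter AI spawn point, showing a rift portal, holo loop cue and minimap/compass icons for attackers or defenders.
; A replicated ownership-state enum drives icons, portal warm-up and activation.
; After playlist load it traces for a damageable building support, ignoring it if indestructible ground also supports it.
; When that support is lost, it briefly turns Movable, teleports to the floor, re-bases and notifies listeners to relocate spawning.
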(event EventBeginPlay
  ;[1 1 1] Look for Building Actor Bases we need to listen to
  (Default|InWorldAllHidden) ; Make sure in-world ConversationComponents are Hidden
  (Class|FortAsyncActionCurrentPlaylistReady|PlaylistDataReadyAsync
    (:OnReady
      (bind (_founddamageablebuildingactorbase _buildingactor) (CallFunction|FindDamageableBuildingActorBase))
      (if _founddamageablebuildingactorbase
        (Utilities|IsValid _buildingactor
          (:Is%20Valid
            (Variables|Default|SetMyBuildingActorBase _buildingactor)
            (Variables|Default|SetMyDestroyedBuildingActorBase 0)
            (CallFunction|BindBuildingActorBaseEvents)))))))

(event Custom|AttackerDisabledDuration
  ;[1 1 1] After being Disabled, show icons for a limited time, then disable everything
  (Variables|Default|SetEncounterSpawnerOwnershipState "NewEnumerator0"))

(event Custom|FindBuildingActorBase
  ;[1 1 1] Look for Building Actor Bases we need to listen to
  (Class|FortAsyncActionCurrentPlaylistReady|PlaylistDataReadyAsync
    (:OnReady
      (bind (_founddamageablebuildingactorbase _buildingactor) (CallFunction|FindDamageableBuildingActorBase))
      (if _founddamageablebuildingactorbase
        (Utilities|IsValid _buildingactor
          (:Is%20Valid
            (Variables|Default|SetMyBuildingActorBase _buildingactor)
            (Variables|Default|SetMyDestroyedBuildingActorBase 0)
            (CallFunction|BindBuildingActorBaseEvents)))))))

(event Custom|BindBuildingActorBaseEvents
  ;[1 1 1] If our Building Actor Base dies, re-place ourselves downward.
  (Networking|SwitchHasAuthority
    (:Authority
      (bind _mybuildingactorbase (Variables|Default|GetMyBuildingActorBase)
        (:then
          (EventDispatchers|AssignOnDied _mybuildingactorbase (delegate Custom|OnDied_Event))
          (Game|AssignOnDestroyed _mybuildingactorbase (delegate Custom|OnDestroyed_Event))
          (Game|AssignOnEndPlay _mybuildingactorbase (delegate Custom|OnEndPlay_Event)))
        (:else)))))

(event Custom|UnBindBuildingActorBaseEvents
  ;[1 1 1] If our Building Actor Base dies, re-place ourselves downward.
  (Networking|SwitchHasAuthority
    (:Authority
      (bind _mybuildingactorbase (Variables|Default|GetMyBuildingActorBase)
        (:then
          (EventDispatchers|UnbindEventfromOnDied _mybuildingactorbase (delegate Custom|OnDied_Event))
          (Game|UnbindEventfromOnDestroyed _mybuildingactorbase (delegate Custom|OnDestroyed_Event))
          (Game|UnbindEventfromOnEndPlay _mybuildingactorbase (delegate Custom|OnEndPlay_Event))
          (Variables|Default|SetMyDestroyedBuildingActorBase (Variables|Default|GetMyBuildingActorBase)) ; Track the Building Actor that just died so we don't trace against it or base to it again.
          (Variables|Default|SetMyBuildingActorBase 0) ; Now clear our base
        )
        (:else)))))

(event Custom|OnDestroyed_Event
  (params (DestroyedActor :type Actor%20Object%20Reference))
  ;[1 1 1] If our Building Actor Base dies, re-place ourselves downward.
  (CallFunction|UnBindBuildingActorBaseEvents)
  (CallFunction|TeleporttoFloor) ; Since our Base is gone, try to teleport downward to the next thing.
)

(event Custom|OnDied_Event
  (params (DamagedActor :type Actor%20Object%20Reference) (Damage :type Float%20%28single-precision%29) (InstigatedBy :type Controller%20Object%20Reference) (DamageCauser :type Actor%20Object%20Reference) (HitLocation :type Vector) (FHitComponent :type Primitive%20Component%20Object%20Reference) (BoneName :type Name) (Momentum :type Vector))
  ;[1 1 1] If our Building Actor Base dies, re-place ourselves downward.
  (CallFunction|UnBindBuildingActorBaseEvents)
  (CallFunction|TeleporttoFloor) ; Since our Base is gone, try to teleport downward to the next thing.
)

(event Custom|OnEndPlay_Event
  (params (Actor :type Actor%20Object%20Reference) (EndPlayReason :type EEndPlayReason%20Enum :default "Destroyed"))
  ;[1 1 1] If our Building Actor Base dies, re-place ourselves downward.
  (CallFunction|UnBindBuildingActorBaseEvents)
  (CallFunction|TeleporttoFloor) ; Since our Base is gone, try to teleport downward to the next thing.
)

(event Custom|TeleportToFloor
  ;[1 1 1] If our Building Actor Base dies, re-place ourselves downward.
  (CallFunction|TeleportActortoFloor)
  (Utilities|FlowControl|Delay 1.0) ; Needs a small delay (longer than a Tick)\nor the spawner returns to where it was.
  (Transformation|SetMobility (Variables|Default|GetDefaultSceneRoot) "Static") ; Movable to Static so AI collision isn't wonky.
)

(fn EncounterSpawnerPawnSpawnFX
  (params (FortPlayerPawn :type Fort%20Player%20Pawn%20Object%20Reference))
  (Utilities|IsValid FortPlayerPawn
    (:Is%20Valid
      (bind _bga_portal_athena_encounter (Variables|Default|GetBGA_Portal_Athena_Encounter)
        (:then
          (bind _childactor (Variables|/Script/Engine.ChildActorComponent|GetChildActor _bga_portal_athena_encounter)
            (:then
              (bind _asbga_portal_athena_encounter (Utilities|Casting|CastToBGA_Portal_Athena_Encounter _childactor))
              (Class|BGAPortalAthena|PlayTeleportFX _asbga_portal_athena_encounter (Transformation|GetActorLocation FortPlayerPawn) FortPlayerPawn :!pinned true) ; Play Rift Portal VFX when AI spawn
            )
            (:else)))
        (:else)))))

(fn FindDamageableBuildingActorBase
  (bind _possiblebuildingactorbase (LocalVariables|GetPossibleBuildingActorBase))
  (bind (_origin _boxextent) (Collision|GetActorBounds _possiblebuildingactorbase false false))
  (bind _debug (Variables|Default|GetDebug))
  (LocalVariables|SetPossibleBuildingActorBase 0)
  (LocalVariables|SetBuildingActorBaseTraceLocation "0, 0, 0")
  (bind (_outhits _returnvalue) (Collision|MultiLineTraceForObjects (+ (Transformation|GetActorLocation) "0,0,100.000000") (+ (Transformation|GetActorLocation) "0,0,-100.000000") (Utilities|Array|MakeArray "ObjectTypeQuery1") false (Utilities|Array|MakeArray (Variables|Default|GetMyDestroyedBuildingActorBase)) (select (Variables|Default|GetDebug) "ForDuration" "None") true "(R=1.000000,G=0.000000,B=0.000000,A=1.000000)" "(R=0.000000,G=1.000000,B=0.000000,A=1.000000)" 3e+01 :!pinned true)) ; Trace for an object we might need to base ourselves upon.
  (if _returnvalue
    (bind _array_element (Utilities|Array|ForEachLoopwithBreak _outhits)
      (:LoopBody
        (if (Variables|/Script/Engine.Actor|GetbCanBeDamaged (.HitActor (Collision|BreakHitResult _array_element)))
          (bind _asbuilding_actor (Utilities|Casting|CastToBuildingActor (.HitActor (Collision|BreakHitResult _array_element)) :!pinned true)) ; Thing we have traced is a damageable building actor?
          (bind _mydestroyedbuildingactorbase (Variables|Default|GetMyDestroyedBuildingActorBase)
            (:then
              (if (!= _mydestroyedbuildingactorbase _asbuilding_actor)
                _))
            (:else
              _))
          (LocalVariables|GetPossibleBuildingActorBase
            (:then)
            (:else
              (LocalVariables|SetPossibleBuildingActorBase _asbuilding_actor)
              (LocalVariables|SetBuildingActorBaseTraceLocation (.Location (Collision|BreakHitResult _array_element)))
              (Development|PrintString (Utilities|String|BuildString(Object) (Utilities|String|BuildString(Object) "" "Spawn = " self " | ") "Hit = " (.HitActor (Collision|BreakHitResult _array_element))) (Variables|Default|GetDebug) (Variables|Default|GetDebug)))) ; Only set if NOT already set.\n\nIf we already found something, stick with it.
        ))
      (:Completed
        (bind _possiblebuildingactorbase_2 (LocalVariables|GetPossibleBuildingActorBase)
          (:then
            (LocalVariables|SetWeAreActuallyBasedOnSomethingIndestructible false)
            (bind _array_element_2 (Utilities|Array|ForEachLoopwithBreak _outhits :!pinned true)
              (:LoopBody
                (if (Variables|/Script/Engine.Actor|GetbCanBeDamaged (.HitActor (Collision|BreakHitResult _array_element_2)))
                  _
                  (elif (< (Math|Vector|VectorLength (- (.Location (Collision|BreakHitResult _array_element_2)) (LocalVariables|GetBuildingActorBaseTraceLocation))) 2e+01)
                    (LocalVariables|SetWeAreActuallyBasedOnSomethingIndestructible true :!pinned true) ; We're actually on the ground (or similar)
                  )))
              (:Completed
                _)) ; Double check our results to make sure we aren't actually \nsupported by the ground or something indestructible.
            (if (LocalVariables|GetWeAreActuallyBasedOnSomethingIndestructible)
              (if _debug
                (Debug|DebugBoxSendtoAll _origin _boxextent (Utilities|Struct|MakeColor 0 255 255 255) true -1.0 0 4.0)
                (else
                  _))
              (return false 0)
              (elif _debug
                (Debug|DebugBoxSendtoAll _origin _boxextent (Utilities|Struct|MakeColor 0 255 0 255) true -1.0 0 4.0)
                (return true _possiblebuildingactorbase_2)
                (else
                  _))))
          (:else
            (return false 0)))))))

(fn InWorldAllHidden
  (InteractionUI|AddHideUIIndicatorReason (Variables|Default|GetNonPlayerConversationComponent_Spawn_Attacker) (Utilities|Name|MakeLiteralName "None")))

(fn InWorldAttackerDisable
  (bind _returnvalue (Utilities|Name|MakeLiteralName "None"))
  (InteractionUI|AddHideUIIndicatorReason (Variables|Default|GetNonPlayerConversationComponent_Spawn_Defender) _returnvalue)
  (if (Variables|Default|GetAttackerShowUI)
    (InteractionUI|RemoveHideUIIndicatorReason (Variables|Default|GetNonPlayerConversationComponent_Spawn_Attacker_Disabled) _returnvalue)))

(fn InWorldAttackerEnable
  (bind _returnvalue (Utilities|Name|MakeLiteralName "None"))
  (InteractionUI|AddHideUIIndicatorReason (Variables|Default|GetNonPlayerConversationComponent_Spawn_Defender) _returnvalue)
  (if (Variables|Default|GetAttackerShowUI)
    (InteractionUI|RemoveHideUIIndicatorReason (Variables|Default|GetNonPlayerConversationComponent_Spawn_Attacker) _returnvalue)))

(fn InWorldDefenderDisable
  (params (Pre-VisNewSpawn :type Boolean))
  (bind _returnvalue (Utilities|Name|MakeLiteralName "None"))
  (InteractionUI|AddHideUIIndicatorReason (Variables|Default|GetNonPlayerConversationComponent_Spawn_Attacker) _returnvalue)
  (if (or (Variables|Default|GetDefenderShowUIForDisabledState) Pre-VisNewSpawn)
    (InteractionUI|RemoveHideUIIndicatorReason (Variables|Default|GetNonPlayerConversationComponent_Spawn_Defender_Disabled) _returnvalue)))

(fn InWorldDefenderEnable
  (bind _returnvalue (Utilities|Name|MakeLiteralName "None"))
  (InteractionUI|AddHideUIIndicatorReason (Variables|Default|GetNonPlayerConversationComponent_Spawn_Attacker) _returnvalue)
  (InteractionUI|RemoveHideUIIndicatorReason (Variables|Default|GetNonPlayerConversationComponent_Spawn_Defender) _returnvalue))

(fn MiniMapSetAttackerActivated
  (bind _athenaspecialactor_spawn_attacker_minimap (Variables|Default|GetAthenaSpecialActor_Spawn_Attacker_MiniMap))
  (bind _athenaspecialactor_spawn_attacker_compass (Variables|Default|GetAthenaSpecialActor_Spawn_Attacker_Compass))
  (Components|Activation|Deactivate (Variables|Default|GetAthenaSpecialActor_Spawn_Attacker_Disabled_MiniMap))
  (if (Variables|Default|GetAttackerShowUI)
    (Utilities|FlowControl|Sequence
      (:then_0
        (if (not (Components|Activation|IsActive _athenaspecialactor_spawn_attacker_minimap))
          (Components|Activation|Activate _athenaspecialactor_spawn_attacker_minimap false)))
      (:then_1
        (if (not (Components|Activation|IsActive _athenaspecialactor_spawn_attacker_compass))
          (Components|Activation|Activate _athenaspecialactor_spawn_attacker_compass false))))))

(fn MiniMapSetAttackerDisabled
  (bind _athenaspecialactor_spawn_attacker_disabled_compass (Variables|Default|GetAthenaSpecialActor_Spawn_Attacker_Disabled_Compass))
  (bind _athenaspecialactor_spawn_attacker_disabled_minimap (Variables|Default|GetAthenaSpecialActor_Spawn_Attacker_Disabled_MiniMap))
  (Components|Activation|Deactivate (Variables|Default|GetAthenaSpecialActor_Spawn_Defender_Disabled_MiniMap))
  (if (Variables|Default|GetAttackerShowUI)
    (Utilities|FlowControl|Sequence
      (:then_0
        (if (not (Components|Activation|IsActive _athenaspecialactor_spawn_attacker_disabled_minimap))
          (Components|Activation|Activate _athenaspecialactor_spawn_attacker_disabled_minimap false)))
      (:then_1
        (if (not (Components|Activation|IsActive _athenaspecialactor_spawn_attacker_disabled_compass))
          (Components|Activation|Activate _athenaspecialactor_spawn_attacker_disabled_compass false))))))

(fn MiniMapSetDefenderActivated
  (bind _athenaspecialactor_spawn_defender_minimap (Variables|Default|GetAthenaSpecialActor_Spawn_Defender_MiniMap))
  (bind _athenaspecialactor_spawn_defender_compass (Variables|Default|GetAthenaSpecialActor_Spawn_Defender_Compass))
  (Components|Activation|Deactivate (Variables|Default|GetAthenaSpecialActor_Spawn_Attacker_Disabled_MiniMap))
  (Utilities|FlowControl|Sequence
    (:then_0
      (if (not (Components|Activation|IsActive _athenaspecialactor_spawn_defender_minimap))
        (Components|Activation|Activate _athenaspecialactor_spawn_defender_minimap false)))
    (:then_1
      (if (not (Components|Activation|IsActive _athenaspecialactor_spawn_defender_compass))
        (Components|Activation|Activate _athenaspecialactor_spawn_defender_compass false)))))

(fn MiniMapSetDefenderDisabled
  (bind _athenaspecialactor_spawn_defender_disabled_minimap (Variables|Default|GetAthenaSpecialActor_Spawn_Defender_Disabled_MiniMap))
  (bind _athenaspecialactor_spawn_defender_disabled_compass (Variables|Default|GetAthenaSpecialActor_Spawn_Defender_Disabled_Compass))
  (Components|Activation|Deactivate (Variables|Default|GetAthenaSpecialActor_Spawn_Attacker_Disabled_MiniMap))
  (Utilities|FlowControl|Sequence
    (:then_0
      (if (not (Components|Activation|IsActive _athenaspecialactor_spawn_defender_disabled_minimap))
        (Components|Activation|Activate _athenaspecialactor_spawn_defender_disabled_minimap false)))
    (:then_1
      (if (not (Components|Activation|IsActive _athenaspecialactor_spawn_defender_disabled_compass))
        (Components|Activation|Activate _athenaspecialactor_spawn_defender_disabled_compass false)))))

(fn MiniMapSetHidden
  (Components|Activation|Deactivate (Variables|Default|GetAthenaSpecialActor_Spawn_Attacker_Disabled_MiniMap))
  (Variables|Default|SetSpawningIsActive false)
  (CallFunction|VisualsDisable))

(fn OnRep_EncounterSpawnerOwnershipState
  (switch Utilities|FlowControl|Switch|SwitchonEnum_Encounter_Visual_Spawn_Ownership_State (Variables|Default|GetEncounterSpawnerOwnershipState)
    (:NewEnumerator0
      (CallFunction|MiniMapSetHidden)
      (Default|InWorldAllHidden))
    (:NewEnumerator1
      (CallFunction|SetAttackerDisabled true)
      (bind _bga_portal_athena_encounter (Variables|Default|GetBGA_Portal_Athena_Encounter)
        (:then
          (bind _childactor (Variables|/Script/Engine.ChildActorComponent|GetChildActor _bga_portal_athena_encounter)
            (:then
              (bind _asbga_portal_athena_encounter (Utilities|Casting|CastToBGA_Portal_Athena_Encounter _childactor))
              (Class|BGAPortalAthenaEncounter|AttackerWarmUp _asbga_portal_athena_encounter))
            (:else)))
        (:else)))
    (:NewEnumerator2
      (CallFunction|SetAttackerActivated))
    (:NewEnumerator3
      (CallFunction|SetAttackerDisabled true)
      (Networking|SwitchHasAuthority
        (:Authority)))
    (:NewEnumerator4
      (CallFunction|SetDefenderDisabled true)
      (bind _bga_portal_athena_encounter_2 (Variables|Default|GetBGA_Portal_Athena_Encounter)
        (:then
          (bind _childactor_2 (Variables|/Script/Engine.ChildActorComponent|GetChildActor _bga_portal_athena_encounter_2)
            (:then
              (bind _asbga_portal_athena_encounter_2 (Utilities|Casting|CastToBGA_Portal_Athena_Encounter _childactor_2))
              (Class|BGAPortalAthenaEncounter|DefenderWarmUp _asbga_portal_athena_encounter_2))
            (:else)))
        (:else)))
    (:NewEnumerator5
      (CallFunction|SetDefenderActivated))
    (:NewEnumerator6
      (CallFunction|SetDefenderDisabled true))
    (:NewEnumerator8
      (Default|MiniMapSetAttackerDisabled))
    (:NewEnumerator9
      (CallFunction|MiniMapSetDefenderDisabled))))

(fn SetAttackerActivated
  (Variables|Default|SetBelongsToAttackers true)
  (Default|MiniMapSetAttackerActivated)
  (Variables|Default|SetSpawningIsActive true)
  (CallFunction|VisualsActivate))

(fn SetAttackerDisabled
  (params (Pre-VisNewSpawn :type Boolean))
  (Variables|Default|SetBelongsToAttackers true)
  (Default|MiniMapSetAttackerDisabled)
  (if Pre-VisNewSpawn
    (Variables|Default|SetSpawningIsActive false)
    (CallFunction|VisualsActivate)
    (else
      (Variables|Default|SetSpawningIsActive false)
      (CallFunction|VisualsDisable))))

(fn SetDefenderActivated
  (Variables|Default|SetBelongsToAttackers false)
  (Default|MiniMapSetDefenderActivated)
  (Variables|Default|SetSpawningIsActive true)
  (CallFunction|VisualsActivate))

(fn SetDefenderDisabled
  (params (Pre-VisNewSpawn :type Boolean))
  (Variables|Default|SetBelongsToAttackers false)
  (if (Variables|Default|GetDefenderShowUIForDisabledState)
    _
    (elif Pre-VisNewSpawn
      _
      (else
        (CallFunction|MiniMapSetHidden)
        (Variables|Default|SetSpawningIsActive false)
        (CallFunction|VisualsDisable))))
  (CallFunction|MiniMapSetDefenderDisabled)
  (Variables|Default|SetSpawningIsActive false)
  (CallFunction|VisualsActivate))

(fn TeleportActorToFloor
  (Utilities|FlowControl|Sequence
    (:then_0
      (bind (_outhit _returnvalue) (Collision|LineTraceForObjects (Transformation|GetActorLocation) (+ (Transformation|GetActorLocation) "0,0,-10000.000000") (Utilities|Array|MakeArray "ObjectTypeQuery1") false (Utilities|Array|MakeArray (Variables|Default|GetMyDestroyedBuildingActorBase)) (select (Variables|Default|GetDebug) "ForDuration" "None") true "(R=1.000000,G=0.000000,B=0.000000,A=1.000000)" "(R=0.000000,G=1.000000,B=0.000000,A=1.000000)" 3e+01 :!pinned true)) ; We lost our base, trace down to find a new base, then teleport there.
      (if _returnvalue
        (if (Variables|Default|GetDebug)
          (Debug|DebugSphereSendtoAll (.ImpactPoint (Collision|BreakHitResult _outhit)) 1e+02 8 (Utilities|Struct|MakeColor 255 0 0 255) true 3e+01 0 4.0))
        (Utilities|FlowControl|Sequence
          (:then_0
            (bind _bga_portal_athena_encounter (Variables|Default|GetBGA_Portal_Athena_Encounter)
              (:then
                (bind _childactor (Variables|/Script/Engine.ChildActorComponent|GetChildActor _bga_portal_athena_encounter)
                  (:then
                    (bind _asbga_portal_athena_encounter (Utilities|Casting|CastToBGA_Portal_Athena_Encounter _childactor))
                    (Class|BGAPortalAthenaEncounter|VFXPortalReBaseTeleportOfItself _asbga_portal_athena_encounter (.ImpactPoint (Collision|BreakHitResult _outhit)) self :!pinned true) ; Reach up into the Rift Portal VFX BP and make it play similar \nteleport VFX for when it teleports itself to a new location.
                  )
                  (:else)))
              (:else)))
          (:then_1
            (Transformation|SetMobility (Variables|Default|GetDefaultSceneRoot) "Movable") ; Static to Movable so it can be teleported.
            (Transformation|SetActorLocation self (.ImpactPoint (Collision|BreakHitResult _outhit)) false true)))))
    (:then_1
      (CallFunction|FindBuildingActorBase :!pinned true) ; Will trace near ourselves to re-base, if appropriate.
    )
    (:then_2
      (Default|CallSpawnVisualActorTeleported self :!pinned true) ; Will call back into the Encounter actor to re-place the Point Provider for the new spawn location
    )))

(fn ConstructionScript
  (|CallParentFunction)
  (Transformation|SetWorldLocation (Variables|Default|GetStaticMesh) (Transformation|GetActorLocation) false true) ; With the Static Mesh set to Mobility Static (required for proper NPC AI movement), \nwe must re-set the Location to match the Actor for some reason.
)

(fn VisualsActivate
  (bind _holo_loop_cue (Variables|Default|GetHolo%20Loop%20Cue))
  (Networking|CanExecuteCosmeticEvents
    (:True
      (Utilities|FlowControl|Sequence
        (:then_0
          (if (not (Components|Activation|IsActive _holo_loop_cue))
            (Components|Activation|Activate _holo_loop_cue false)))
        (:then_2
          (bind _bga_portal_athena_encounter (Variables|Default|GetBGA_Portal_Athena_Encounter)
            (:then
              (bind _childactor (Variables|/Script/Engine.ChildActorComponent|GetChildActor _bga_portal_athena_encounter)
                (:then
                  (bind _asbga_portal_athena_encounter (Utilities|Casting|CastToBGA_Portal_Athena_Encounter _childactor))
                  (Class|BGAPortalAthenaEncounter|EnableorDisablePortal _asbga_portal_athena_encounter true (Variables|Default|GetSpawningIsActive)))
                (:else)))
            (:else)))))))

(fn VisualsDisable
  (bind _holo_loop_cue (Variables|Default|GetHolo%20Loop%20Cue))
  (Networking|CanExecuteCosmeticEvents
    (:True
      (Utilities|FlowControl|Sequence
        (:then_0
          (if (Components|Activation|IsActive _holo_loop_cue)
            (Components|Activation|Deactivate _holo_loop_cue)))
        (:then_2
          (bind _bga_portal_athena_encounter (Variables|Default|GetBGA_Portal_Athena_Encounter)
            (:then
              (bind _childactor (Variables|/Script/Engine.ChildActorComponent|GetChildActor _bga_portal_athena_encounter)
                (:then
                  (bind _asbga_portal_athena_encounter (Utilities|Casting|CastToBGA_Portal_Athena_Encounter _childactor))
                  (Class|BGAPortalAthenaEncounter|EnableorDisablePortal _asbga_portal_athena_encounter false (Variables|Default|GetSpawningIsActive)))
                (:else)))
            (:else)))))))
