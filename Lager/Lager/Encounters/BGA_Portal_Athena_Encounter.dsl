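; Encounter spawn portal that handles VFX, sounds and building-block collision for attacker and defender spawners.
; EnableOrDisablePortal reads team and encounter state from its parent actors, then treats portal-open and spawning-active as separate states.
; Closing fades out and deactivates the Niagara loop except in replays, while inactive spawning dims only its head.
; Events add team warm-up sounds, a beam and pulse on player teleport, and effects when the portal relocates itself.
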(event Custom|CE_FadeDownSpawnerVFX
  (bind _ns_spawn_idle_loop (Variables|Default|GetNS_Spawn_Idle_Loop))
  ;[1 1 1] FadeDown the spawner VFX
  (bind (_direction _fadedowncurve) (Timeline|FadeDownSpawnerVFX PlayFromStart)
    (:Update
      (Development|PrintString "PortalFadeDown!!!!!!!!!!!!!!!!!!!!" false false "(R=0.000000,G=0.935358,B=1.000000,A=1.000000)" 2e+01)
      (Effects|Components|ParticleSystem|SetFloatParameter _ns_spawn_idle_loop "User.OverallVisibility" _fadedowncurve)
      (Effects|Components|ParticleSystem|SetFloatParameter _ns_spawn_idle_loop "IsSpawnActiveState" 1.0))
    (:Finished
      (Development|PrintString "PortalClosed!!!!!!!!!!!!!!!!!!!!" false false "(R=0.000000,G=0.175256,B=1.000000,A=1.000000)" 2e+01)
      (Components|Activation|Deactivate (Variables|Default|GetNS_Spawn_Idle_Loop)))))

(event Custom|CE_UpdateSpawnerVFXState
  ;[1 1 1] Only Fade Down the head part VFX to display the deactived spawning state
  (Utilities|FlowControl|DoOnce true
    (:Completed
      (bind (_direction _updatecurve) (Timeline|UpdateSpawnerVFXState PlayFromStart)
        (:Update
          (Effects|Components|ParticleSystem|SetFloatParameter (Variables|Default|GetNS_Spawn_Idle_Loop) "IsSpawnActiveState" _updatecurve))))))

(event Custom|CE_Reset
  ;[1 1 1] Only Fade Down the head part VFX to display the deactived spawning state
  (Utilities|FlowControl|DoOnce true
    (:Completed
      (Timeline|UpdateSpawnerVFXState PlayFromStart))))

(event EventBeginPlay
  (|CallParentFunction)
  (CallFunction|EnableorDisablePortal (Variables|Portal|GetPortalOpen) (Variables|Default|GetSpawningIsActive)))

(event Custom|VFXPortalReBaseTeleportOfItself
  (params (TeleportLocation :type Vector) (ActorBeingTeleported :type Actor%20Object%20Reference))
  ;[1 1 1] If the Portal's Base is destroyed, it teleports itself down.
  ;
  ; These FX play while teleporting to the new location.
  (Effects|SpawnEmitteratLocation (Variables|VFX|GetEnterPortal) (Transformation|GetActorLocation ActorBeingTeleported)) ; Teleport VFX where this actor is leaving from.
  (Effects|SpawnEmitteratLocation (Variables|VFX|GetEnterPortal) TeleportLocation) ; Teleport VFX where this actor is going to.
  (Audio|PlaySoundatLocation (Variables|Sound|GetEnteredPortal) (Transformation|GetActorLocation ActorBeingTeleported))
  (Audio|PlaySoundatLocation (Variables|Sound|GetExitedPortal) TeleportLocation))

(event EventPlayTeleportFX
  (params (PlayerLocation :type Vector :default "0, 0, 0") (Player :type Fort%20Player%20Pawn%20Object%20Reference))
  ;[1 1 1] Set the Spawn teleport beam VFX location
  (|CallParentFunction PlayerLocation Player)
  (bind _enterportalfx (Variables|Default|GetEnterPortalFX)
    (:then
      (Effects|Components|ParticleSystem|SetVectorParameter _enterportalfx "BeamStartPoint" (+ (Transformation|GetActorLocation) "0,0,250.000000"))
      (bind (_direction _pulsecurve) (Timeline|TeleportPulseOnSpawnerFX PlayFromStart)
        (:Update
          (Effects|Components|ParticleSystem|SetFloatParameter (Variables|Default|GetNS_Spawn_Idle_Loop) "IsSpawnActiveState" (Math|Float|Lerp 1.0 1.45 _pulsecurve)))))
    (:else
      (Timeline|TeleportPulseOnSpawnerFX PlayFromStart))))

(event Custom|AttackerWarmUp
  ;[1 1 1] Pre-Activation Warm-Up sounds (and possibly VFX)
  (Networking|CanExecuteCosmeticEvents
    (:True
      (Audio|PlaySoundatLocation (Variables|EncounterSounds|GetSoundAttackerWarmUp) (Transformation|GetActorLocation)))))

(event Custom|DefenderWarmUp
  ;[1 1 1] Pre-Activation Warm-Up sounds (and possibly VFX)
  (Networking|CanExecuteCosmeticEvents
    (:True
      (Audio|PlaySoundatLocation (Variables|EncounterSounds|GetSoundDefenderWarmUp) (Transformation|GetActorLocation)))))

(fn EnableOrDisablePortal
  (params (VisualsEnabled :type Boolean) (SpawningIsActive :type Boolean))
  (bind _returnvalue (Actor|GetParentActor))
  (bind _returnvalue_2 (Actor|GetParentActor _returnvalue))
  (bind _blockbuildingbox (Variables|Default|GetBlockBuildingBox))
  (bind _audio_looping (Variables|Sound|GetAudio_Looping))
  (Variables|Portal|SetPortalOpen VisualsEnabled)
  (Variables|Default|SetSpawningIsActive SpawningIsActive)
  (Utilities|FlowControl|Sequence
    (:then_0
      ;[1 1 1] Set Vars
      (Utilities|IsValid _returnvalue
        (:Is%20Valid
          (bind _asbga_encounter_visual_spawn (Utilities|Casting|CastToBGA_Encounter_Visual_Spawn _returnvalue))
          (Variables|Default|SetBelongsToAttackers (Variables|/Lager/Lager/Encounters/BGA_Encounter_Visual_Spawn.BGA_Encounter_Visual_Spawn_C|GetBelongsToAttackers _asbga_encounter_visual_spawn))
          (Utilities|IsValid _returnvalue_2
            (:Is%20Valid
              (bind _asbga_encounter_linear_parent (Utilities|Casting|CastToBGA_Encounter_Linear_Parent _returnvalue_2))
              (Variables|Default|SetLinearEncounterIsActive (Variables|/Lager/Lager/Encounters/BGA_Encounter_Linear_Parent.BGA_Encounter_Linear_Parent_C|GetLinearEncounterIsActive _asbga_encounter_linear_parent)))))))
    (:then_1
      ;[1 1 1] Active/Disabled Visuals
      (if (Variables|Portal|GetPortalOpen)
        (Components|Activation|Activate _audio_looping false)
        (Effects|Components|ParticleSystem|SetFloatParameter (Variables|Default|GetNS_Spawn_Idle_Loop) "User.OverallVisibility" 1.0)
        ;[1 1 1] Enable/Disable building blocking.
        (Collision|SetCollisionEnabled _blockbuildingbox "QueryOnly")
        (elif (Replay|GetIsPlayingReplay)
          _
          (else
            (Components|Activation|Deactivate _audio_looping)
            (Utilities|FlowControl|Sequence
              (:then_0
                (Collision|SetCollisionEnabled _blockbuildingbox "NoCollision"))
              (:then_1
                (CallFunction|CEFadeDownSpawnerVFX)))))))
    (:then_2
      ;[1 1 1] Separate visuals as to whether a spawner with Active/PortalOpen visuals (above) is also Enabled for Spawning.
      ; We can have an Active/PortalOpen=True, but have Spawning=False.
      (if (Variables|Default|GetSpawningIsActive)
        (Effects|Components|ParticleSystem|SetFloatParameter (Variables|Default|GetNS_Spawn_Idle_Loop) "IsSpawnActiveState" 1.0)
        (CallFunction|CEReset)
        (else
          (CallFunction|CEUpdateSpawnerVFXState))))
    (:then_3
      ;[1 1 1] Sounds
      (if (Variables|Default|GetLinearEncounterIsActive)
        (if (Variables|Portal|GetPortalOpen)
          (Audio|PlaySoundatLocation (select (Variables|Default|GetBelongsToAttackers) (Variables|EncounterSounds|GetSoundAttackerPowerOn) (Variables|EncounterSounds|GetSoundDefenderPowerOn)) (Transformation|GetActorLocation) :!pinned true) ; Power-Up Sounds
          (else
            (Audio|PlaySoundatLocation (select (Variables|Default|GetBelongsToAttackers) (Variables|EncounterSounds|GetSoundAttackerPowerDown) (Variables|EncounterSounds|GetSoundDefenderPowerDown)) (Transformation|GetActorLocation) :!pinned true) ; Power-Down Sounds
          ))))))

(fn ConstructionScript
  (|CallParentFunction))
