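; Debug cheat manager actor for starting linear encounters by name, exposed through the Debug_StartEncounter event.
; It lowercases the given name, looks it up in an Encounters map and calls DebugStartLinearEncounter on the valid result.
; Optionally it teleports every Fortnite player controller's pawn to that encounter's location, spacing players 100 units apart along Y.
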(event EventBeginPlay)

(event Collision|EventActorBeginOverlap
  (params (OtherActor :type Actor%20Object%20Reference)))

(event EventTick
  (params (DeltaSeconds :type Float%20%28single-precision%29)))

(event Custom|Debug_StartEncounter
  (params (EncounterName :type String) (TeleportPlayers :type Boolean))
  (bind _returnvalue (Utilities|String|ToLower EncounterName))
  (bind _encounters (Variables|Default|GetEncounters))
  (bind (_value _returnvalue_2) (Utilities|Map|Find _encounters _returnvalue))
  (if _returnvalue_2
    (Utilities|IsValid _value
      (:Is%20Valid
        (Class|BGAEncounterLinearParent|DebugStartLinearEncounter _value)
        (if TeleportPlayers
          (bind _outactors (Actor|GetAllActorsOfClass "/Script/FortniteGame.FortPlayerControllerAthena"))
          (for (_array_element _array_index) _outactors
            (Transformation|Teleport (Game|Player|GetPlayerPawn _array_element) (+ (* _array_index "0,100.000000,0.000000") (Transformation|GetActorLocation _value)) "0, 0, 0")))))))

(fn ConstructionScript)
