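# Niagara system for an encounter shutdown effect, played as a converging intro followed by a delayed final burst.
# An intro flash, ring and shimmer accompany GPU embers and streaks pulled inward by point attraction.
# System.FinalFlashDelay times the burst: warm flash with light, shell and crack meshes, layered flare and repelled rising embers, with short-lived sprites trailing slightly later.
# System.DistanceLerp scales opacity or size on several emitters by distance.
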NiagaraSystem /Lager/Lager/Encounters/FX/NS_Encounter_ShutDown
dependencies {
 usedDataInterfaces [
  &NiagaraDataInterfaceCurve
  &NiagaraDataInterfaceVector2DCurve
  &NiagaraDataInterfaceVectorCurve
 ]
 usedDynamicInputs [&FN_LOD_Distance_Remap]
 usedModules [
  &self:SystemSpawnScript.SystemScriptSource.SystemScriptGraph.NiagaraNodeAssignment_2.SetVariables_#1
  &self:SystemSpawnScript.SystemScriptSource.SystemScriptGraph.NiagaraNodeAssignment_1.SetVariables_#2
  &self:SystemSpawnScript.SystemScriptSource.SystemScriptGraph.NiagaraNodeAssignment_0.SetVariables_#3
  &SystemState
 ]
}
emitters {
 CrackSpline {
  emitterUpdateScript [
   &EmitterState()
   &SpawnBurst_Instantaneous(Spawn Count=2, Spawn Time=$System.FinalFlashDelay, Use Spawn Probability=false)
  ]
  particleSpawnScript [
   /SetVariables(Particles.MeshOrientation=&MakeQuatFromAxisAngle(
     Axis=&RandomRangeVector(Minimum=<0 0 0>, Maximum=<1 1 1>, Evaluation Type=::NewEnumerator1)
     AngleInDegrees=90
    )) {isSetParameters=true}
   /Niagara/Modules/Spawn/Initialization/InitializeParticle(
    Lifetime=&RandomRangeFloat(Minimum=0.7, Maximum=1, Evaluation Type=::NewEnumerator1)
    Mesh Scale=&Multiply_VectorByFloat(Vector=&RandomRangeVector(Minimum=<2 2 3>, Maximum=<4 4 5>, Evaluation Type=::NewEnumerator1), Float=3)
    Module.Write Color=true
    Module.Write Lifetime=true
    Module.Write Mass=true
    Module.Write Material Random=true
    Module.Write Position=false
    Module.Write Scale=true
    Module.Write SpriteRotation=false
    Module.Write SpriteSize=false
    Position=<0 0 15>
   )
   &Color(Color=&MakeLinearColorFromVectorAndFloat("Vector (RGB)"=<1.5 1.5 1.5>, "Float (Alpha)"=&ClampFloat(Float=1)))
   &SystemLocation(MaskSpawnGroup=false, Offset=<0 0 30>)
  ]
  particleUpdateScript [
   &ParticleState()
   &ScaleColor(
    Scale Alpha=&FloatFromCurve(FloatCurve=DataInterface() {propertyValues {Curve {keys [<at 0.77 1 RCIM_Cubic>,<at 1 0>]}}})
    Scale RGB=&VectorFromCurve(
     VectorCurve=DataInterface() {propertyValues {
       XCurve {keys [
         {arriveTangent=-6.9225, interpMode=RCIM_Cubic, time=0, value=4.5}
         {arriveTangent=-3.6158, interpMode=RCIM_Cubic, leaveTangent=-3.6158, time=0.65, value=1}
         <at 1 0.8842 RCIM_Cubic>
        ]}
       YCurve {keys [
         {arriveTangent=-15.8226, interpMode=RCIM_Cubic, time=0, value=6}
         {arriveTangent=-5.6986, interpMode=RCIM_Cubic, leaveTangent=-5.6986, time=0.65, value=1}
         {interpMode=RCIM_Cubic, leaveTangent=-5.843, time=1, value=0.3014}
        ]}
       ZCurve {keys [
         {arriveTangent=-15.2224, interpMode=RCIM_Cubic, time=0, value=6}
         {arriveTangent=-5, interpMode=RCIM_Cubic, leaveTangent=-5, time=0.65, value=1}
         <at 1 1 RCIM_Cubic>
        ]}
      }}
     Scale Curve=<0.12 0.12 0.12>
    )
    ScaleA=true
    ScaleRGB=true
   )
   &DynamicMaterialParameters(
    Index 0 Param 1=&FloatFromCurve(FloatCurve=DataInterface() {propertyValues {Curve {keys [
         <at 0 1 RCIM_Cubic>
         {arriveTangent=-1, interpMode=RCIM_Cubic, leaveTangent=-1, time=0.3329, value=0.616}
         {interpMode=RCIM_Cubic, leaveTangent=-0.3905, time=1, value=0}
        ]}}})
    Index 0 Param 2=&FloatFromCurve(
     FloatCurve=DataInterface() {propertyValues {Curve {keys [
         {arriveTangent=-2.0225, interpMode=RCIM_Cubic, time=0, value=0.6}
         {arriveTangent=-0.6, interpMode=RCIM_Cubic, leaveTangent=-0.6, time=0.65, value=0.25}
         <at 1 0 RCIM_Cubic>
        ]}}}
     Scale Curve=1.5
    )
    Index 0 Param 3=&FloatFromCurve(FloatCurve=DataInterface() {propertyValues {Curve {keys [{arriveTangent=1.0145, interpMode=RCIM_Cubic, time=0, value=-0.1}, <at 0.25 0>]}}})
    Index 0 Param 4=&FloatFromCurve(FloatCurve=DataInterface() {propertyValues {Curve {keys [<at 0 0 RCIM_Cubic>,<at 1 0>]}}})
    Module.Param0WriteEnabled=true
    Module.Param1WriteEnabled=true
    Module.Param2WriteEnabled=true
    Module.Param3WriteEnabled=true
   )
   &CameraOffset(Camera Offset Amount=-100)
  ]
  rendererClasses [&NiagaraMeshRendererProperties]
  renderers [{rendererClass=&NiagaraMeshRendererProperties, rendererIndex=0}]
 }
 Embers {
  simTarget=GPUComputeSim
  emitterUpdateScript [
   &EmitterState(Life Cycle Mode=::NewEnumerator1, Loop Behavior=::NewEnumerator1, Loop Delay=$System.FinalFlashDelay, Loop Duration=1, UseLoopDelay=true)
   &SpawnRate(
    SpawnRate=&FloatFromCurve(
     FloatCurve=DataInterface() {propertyValues {Curve {keys [<at 0 1 RCIM_Cubic>, <at 0.3 0.1>, {tangentMode=RCTM_User, time=1, value=0}]}}}
     CurveIndex=$Emitter.NormalizedLoopAge
     Scale Curve=1000
    )
    Use Spawn Probability=false
   )
  ]
  particleSpawnScript [
   /SetVariables(Particles.SpriteRotationRate=&RandomRangeFloat(Minimum=-1, Maximum=1, Evaluation Type=::NewEnumerator1)) {isSetParameters=true, module=SetVariables_#4}
   /Niagara/Modules/Spawn/Initialization/V2/InitializeParticle(
    Color Maximum=<rgba 20 8.52 0.5167 1>
    Color Minimum=<rgba 3.3946 13.449 20 1>
    Color Mode=::NewEnumerator2
    Lifetime Max=3
    Lifetime Min=2
    Lifetime Mode=::NewEnumerator1
    Sprite Size Mode=::NewEnumerator3
    Uniform Sprite Size=&Multiply_Float(A=&RandomRangeFloat(Minimum=10, Maximum=25), B=$System.DistanceLerp)
    UsePositionOffset=false
    Write Material Random=true
   ) {module=InitializeParticle001}
   /Niagara/Modules/Spawn/Location/SphereLocation(
    MaskSpawnGroup=false
    Module.Override Randomness=false
    Sphere Radius=&RandomRangeFloat(Minimum=50, Maximum=170, Evaluation Type=::NewEnumerator1)
    Surface Only=true
   )
   &AddVelocity(Velocity=&RandomRangeVector(Minimum=<150 150 500>, Maximum=<-150 -150 2500>))
  ]
  particleUpdateScript [
   &ParticleState()
   &CurlNoiseForce(Noise Frequency=30, Noise Strength=500, Pan Noise=false)
   &AccelerationForce(Acceleration=&MakeVector(X=0, Y=0, Z=&RandomRangeFloat(Minimum=0, Maximum=150)))
   &GravityForce(Gravity=<0 0 -1880>)
   &PointAttractionForce(
    Attraction Radius=1000
    Attraction Strength=&FloatFromCurve(
     FloatCurve=DataInterface() {propertyValues {Curve {keys [
         {arriveTangent=5.2479, interpMode=RCIM_Cubic, leaveTangent=5.2479, tangentMode=RCTM_User, time=0, value=0}
         {arriveTangent=0.2391, interpMode=RCIM_Cubic, leaveTangent=0.2391, tangentMode=RCTM_User, time=0.8, value=1}
        ]}}}
     Scale Curve=-30
    )
    Attractor Position Offset:Vector3f=<0 0 300>
    Kill Within Radius=false
    Module.Use Falloff=true
    UseAttractorOffset=true
   )
   &Drag(Drag=4, Use Linear Drag=true, Use Rotational Drag=false)
   &SolveForcesAndVelocity(Clamp Velocity=false, Limit Acceleration=false)
   &ScaleSpriteSize(
    Uniform Curve Scale=&RandomRangeFloat(Minimum=0.3, Maximum=1.5, Evaluation Type=::NewEnumerator1)
    Uniform Curve Sprite Scale=DataInterface() {propertyValues {Curve {keys [
        {interpMode=RCIM_Cubic, tangentMode=RCTM_User, time=0, value=2}
        {interpMode=RCIM_Cubic, tangentMode=RCTM_User, time=0.1, value=1.175}
        {arriveTangent=-4.9385, interpMode=RCIM_Cubic, leaveTangent=-4.9385, tangentMode=RCTM_User, time=1, value=0}
       ]}}}
   )
   &ScaleColor(Scale Alpha=&Lerp_Float(A=0.6, B=1, Alpha=$System.DistanceLerp), ScaleA=true, ScaleRGB=true)
  ]
  rendererClasses [&NiagaraSpriteRendererProperties]
  renderers [{rendererClass=&NiagaraSpriteRendererProperties}]
 }
 Embers001 {
  simTarget=GPUComputeSim
  emitterUpdateScript [
   &EmitterState(Life Cycle Mode=::NewEnumerator1, Loop Behavior=::NewEnumerator1, Loop Duration=1, UseLoopDelay=false)
   &SpawnRate(
    SpawnRate=&FloatFromCurve(
     FloatCurve=DataInterface() {propertyValues {Curve {keys [<at 0 1 RCIM_Cubic>, <at 0.5369 0.05>, {tangentMode=RCTM_User, time=1, value=0}]}}}
     CurveIndex=$Emitter.NormalizedLoopAge
     Scale Curve=1000
    )
    Use Spawn Probability=false
   ) {module=SpawnRate001}
  ]
  particleSpawnScript [
   /SetVariables(Particles.SpriteRotationRate=&RandomRangeFloat(Minimum=-1, Maximum=1, Evaluation Type=::NewEnumerator1)) {isSetParameters=true, module=SetVariables_#4}
   &InitializeParticle(
    Lifetime Max=1.3
    Lifetime Min=0.8
    Lifetime Mode=::NewEnumerator1
    Sprite Size Mode=::NewEnumerator1
    Uniform Sprite Size Max=15
    Uniform Sprite Size Min=3
    UsePositionOffset=false
    Write Material Random=true
   ) {module=InitializeParticle001}
   &Color(Color=&MakeLinearColorFromVectorAndFloat("Vector (RGB)"=<2.6656 15.8408 16>, "Float (Alpha)"=&ClampFloat(Float=1)))
   &SphereLocation(
    Hemisphere Z=true
    MaskSpawnGroup=false
    Module.Override Randomness=false
    Sphere Radius=&RandomRangeFloat(Minimum=50, Maximum=1000, Evaluation Type=::NewEnumerator1)
    Surface Only=true
   )
  ]
  particleUpdateScript [
   &ParticleState()
   &CurlNoiseForce(Noise Strength=10000, Pan Noise=false)
   &AccelerationForce(Acceleration=&MakeVector(X=0, Y=0, Z=&RandomRangeFloat(Minimum=0, Maximum=1500)))
   &GravityForce(Gravity=<0 0 -1080>)
   &PointAttractionForce(
    Attraction Radius=3000
    Attraction Strength=&FloatFromCurve(
     FloatCurve=DataInterface() {propertyValues {Curve {keys [
         {arriveTangent=4.1785, interpMode=RCIM_Cubic, leaveTangent=4.1785, tangentMode=RCTM_User, time=0, value=0}
         {arriveTangent=0.9995, interpMode=RCIM_Cubic, leaveTangent=0.9995, tangentMode=RCTM_User, time=0.2223, value=0.6825}
         {arriveTangent=-0.0405, interpMode=RCIM_Cubic, leaveTangent=-0.0405, tangentMode=RCTM_User, time=1, value=1}
        ]}}}
     Scale Curve=500
    )
    Kill Within Radius=false
    Module.Use Falloff=true
    UseAttractorOffset=true
   )
   &Drag(Drag=4, Use Linear Drag=true, Use Rotational Drag=false)
   &SolveForcesAndVelocity(Clamp Velocity=false, Limit Acceleration=false)
   &ScaleSpriteSize(Uniform Curve Sprite Scale=DataInterface() {propertyValues {Curve {keys [
        {interpMode=RCIM_Cubic, tangentMode=RCTM_User, time=0, value=0}
        <at 0.1 2.025>
        {arriveTangent=-6.2277, interpMode=RCIM_Cubic, leaveTangent=-6.2277, tangentMode=RCTM_User, time=0.446, value=1.4417}
        {tangentMode=RCTM_User, time=1, value=0}
       ]}}})
   &ScaleColor(Scale Alpha=&Lerp_Float(A=0.3, B=1, Alpha=$System.DistanceLerp), ScaleA=true, ScaleRGB=true)
  ]
  rendererClasses [&NiagaraSpriteRendererProperties]
  renderers [{rendererClass=&NiagaraSpriteRendererProperties}]
 }
 Empty {
  emitterUpdateScript [
   &EmitterState(
    Life Cycle Mode=::NewEnumerator1
    Loop Behavior=::NewEnumerator1
    Loop Delay=&Add_Float(A=$System.FinalFlashDelay, B=0.6)
    Loop Duration=1.5
    UseLoopDelay=true
   )
   &SpawnRate(
    SpawnRate=&FloatFromCurve(
     FloatCurve=DataInterface() {propertyValues {Curve {keys [<at 0 1 RCIM_Cubic>, <at 0.3 0.5>, {tangentMode=RCTM_User, time=1, value=0}]}}}
     CurveIndex=$Emitter.NormalizedLoopAge
     Scale Curve=50
    )
    Use Spawn Probability=false
   ) {module=SpawnRate001}
  ]
  particleSpawnScript [
   &InitializeParticle(
    Color=&LerpLinearColors(StartColor=<rgba 1 0.2862 0.0242 1>, EndColor=<rgba 0 0.4739 1 1>, LerpFactor=&UniformAOrBFloat(A=0))
    Lifetime Max=0.4
    Lifetime Min=0.2
    Lifetime Mode=::NewEnumerator1
    Sprite Rotation Mode=::NewEnumerator2
    Sprite Size Mode=::NewEnumerator1
    Uniform Sprite Size Max=100
    Uniform Sprite Size Min=50
    UsePositionOffset=false
    Write Material Random=true
   )
   &SystemLocation(MaskSpawnGroup=false, Offset=<0 0 -50>)
   &ShapeLocation(
    Cylinder Height=50
    Cylinder Radius=200
    MaskSpawnGroup=false
    Override Seed=false
    Shape Primitive=::NewEnumerator1
    Surface Only=false
    UseOwnerScale=true
   )
  ]
  particleUpdateScript [
   &ParticleState()
   &ScaleSpriteSize(
    Uniform Curve Scale=10
    Uniform Curve Sprite Scale=DataInterface() {propertyValues {Curve {keys [
        {interpMode=RCIM_Cubic, tangentMode=RCTM_User, time=0, value=0}
        {interpMode=RCIM_Cubic, tangentMode=RCTM_User, time=0.2, value=0.95}
        {tangentMode=RCTM_User, time=1, value=0}
       ]}}}
   )
  ]
  rendererClasses [&NiagaraSpriteRendererProperties]
  renderers [{rendererClass=&NiagaraSpriteRendererProperties}]
 }
 FinalFlash {
  emitterUpdateScript [
   &EmitterState(Life Cycle Mode=::NewEnumerator1, Loop Behavior=::NewEnumerator1, Loop Duration Mode=::NewEnumerator1, UseLoopDelay=false)
   &SpawnBurst_Instantaneous(Spawn Time=$System.FinalFlashDelay, Use Spawn Probability=false, UseLoopCountLimit=false)
  ]
  particleSpawnScript [
   &InitializeParticle(
    Color=<rgba 1 0.4322 0.0849 1>
    Lifetime=0.5
    Sprite Size=<600 1500>
    Sprite Size Mode=::NewEnumerator4
    UsePositionOffset=false
    Write Material Random=true
   )
   /SetVariables(Particles.GlowSpriteAlignment:Vector3f=<0 0 1>) {isSetParameters=true, module=SetVariables_#5}
  ]
  particleUpdateScript [
   /SetVariables(Particles.LightExponent=3) {isSetParameters=true, module=SetVariables_#6}
   /SetVariables(Particles.LightRadius=3) {isSetParameters=true, module=SetVariables_#7}
   /SetVariables(Particles.LightColorv2:LinearColor=&MultiplyLinearColor(
     Linear Color=<rgba 10 3.7104 0.8755 1>
     Scale Factor=&MakeVector4_FromFloat(Float=&FloatFromCurve(
       FloatCurve=DataInterface() {propertyValues {Curve {keys [
           {arriveTangent=-2.1707, interpMode=RCIM_Cubic, leaveTangent=-2.1707, tangentMode=RCTM_User, time=0, value=1}
           {arriveTangent=-0.747, interpMode=RCIM_Cubic, leaveTangent=-0.747, tangentMode=RCTM_User, time=0.3, value=0.2}
           {tangentMode=RCTM_User, time=1, value=0}
          ]}}}
       Scale Curve=10
      ))
    )) {isSetParameters=true, module=SetVariables_#8}
   &ParticleState()
   &ScaleColor(
    Scale Alpha=&Multiply_Float(
     A=$System.DistanceLerp
     B=&FloatFromCurve(FloatCurve=DataInterface() {propertyValues {Curve {keys [<at 0 1 RCIM_Cubic>, {tangentMode=RCTM_User, time=1, value=0}]}}})
    )
    Scale RGB=&VectorFromFloat(Value=&FloatFromCurve(
      FloatCurve=DataInterface() {propertyValues {Curve {keys [
          {arriveTangent=-2.1707, interpMode=RCIM_Cubic, leaveTangent=-2.1707, tangentMode=RCTM_User, time=0, value=1}
          {tangentMode=RCTM_User, time=1, value=0}
         ]}}}
      Scale Curve=10
     ))
    ScaleA=true
    ScaleRGB=true
   )
   &ScaleSpriteSize(
    Non-Uniform Curve Scale=<1.5 2>
    Non-Uniform Curve Sprite Scale=DataInterface() {propertyValues {
      XCurve {keys [{arriveTangent=-1.1225, interpMode=RCIM_Cubic, leaveTangent=-1.1225, tangentMode=RCTM_User, time=0, value=1}, <at 1 0.25 RCIM_Cubic>]}
      YCurve {keys [{arriveTangent=1.8991, interpMode=RCIM_Cubic, leaveTangent=1.8991, tangentMode=RCTM_User, time=0, value=0.4}, <at 1 1 RCIM_Cubic>]}
     }}
    Scale Sprite Size Mode=::NewEnumerator3
   )
  ]
  rendererClasses [&NiagaraSpriteRendererProperties, &NiagaraLightRendererProperties]
  renderers [
   {rendererClass=&NiagaraSpriteRendererProperties}
   {rendererClass=&NiagaraLightRendererProperties, rendererIndex=1}
   {rendererClass=&NiagaraSpriteRendererProperties, rendererIndex=2}
  ]
 }
 IntroFlash {
  emitterUpdateScript [
   &EmitterState(Life Cycle Mode=::NewEnumerator1, Loop Behavior=::NewEnumerator1, Loop Duration Mode=::NewEnumerator1, UseLoopDelay=false)
   &SpawnBurst_Instantaneous(Use Spawn Probability=false, UseLoopCountLimit=false)
  ]
  particleSpawnScript [
   &InitializeParticle(Lifetime=0.8, Sprite Size=<600 800>, Sprite Size Mode=::NewEnumerator4, UsePositionOffset=false, Write Material Random=true)
   /SetVariables(Particles.GlowSpriteAlignment:Vector3f=<0 0 1>) {isSetParameters=true, module=SetVariables_#5}
   &Color(Color=&MakeLinearColorFromVectorAndFloat("Vector (RGB)"=<0.2027 0.7676 1>, "Float (Alpha)"=&ClampFloat(Float=1)))
  ]
  particleUpdateScript [
   /SetVariables(Particles.LightExponent=3) {isSetParameters=true, module=SetVariables_#6}
   /SetVariables(Particles.LightRadius=&Lerp_Float(
     A=0
     B=&RandomRangeFloat(Minimum=1, Maximum=1.5, Evaluation Type=::NewEnumerator1)
     Alpha=&FloatFromCurve(
      FloatCurve=DataInterface() {propertyValues {Curve {keys [
          {arriveTangent=9.0725, interpMode=RCIM_Cubic, leaveTangent=9.0725, tangentMode=RCTM_User, time=0, value=0}
          {arriveTangent=-2.5249, interpMode=RCIM_Cubic, leaveTangent=-2.5249, tangentMode=RCTM_User, time=0.1408, value=0.9847}
          {arriveTangent=-0.0405, interpMode=RCIM_Cubic, leaveTangent=-0.0405, tangentMode=RCTM_User, time=1, value=0}
         ]}}}
      Scale Curve=&Lerp_Float(A=0.5, B=6, Alpha=$System.DistanceLerp)
     )
    )) {isSetParameters=true, module=SetVariables_#7}
   /SetVariables(Particles.LightColorv2:LinearColor=&MultiplyLinearColor(
     Linear Color=<rgba 0.5092 2.0174 3 1>
     Scale Factor=&MakeVector4_FromFloat(Float=&FloatFromCurve(
       FloatCurve=DataInterface() {propertyValues {Curve {keys [
           {tangentMode=RCTM_User, time=0, value=1}
           {arriveTangent=-0.0405, interpMode=RCIM_Cubic, leaveTangent=-0.0405, tangentMode=RCTM_User, time=0.7009, value=1}
           {arriveTangent=-8.0089, interpMode=RCIM_Cubic, leaveTangent=-8.0089, tangentMode=RCTM_User, time=1, value=0}
          ]}}}
       Scale Curve=10
      ))
    )) {isSetParameters=true, module=SetVariables_#8}
   &ParticleState()
   &ScaleColor(
    Scale Alpha=&Multiply_Float(
     A=$System.DistanceLerp
     B=&FloatFromCurve(FloatCurve=DataInterface() {propertyValues {Curve {keys [
          {interpMode=RCIM_Cubic, tangentMode=RCTM_User, time=0, value=0}
          {arriveTangent=-1.8051, interpMode=RCIM_Cubic, leaveTangent=-1.8051, tangentMode=RCTM_User, time=0.1801, value=0.9667}
          {tangentMode=RCTM_User, time=1, value=0}
         ]}}})
    )
    Scale RGB=&VectorFromFloat(Value=&Lerp_Float(
      A=0
      B=&RandomRangeFloat(Minimum=0, Maximum=1, Evaluation Type=::NewEnumerator1)
      Alpha=&FloatFromCurve(
       FloatCurve=DataInterface() {propertyValues {Curve {keys [
           {arriveTangent=9.0725, interpMode=RCIM_Cubic, leaveTangent=9.0725, tangentMode=RCTM_User, time=0, value=0}
           {arriveTangent=-2.5249, interpMode=RCIM_Cubic, leaveTangent=-2.5249, tangentMode=RCTM_User, time=0.1408, value=0.9847}
           {arriveTangent=-0.0405, interpMode=RCIM_Cubic, leaveTangent=-0.0405, tangentMode=RCTM_User, time=1, value=0}
          ]}}}
       Scale Curve=0.7
      )
     ))
    ScaleA=true
    ScaleRGB=true
   )
   &ScaleSpriteSize(
    Uniform Curve Scale=&Lerp_Float(A=0.2, B=10, Alpha=$System.DistanceLerp)
    Uniform Curve Sprite Scale=DataInterface() {propertyValues {Curve {keys [{tangentMode=RCTM_User, time=0, value=1}, {tangentMode=RCTM_User, time=1, value=0}]}}}
   )
  ]
  rendererClasses [&NiagaraSpriteRendererProperties, &NiagaraLightRendererProperties]
  renderers [
   {rendererClass=&NiagaraSpriteRendererProperties}
   {rendererClass=&NiagaraLightRendererProperties, rendererIndex=1}
   {rendererClass=&NiagaraSpriteRendererProperties, rendererIndex=2}
  ]
 }
 RainbowRing {
  emitterUpdateScript [
   &EmitterState(Life Cycle Mode=::NewEnumerator1, Loop Behavior=::NewEnumerator1, Loop Duration=0.95, UseLoopDelay=true)
   &SpawnBurst_Instantaneous(Use Spawn Probability=false, UseLoopCountLimit=false)
  ]
  particleSpawnScript [/Niagara/Modules/Spawn/Initialization/InitializeParticle(
    Lifetime=&Multiply_Float(
     A=0.6
     B=&FloatFromCurve(
      FloatCurve=DataInterface() {propertyValues {Curve {keys [{arriveTangent=-1.5125, interpMode=RCIM_Cubic, leaveTangent=-1.5125, tangentMode=RCTM_User, time=0, value=1}, <at 0.9949 0.2248>]}}}
      CurveIndex=$Emitter.NormalizedLoopAge
     )
    )
    Module.Write Color=true
    Module.Write Lifetime=true
    Module.Write Mass=true
    Module.Write Material Random=true
    Module.Write Position=false
    Module.Write Scale=false
    Module.Write SpriteRotation=true
    Module.Write SpriteSize=true
    Sprite Rotation=&RandomRangeFloat(Minimum=0, Maximum=360)
    Sprite Size=<3000 3000>
   )]
  particleUpdateScript [
   &ParticleState()
   &FN_Sprite_Attributes(
    Camera Offset=50
    Color=&MakeLinearColorFromVectorAndFloat(
     "Vector (RGB)"=<0.8602 1.4152 2>
     "Float (Alpha)"=&FloatFromCurve(
      FloatCurve=DataInterface() {propertyValues {Curve {keys [
          {interpMode=RCIM_Cubic, tangentMode=RCTM_User, time=0, value=0}
          {interpMode=RCIM_Cubic, tangentMode=RCTM_User, time=0.5, value=1}
          {tangentMode=RCTM_User, time=1, value=0}
         ]}}}
      Scale Curve=&FloatFromCurve(
       FloatCurve=DataInterface() {propertyValues {Curve {keys [<at 0.0654 -0.0428>,<at 0.5237 0.7921 RCIM_Cubic>,<at 0.9993 -0.0428>]}}}
       CurveIndex=$Emitter.NormalizedLoopAge
       Scale Curve=1.5
      )
     )
    )
    Create Camera Offset Attribute=true
    Create Color Attribute=true
    Create SpriteSize Attribute=true
    SpriteSize=&Vector2DFromFloat(Value=&Multiply_Float(
      A=200
      B=&FloatFromCurve(
       FloatCurve=DataInterface() {propertyValues {Curve {keys [{interpMode=RCIM_Cubic, tangentMode=RCTM_User, time=0, value=1}, {interpMode=RCIM_Cubic, tangentMode=RCTM_User, time=1, value=0.2}]}}}
       Scale Curve=10
      )
     ))
    Write to Module Namespace=true
   ) {module=FN_Sprite_Attributes002}
   &DynamicMaterialParameters(
    Index 0 Param 1=&FloatFromCurve(FloatCurve=DataInterface() {propertyValues {Curve {keys [
         {arriveTangent=3.2984, interpMode=RCIM_Cubic, leaveTangent=3.2984, tangentMode=RCTM_User, time=0, value=0}
         {interpMode=RCIM_Cubic, tangentMode=RCTM_User, time=1, value=1}
        ]}}})
    Index 0 Param 2=1
    Index 0 Param 3=1
    Index 0 Param 4=1
    Module.Param0WriteEnabled=true
    Module.Param1WriteEnabled=true
    Module.Param2WriteEnabled=true
    Module.Param3WriteEnabled=true
   )
  ]
  rendererClasses [&NiagaraSpriteRendererProperties]
  renderers [{rendererClass=&NiagaraSpriteRendererProperties}]
 }
 ShellLoop {
  emitterSpawnScript [/SetVariables(Emitter.Refraction=1.1) {isSetParameters=true}]
  emitterUpdateScript [
   &EmitterState(
    Inactive Response=::NewEnumerator1
    Life Cycle Mode=::NewEnumerator1
    Loop Behavior=::NewEnumerator1
    Loop Duration Mode=::NewEnumerator1
    UseLoopDelay=false
   )
   &SpawnBurst_Instantaneous(Spawn Time=$System.FinalFlashDelay, Use Spawn Probability=false, UseLoopCountLimit=false)
  ]
  particleSpawnScript [
   &InitialMeshOrientation(Mesh Axis To Orient=<0 0 1>, Mesh Orientation Mode=::NewEnumerator2, Use Rotation Vector=false)
   /Niagara/Modules/Spawn/Initialization/V2/InitializeParticle(Color=<rgba 0.3695 0.9202 1 1>, Lifetime=0.5, Mesh Scale Mode=::NewEnumerator3, UsePositionOffset=false, Write Material Random=true)
  ]
  particleUpdateScript [
   &ParticleState()
   &ScaleMeshSize(Scale Factor=&VectorFromCurve(
     VectorCurve=DataInterface() {propertyValues {
       XCurve {keys [{arriveTangent=1.8564, interpMode=RCIM_Cubic, leaveTangent=1.8564, tangentMode=RCTM_User, time=0, value=0}, <at 1 1 RCIM_Cubic>]}
       YCurve {keys [<at 0 1 RCIM_Cubic>, {arriveTangent=-0.394, interpMode=RCIM_Cubic, leaveTangent=-0.394, tangentMode=RCTM_User, time=1, value=0}]}
       ZCurve {keys [<at 0 1 RCIM_Cubic>, {arriveTangent=-0.461, interpMode=RCIM_Cubic, leaveTangent=-0.461, tangentMode=RCTM_User, time=1, value=0}]}
      }}
     Scale Curve=<3 1 1>
    ))
   &DynamicMaterialParameters(
    Index 0 Param 1=1.8
    Index 0 Param 2=0.2
    Index 0 Param 3=1
    Index 0 Param 4=1
    Module.Param0WriteEnabled=true
    Module.Param1WriteEnabled=true
    Module.Param2WriteEnabled=true
    Module.Param3WriteEnabled=true
   )
   &ScaleColor(Scale Alpha=$System.DistanceLerp, ScaleA=true, ScaleRGB=true)
  ]
  rendererClasses [&NiagaraMeshRendererProperties]
  renderers [{rendererClass=&NiagaraMeshRendererProperties, rendererIndex=0}]
 }
 Shimmer001 {
  emitterUpdateScript [
   &EmitterState(Life Cycle Mode=::NewEnumerator1, Loop Behavior=::NewEnumerator1, Loop Duration=4, UseLoopDelay=false)
   &SpawnBurst_Instantaneous(Spawn Count=12, Use Spawn Probability=false, UseLoopCountLimit=false)
  ]
  particleSpawnScript [
   &InitializeParticle(
    Color=<rgba 0.3569 0.733 1 0.6>
    Lifetime Max=1
    Lifetime Mode=::NewEnumerator1
    Sprite Size Max=<100 800>
    Sprite Size Min=<50 700>
    Sprite Size Mode=::NewEnumerator2
    UsePositionOffset=false
    Write Material Random=true
   )
   &SystemLocation(MaskSpawnGroup=false, Offset=<0 0 250>)
   &ShapeLocation(
    MaskSpawnGroup=false
    Override Seed=false
    Sphere Radius=&RandomRangeFloat(Minimum=200, Maximum=250)
    UseOwnerScale=true
   ) {enabled=false}
   &SphereLocation(
    MaskSpawnGroup=false
    Module.Override Randomness=false
    Sphere Radius=&RandomRangeFloat(Minimum=200, Maximum=800, Evaluation Type=::NewEnumerator1)
    Surface Only=true
   )
   &AddVelocity(
    Module.Use Falloff Distance=false
    Override Seed=false
    Velocity Mode=::NewEnumerator1
    Velocity Speed=&RandomRangeFloat(Minimum=25, Maximum=50)
   )
  ]
  particleUpdateScript [
   &ParticleState()
   &SolveForcesAndVelocity(Clamp Velocity=false, Limit Acceleration=false) {enabled=false}
   &ScaleColor(
    Scale Alpha=&FloatFromCurve(FloatCurve=DataInterface() {propertyValues {Curve {keys [
         {interpMode=RCIM_Cubic, tangentMode=RCTM_User, time=0.1, value=0}
         {interpMode=RCIM_Cubic, tangentMode=RCTM_User, time=0.3, value=1}
         {tangentMode=RCTM_User, time=1, value=0}
        ]}}})
    ScaleA=true
    ScaleRGB=true
   )
   &PointAttractionForce(Attraction Radius=2000, Attraction Strength=50, Kill Within Radius=false, Module.Use Falloff=true, UseAttractorOffset=true)
   &SolveForcesAndVelocity(Clamp Velocity=false, Limit Acceleration=false) {module=SolveForcesAndVelocity001}
   &DynamicMaterialParameters(
    Index 0 Param 1=1
    Index 0 Param 2=1
    Index 0 Param 3=0.5
    Index 0 Param 4=1
    Module.Param0WriteEnabled=true
    Module.Param1WriteEnabled=true
    Module.Param2WriteEnabled=true
    Module.Param3WriteEnabled=true
   )
   &ScaleSpriteSize(
    Non-Uniform Curve Scale=<2 2>
    Non-Uniform Curve Sprite Scale=DataInterface() {propertyValues {
      XCurve {keys [
        {arriveTangent=3.497, interpMode=RCIM_Cubic, leaveTangent=3.497, tangentMode=RCTM_User, time=0, value=0}
        {tangentMode=RCTM_User, time=1, value=1}
       ]}
      YCurve {keys [{tangentMode=RCTM_User, time=0, value=0}, {tangentMode=RCTM_User, time=1, value=1}]}
     }}
    Scale Sprite Size Mode=::NewEnumerator3
   )
  ]
  rendererClasses [&NiagaraSpriteRendererProperties]
  renderers [{rendererClass=&NiagaraSpriteRendererProperties}]
 }
 Sparklines {
  simTarget=GPUComputeSim
  emitterUpdateScript [
   &EmitterState(Life Cycle Mode=::NewEnumerator1, Loop Behavior=::NewEnumerator1, Loop Duration=0.5, UseLoopDelay=false)
   &SpawnRate(
    SpawnRate=&FloatFromCurve(
     FloatCurve=DataInterface() {propertyValues {Curve {keys [<at 0 1 RCIM_Cubic>, <at 0.5369 0.05>, {tangentMode=RCTM_User, time=1, value=0}]}}}
     CurveIndex=$Emitter.NormalizedLoopAge
     Scale Curve=500
    )
    Use Spawn Probability=false
   ) {module=SpawnRate001}
  ]
  particleSpawnScript [
   /SetVariables(Particles.SpriteRotationRate=&RandomRangeFloat(Minimum=-1, Maximum=1, Evaluation Type=::NewEnumerator1)) {isSetParameters=true, module=SetVariables_#4}
   &InitializeParticle(
    Color=<rgba 0.2825 0.8901 1 1>
    Lifetime Max=0.8
    Lifetime Min=0.5
    Lifetime Mode=::NewEnumerator1
    Sprite Size Max=<5 500>
    Sprite Size Min=<3 100>
    Sprite Size Mode=::NewEnumerator2
    UsePositionOffset=false
    Write Material Random=true
   ) {module=InitializeParticle001}
   &SphereLocation(
    Hemisphere Z=true
    MaskSpawnGroup=false
    Module.Override Randomness=false
    Sphere Radius=&RandomRangeFloat(Minimum=500, Maximum=4000, Evaluation Type=::NewEnumerator1)
    Surface Only=true
   )
  ]
  particleUpdateScript [
   &ParticleState()
   &CurlNoiseForce(Noise Strength=10000, Pan Noise=false) {enabled=false}
   &AccelerationForce(Acceleration=&MakeVector(X=0, Y=0, Z=&RandomRangeFloat(Minimum=0, Maximum=1500))) {enabled=false}
   &GravityForce(Gravity=<0 0 -1080>) {enabled=false}
   &PointAttractionForce(
    Attraction Radius=12000
    Attraction Strength=&FloatFromCurve(
     FloatCurve=DataInterface() {propertyValues {Curve {keys [
         {arriveTangent=4.1785, interpMode=RCIM_Cubic, leaveTangent=4.1785, tangentMode=RCTM_User, time=0, value=0}
         {arriveTangent=0.9995, interpMode=RCIM_Cubic, leaveTangent=0.9995, tangentMode=RCTM_User, time=0.2223, value=0.6825}
         {arriveTangent=-0.0405, interpMode=RCIM_Cubic, leaveTangent=-0.0405, tangentMode=RCTM_User, time=1, value=1}
        ]}}}
     Scale Curve=500
    )
    Kill Within Radius=false
    Module.Use Falloff=true
    UseAttractorOffset=true
   )
   &Drag(Drag=4, Use Linear Drag=true, Use Rotational Drag=false)
   &SolveForcesAndVelocity(Clamp Velocity=false, Limit Acceleration=false)
   &ScaleSpriteSize(Uniform Curve Sprite Scale=DataInterface() {propertyValues {Curve {keys [
        {interpMode=RCIM_Cubic, tangentMode=RCTM_User, time=0, value=0}
        <at 0.1 2.025>
        {arriveTangent=-6.2277, interpMode=RCIM_Cubic, leaveTangent=-6.2277, tangentMode=RCTM_User, time=0.446, value=1.4417}
        {tangentMode=RCTM_User, time=1, value=0}
       ]}}})
   &ScaleColor(Scale Alpha=&Lerp_Float(A=0.6, B=1, Alpha=$System.DistanceLerp), ScaleA=true, ScaleRGB=true)
  ]
  rendererClasses [&NiagaraSpriteRendererProperties]
  renderers [{rendererClass=&NiagaraSpriteRendererProperties}]
 }
 ZeroPointFlare {
  emitterUpdateScript [
   &EmitterState(Life Cycle Mode=::NewEnumerator1, Loop Behavior=::NewEnumerator1, Loop Delay=$System.FinalFlashDelay, Loop Duration=0.95, UseLoopDelay=true)
   &SpawnRate(SpawnRate=7, Use Spawn Probability=false)
  ]
  particleSpawnScript [
   /Niagara/Modules/Spawn/Initialization/InitializeParticle(
    Lifetime=&Multiply_Float(
     A=&RandomRangeFloat(Minimum=0.5, Maximum=0.8)
     B=&FloatFromCurve(
      FloatCurve=DataInterface() {propertyValues {Curve {keys [{arriveTangent=-1.5125, interpMode=RCIM_Cubic, leaveTangent=-1.5125, tangentMode=RCTM_User, time=0, value=1}, <at 0.9949 0.2248>]}}}
      CurveIndex=$Emitter.NormalizedLoopAge
     )
    )
    Module.Write Color=true
    Module.Write Lifetime=true
    Module.Write Mass=true
    Module.Write Material Random=true
    Module.Write Position=false
    Module.Write Scale=false
    Module.Write SpriteRotation=true
    Module.Write SpriteSize=true
    Sprite Rotation=&RandomRangeFloat(Minimum=0, Maximum=360)
    Sprite Size=<3000 3000>
   )
   &Color(Color=&MakeLinearColorFromVectorAndFloat("Vector (RGB)"=<0.0175 0.0243 0.03>, "Float (Alpha)"=&ClampFloat(Float=0.6)))
   &SystemLocation(MaskSpawnGroup=false, Offset=<0 0 30>)
  ]
  particleUpdateScript [
   &ParticleState()
   &ScaleColor(
    Scale Alpha=&FloatFromCurve(FloatCurve=DataInterface() {propertyValues {Curve {keys [<at 0 1 RCIM_Cubic>,<at 0.25 1 RCIM_Cubic>,<at 0.75 1 RCIM_Cubic>,<at 1 1>]}}})
    ScaleA=true
    ScaleRGB=true
   )
   /SetVariables(Particles.Size:NiagaraFloat=&FloatFromCurve(
     FloatCurve=DataInterface() {propertyValues {Curve {keys [
         <at -0.0039 0.0032 RCIM_Cubic>
         {arriveTangent=0.0358, interpMode=RCIM_Cubic, leaveTangent=0.0358, tangentMode=RCTM_User, time=0.4967, value=0.9107}
         <at 1 0>
        ]}}}
     Scale Curve=&FloatFromCurve(
      FloatCurve=DataInterface() {propertyValues {Curve {keys [
          <at -0.0005 0.984 RCIM_Cubic>
          {arriveTangent=0.3225, interpMode=RCIM_Cubic, leaveTangent=0.3225, tangentMode=RCTM_User, time=0.4833, value=3.7073}
          <at 1.0108 1.1471>
         ]}}}
      CurveIndex=$Emitter.NormalizedLoopAge
      Scale Curve=0.5
     )
    )) {isSetParameters=true}
   &FN_Sprite_Attributes(
    Camera Offset=-250
    Create Camera Offset Attribute=true
    Create Color Attribute=true
    Create SpriteSize Attribute=true
    SpriteSize=&Vector2DFromFloat(Value=&Multiply_Float(A=500, B=$Particles.Size))
    Write to Module Namespace=true
   )
   &FN_Sprite_Attributes(
    Camera Offset=75
    Color=&MakeLinearColorFromVectorAndFloat("Vector (RGB)"=<0.5625 1.8964 3>, "Float (Alpha)"=0.1)
    Create Camera Offset Attribute=true
    Create Color Attribute=true
    Create SpriteSize Attribute=true
    SpriteSize=&Vector2DFromFloat(Value=&Multiply_Float(A=500, B=$Particles.Size))
    Write to Module Namespace=true
   ) {module=FN_Sprite_Attributes001}
   &FN_Sprite_Attributes(
    Camera Offset=50
    Color=&MakeLinearColorFromVectorAndFloat(
     "Vector (RGB)"=<0.8602 1.4152 2>
     "Float (Alpha)"=&FloatFromCurve(
      FloatCurve=DataInterface() {propertyValues {Curve {keys [
          <at 0.0048 -0.0024 RCIM_Cubic>
          {arriveTangent=2.9233, interpMode=RCIM_Cubic, leaveTangent=2.9233, tangentMode=RCTM_User, time=0.7236, value=0.9786}
          <at 0.9988 -0.0024>
         ]}}}
      Scale Curve=&FloatFromCurve(
       FloatCurve=DataInterface() {propertyValues {Curve {keys [<at 0.0654 -0.0428>,<at 0.5237 0.7921 RCIM_Cubic>,<at 0.9993 -0.0428>]}}}
       CurveIndex=$Emitter.NormalizedLoopAge
       Scale Curve=0.3
      )
     )
    )
    Create Camera Offset Attribute=true
    Create Color Attribute=true
    Create SpriteSize Attribute=true
    SpriteSize=&Vector2DFromFloat(Value=&Multiply_Float(
      A=200
      B=&FloatFromCurve(
       FloatCurve=DataInterface() {propertyValues {Curve {keys [
           <at 0.004 1.0518 RCIM_Cubic>
           {arriveTangent=-0.8729, interpMode=RCIM_Cubic, leaveTangent=-0.8729, tangentMode=RCTM_User, time=0.9881, value=0.7637}
          ]}}}
       Scale Curve=2
      )
     ))
    Write to Module Namespace=true
   ) {module=FN_Sprite_Attributes002}
   &FN_Sprite_Attributes(
    Camera Offset=10
    Color=&MakeLinearColorFromVectorAndFloat(
     "Vector (RGB)"=&Multiply_VectorByFloat(Vector=<1.5665 2.3369 6>, Float=0.5)
     "Float (Alpha)"=&FloatFromCurve(
      FloatCurve=DataInterface() {propertyValues {Curve {keys [
          <at 0.0048 -0.0024 RCIM_Cubic>
          {arriveTangent=2.9233, interpMode=RCIM_Cubic, leaveTangent=2.9233, tangentMode=RCTM_User, time=0.7236, value=0.9786}
          <at 0.9988 -0.0024>
         ]}}}
      Scale Curve=0.5
     )
    )
    Create Camera Offset Attribute=true
    Create Color Attribute=true
    Create Sprite Rotation Attribute=true
    Create SpriteSize Attribute=true
    SpriteSize=&Multiply_Vector2(
     A=<1000 20>
     B=&Vector2DFromFloat(Value=&RandomRangeFloat(Minimum=0.2, Maximum=1))
    )
    Write to Module Namespace=true
   ) {module=FN_Sprite_Attributes003}
   &FN_Sprite_Attributes(
    Camera Offset=45
    Color=&MakeLinearColorFromVectorAndFloat("Vector (RGB)"=<0.2398 1.0892 1.5>, "Float (Alpha)"=&ClampFloat(Float=0.5))
    Create Camera Offset Attribute=true
    Create Color Attribute=true
    Create Sprite Rotation Attribute=true
    Create SpriteSize Attribute=true
    SpriteSize=&Multiply_Vector2(
     A=&RandomRangeVector2D(Minimum=<120 80>, Maximum=<200 140>, Evaluation Type=::NewEnumerator1)
     B=&Vector2DFromFloat(Value=&FloatFromCurve(
       FloatCurve=DataInterface() {propertyValues {Curve {keys [
           <at 0 0 RCIM_Cubic>
           {arriveTangent=1.4286, interpMode=RCIM_Cubic, leaveTangent=1.4286, time=0.5754, value=0.3493}
           {arriveTangent=-0.8228, interpMode=RCIM_Cubic, leaveTangent=-0.8228, time=0.7, value=1}
           <at 1 0>
          ]}}}
       Scale Curve=5
      ))
    )
    Write to Module Namespace=true
   ) {module=FN_Sprite_Attributes004}
   &FN_Light_Attributes(
    Color=<rgba 1 2 10 1>
    Create Color Attribute=true
    Create Position Attribute=true
    Enable Visibility Tag=true
    Position=&Add_Vector(
     A=&Add_Vector(
      A=Linked() {linkedVariable {name=Particles.Position, type=NiagaraPosition}}
      B=&RandomVector(Vector Scale=20, Evaluation Type=::NewEnumerator1)
     )
     B=&TransformVector(Vector=<0 0 50>, Source Space=::World, Destination Space=::Local)
    )
    Scale Color=&FloatFromCurve(
     FloatCurve=DataInterface() {propertyValues {Curve {keys [<at -0.0008 -0.006 RCIM_Cubic>, {arriveTangent=0.006, interpMode=RCIM_Cubic, leaveTangent=0.006, time=0.4887, value=0.9748}, <at 1 0>]}}}
     Scale Curve=10
    )
    VisibilityTag=HlslExpression() {hlslExpression="Particles.UniqueID && 1"}
   )
   &FN_Gen5_Light(Color:LinearColor=<rgba 1 2 10 1>, Intensity=1000, Position Offset:Vector3f=<0 0 400>)
  ]
  rendererClasses [&NiagaraSpriteRendererProperties]
  renderers [
   {rendererClass=&NiagaraSpriteRendererProperties}
   {rendererClass=&NiagaraSpriteRendererProperties, rendererIndex=1}
   {rendererClass=&NiagaraSpriteRendererProperties, rendererIndex=2}
   {rendererClass=&NiagaraSpriteRendererProperties, rendererIndex=3}
   {rendererClass=&NiagaraSpriteRendererProperties, rendererIndex=4}
  ]
 }
}
summary {
 systemName=NS_Encounter_ShutDown
 emitters [
  {emitterName=Embers, simTarget=GPUComputeSim, rendererClasses [&NiagaraSpriteRendererProperties]}
  {emitterName=FinalFlash, rendererClasses [&NiagaraSpriteRendererProperties, &NiagaraLightRendererProperties]}
  {emitterName=ShellLoop, rendererClasses [&NiagaraMeshRendererProperties]}
  {emitterName=IntroFlash, rendererClasses [&NiagaraSpriteRendererProperties, &NiagaraLightRendererProperties]}
  {emitterName=Embers001, simTarget=GPUComputeSim, rendererClasses [&NiagaraSpriteRendererProperties]}
  {emitterName=Empty, rendererClasses [&NiagaraSpriteRendererProperties]}
  {emitterName=Sparklines, simTarget=GPUComputeSim, rendererClasses [&NiagaraSpriteRendererProperties]}
  {emitterName=Shimmer001, rendererClasses [&NiagaraSpriteRendererProperties]}
  {emitterName=ZeroPointFlare, rendererClasses [&NiagaraSpriteRendererProperties]}
  {emitterName=CrackSpline, rendererClasses [&NiagaraMeshRendererProperties]}
  {emitterName=RainbowRing, rendererClasses [&NiagaraSpriteRendererProperties]}
 ]
 userVariables []
}
