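# Looping Niagara effect for an encounter spawn point: a cyan energy flare with cracks, glass shards, embers, shimmer and light.
# Shared system timing and flare height place the bursts, while intensity and flicker curves animate the light and flare.
# User.IsSpawnActiveState gates or blends active-state layers, and User.OverallVisibility fades key elements.
# System.DistanceLerp scales brightness, size and spawn probability by distance, with embers and glass simulated on the GPU.
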NiagaraSystem /Lager/Lager/Encounters/FX/NS_Encounter_SpawnIdleLoop
dependencies {
 usedDataInterfaces [
  &NiagaraDataInterfaceCurve
  &NiagaraDataInterfaceArrayFloat
  &NiagaraDataInterfaceVector2DCurve
  &NiagaraDataInterfaceVectorCurve
 ]
 usedDynamicInputs [&FN_LOD_Distance_Remap]
 usedModules [
  &self:SystemSpawnScript.SystemScriptSource.SystemScriptGraph.NiagaraNodeAssignment_1.SetVariables_#1
  &self:SystemSpawnScript.SystemScriptSource.SystemScriptGraph.NiagaraNodeAssignment_2.SetVariables_#2
  &self:SystemSpawnScript.SystemScriptSource.SystemScriptGraph.NiagaraNodeAssignment_8.SetVariables_#3
  &self:SystemSpawnScript.SystemScriptSource.SystemScriptGraph.NiagaraNodeAssignment_4.SetVariables_#4
  &self:SystemSpawnScript.SystemScriptSource.SystemScriptGraph.NiagaraNodeAssignment_7.SetVariables_#5
  &SystemState
  &self:SystemSpawnScript.SystemScriptSource.SystemScriptGraph.NiagaraNodeAssignment_3.SetVariables_#6
  &self:SystemSpawnScript.SystemScriptSource.SystemScriptGraph.NiagaraNodeAssignment_0.SetVariables_#7
 ]
}
emitters {
 CrackSpline {
  emitterUpdateScript [
   &EmitterState(Life Cycle Mode=::NewEnumerator1, Loop Behavior=::NewEnumerator1, Loop Duration=8, UseLoopDelay=true)
   &SpawnBurst_Instantaneous(
    Spawn Count=2
    Spawn Probability=&ClampFloat(Float=&Lerp_Float(A=-2, B=1, Alpha=$System.DistanceLerp))
    Spawn Time=&Multiply_Float(A=$System.ActiveSpawnTime, B=$System.SpawnTimeMul)
    Use Spawn Probability=true
   )
  ]
  particleSpawnScript [
   /SetVariables(Particles.MeshOrientation=&MakeQuatFromAxisAngle(
     Axis=&RandomRangeVector(Minimum=<0 0 0>, Maximum=<1 1 1>, Evaluation Type=::NewEnumerator1)
     AngleInDegrees=90
    )) {isSetParameters=true, module=SetVariables_#8}
   /Niagara/Modules/Spawn/Initialization/InitializeParticle(
    Lifetime=&RandomRangeFloat(Minimum=4, Maximum=6, Evaluation Type=::NewEnumerator1)
    Mesh Scale=&Multiply_VectorByFloat(Vector=&RandomRangeVector(Minimum=<2 2 3>, Maximum=<4 4 5>, Evaluation Type=::NewEnumerator1), Float=3)
    Module.Write Color=true
    Module.Write Lifetime=true
    Module.Write Mass=true
    Module.Write Material Random=true
    Module.Write Position=false
    Module.Write Scale=true
    Module.Write SpriteRotation=false
    Module.Write SpriteSize=false
    Position=<0 0 15>
   )
   &Color(Color=&MakeLinearColorFromVectorAndFloat("Vector (RGB)"=<1.5 1.5 1.5>, "Float (Alpha)"=&ClampFloat(Float=1)))
   &SystemLocation(MaskSpawnGroup=false, Offset=&MakeVector(X=0, Y=0, Z=$System.FlareHeight))
  ]
  particleUpdateScript [
   &ParticleState()
   &ScaleColor(
    Scale Alpha=&FloatFromCurve(FloatCurve=DataInterface() {propertyValues {Curve {keys [<at 0.77 1 RCIM_Cubic>,<at 1 0>]}}})
    Scale RGB=&VectorFromCurve(
     VectorCurve=DataInterface() {propertyValues {
       XCurve {keys [
         {arriveTangent=-6.9225, interpMode=RCIM_Cubic, time=0, value=4.5}
         {arriveTangent=-3.6158, interpMode=RCIM_Cubic, leaveTangent=-3.6158, time=0.65, value=1}
         <at 1 0.8842 RCIM_Cubic>
        ]}
       YCurve {keys [
         {arriveTangent=-15.8226, interpMode=RCIM_Cubic, time=0, value=6}
         {arriveTangent=-5.6986, interpMode=RCIM_Cubic, leaveTangent=-5.6986, time=0.65, value=1}
         {interpMode=RCIM_Cubic, leaveTangent=-5.843, time=1, value=0.3014}
        ]}
       ZCurve {keys [
         {arriveTangent=-15.2224, interpMode=RCIM_Cubic, time=0, value=6}
         {arriveTangent=-5, interpMode=RCIM_Cubic, leaveTangent=-5, time=0.65, value=1}
         <at 1 1 RCIM_Cubic>
        ]}
      }}
     Scale Curve=<0.12 0.12 0.12>
    )
    ScaleA=true
    ScaleRGB=true
   )
   &DynamicMaterialParameters(
    Index 0 Param 1=&FloatFromCurve(FloatCurve=DataInterface() {propertyValues {Curve {keys [
         <at 0 1 RCIM_Cubic>
         {arriveTangent=-1, interpMode=RCIM_Cubic, leaveTangent=-1, time=0.3329, value=0.616}
         {interpMode=RCIM_Cubic, leaveTangent=-0.3905, time=1, value=0}
        ]}}})
    Index 0 Param 2=&FloatFromCurve(
     FloatCurve=DataInterface() {propertyValues {Curve {keys [
         {arriveTangent=-2.0225, interpMode=RCIM_Cubic, time=0, value=0.6}
         {arriveTangent=-0.6, interpMode=RCIM_Cubic, leaveTangent=-0.6, time=0.65, value=0.25}
         <at 1 0 RCIM_Cubic>
        ]}}}
     Scale Curve=1.5
    )
    Index 0 Param 3=&FloatFromCurve(FloatCurve=DataInterface() {propertyValues {Curve {keys [{arriveTangent=1.0145, interpMode=RCIM_Cubic, time=0, value=-0.1}, <at 0.25 0>]}}})
    Index 0 Param 4=&FloatFromCurve(FloatCurve=DataInterface() {propertyValues {Curve {keys [<at 0 0 RCIM_Cubic>,<at 1 0>]}}})
    Module.Param0WriteEnabled=true
    Module.Param1WriteEnabled=true
    Module.Param2WriteEnabled=true
    Module.Param3WriteEnabled=true
   )
   &CameraOffset(Camera Offset Amount=-100)
  ]
  rendererClasses [&NiagaraMeshRendererProperties]
  renderers [{rendererClass=&NiagaraMeshRendererProperties, rendererIndex=0}]
 }
 CrackSpline001 {
  emitterUpdateScript [
   &EmitterState(Life Cycle Mode=::NewEnumerator1, Loop Behavior=::NewEnumerator1, Loop Duration Mode=::NewEnumerator1, UseLoopDelay=true)
   &SpawnBurst_Instantaneous(
    Spawn Count=1
    Spawn Time=&Multiply_Float(A=$System.ActiveSpawnTime, B=$System.SpawnTimeMul)
    Use Spawn Probability=false
   )
  ]
  particleSpawnScript [
   /SetVariables(Particles.MeshOrientation=&MakeQuatFromAxisAngle(
     Axis=&RandomRangeVector(Minimum=<0 0 0>, Maximum=<1 1 1>, Evaluation Type=::NewEnumerator1)
     AngleInDegrees=90
    )) {isSetParameters=true, module=SetVariables_#8}
   &InitializeParticle(
    Lifetime=&RandomRangeFloat(Minimum=4, Maximum=6, Evaluation Type=::NewEnumerator1)
    Mesh Scale=&Multiply_VectorByFloat(Vector=&RandomRangeVector(Minimum=<2 2 3>, Maximum=<4 4 5>, Evaluation Type=::NewEnumerator1), Float=2)
    Module.Write Color=true
    Module.Write Lifetime=true
    Module.Write Mass=true
    Module.Write Material Random=true
    Module.Write Position=false
    Module.Write Scale=true
    Module.Write SpriteRotation=false
    Module.Write SpriteSize=false
    Position=<0 0 15>
   )
   &Color(Color=&MakeLinearColorFromVectorAndFloat("Vector (RGB)"=<1.5 1.5 1.5>, "Float (Alpha)"=&ClampFloat(Float=1)))
   &SystemLocation(MaskSpawnGroup=false, Offset=&MakeVector(X=0, Y=0, Z=$System.FlareHeight))
  ]
  particleUpdateScript [
   &ParticleState(Kill Particles When Lifetime Has Elapsed=false, Let Infinitely Lived Particles Die When Emitter Deactivates=true)
   &ScaleColor(
    Scale Alpha=&FloatFromCurve(FloatCurve=DataInterface() {propertyValues {Curve {keys [<at 0.77 1 RCIM_Cubic>,<at 1 0>]}}}, CurveIndex=0)
    Scale RGB=&VectorFromCurve(
     VectorCurve=DataInterface() {propertyValues {
       XCurve {keys [
         {arriveTangent=-6.9225, interpMode=RCIM_Cubic, time=0, value=4.5}
         {arriveTangent=-3.6158, interpMode=RCIM_Cubic, leaveTangent=-3.6158, time=0.65, value=1}
         <at 1 0.8842 RCIM_Cubic>
        ]}
       YCurve {keys [
         {arriveTangent=-15.8226, interpMode=RCIM_Cubic, time=0, value=6}
         {arriveTangent=-5.6986, interpMode=RCIM_Cubic, leaveTangent=-5.6986, time=0.65, value=1}
         {interpMode=RCIM_Cubic, leaveTangent=-5.843, time=1, value=0.3014}
        ]}
       ZCurve {keys [
         {arriveTangent=-15.2224, interpMode=RCIM_Cubic, time=0, value=6}
         {arriveTangent=-5, interpMode=RCIM_Cubic, leaveTangent=-5, time=0.65, value=1}
         <at 1 1 RCIM_Cubic>
        ]}
      }}
     CurveIndex=0
     Scale Curve=<0.12 0.12 0.12>
    )
    ScaleA=true
    ScaleRGB=true
   )
   &DynamicMaterialParameters(
    Index 0 Param 1=&FloatFromCurve(
     FloatCurve=DataInterface() {propertyValues {Curve {keys [
         <at 0 1 RCIM_Cubic>
         {arriveTangent=-1, interpMode=RCIM_Cubic, leaveTangent=-1, time=0.3329, value=0.616}
         {interpMode=RCIM_Cubic, leaveTangent=-0.3905, time=1, value=0}
        ]}}}
     CurveIndex=&Lerp_Float(
      A=1
      B=&Add_Float(
       A=&OneMinusFloat(Float=$User.OverallVisibility)
       B=&FloatFromCurve(
        FloatCurve=DataInterface() {propertyValues {Curve {keys [
            {arriveTangent=0.1323, interpMode=RCIM_Cubic, leaveTangent=0.1323, tangentMode=RCTM_User, time=0, value=1}
            {arriveTangent=-3.83, interpMode=RCIM_Cubic, leaveTangent=-3.83, tangentMode=RCTM_User, time=1, value=0}
           ]}}}
        CurveIndex=$System.DistanceLerp
       )
      )
      Alpha=$User.IsSpawnActiveState
     )
    )
    Index 0 Param 2=&FloatFromCurve(
     FloatCurve=DataInterface() {propertyValues {Curve {keys [
         {arriveTangent=-2.0225, interpMode=RCIM_Cubic, time=0, value=0.6}
         {arriveTangent=-0.6, interpMode=RCIM_Cubic, leaveTangent=-0.6, time=0.65, value=0.25}
         <at 1 0 RCIM_Cubic>
        ]}}}
     CurveIndex=0
     Scale Curve=1.5
    )
    Index 0 Param 3=&FloatFromCurve(
     FloatCurve=DataInterface() {propertyValues {Curve {keys [{arriveTangent=1.0145, interpMode=RCIM_Cubic, time=0, value=-0.1}, <at 0.25 0>]}}}
     CurveIndex=0
    )
    Index 0 Param 4=&FloatFromCurve(FloatCurve=DataInterface() {propertyValues {Curve {keys [<at 0 0 RCIM_Cubic>,<at 1 0>]}}}, CurveIndex=0)
    Module.Param0WriteEnabled=true
    Module.Param1WriteEnabled=true
    Module.Param2WriteEnabled=true
    Module.Param3WriteEnabled=true
   )
   &CameraOffset(Camera Offset Amount=-100)
   &ScaleMeshSize(Scale Factor=&VectorFromFloat(Value=$User.IsSpawnActiveState))
  ]
  rendererClasses [&NiagaraMeshRendererProperties]
  renderers [{rendererClass=&NiagaraMeshRendererProperties, rendererIndex=0}]
 }
 Embers {
  simTarget=GPUComputeSim
  emitterUpdateScript [
   &EmitterState(Life Cycle Mode=::NewEnumerator1, Loop Duration=8, UseLoopDelay=true)
   &SpawnRate(
    SpawnRate=&Multiply_Float(
     A=22
     B=&FloatFromCurve(
      FloatCurve=DataInterface() {propertyValues {Curve {keys [
          <at 0 1 RCIM_Cubic>
          {arriveTangent=-0.2543, interpMode=RCIM_Cubic, leaveTangent=-0.2543, time=1.2828, value=1.6362}
          {arriveTangent=-0.3703, interpMode=RCIM_Cubic, leaveTangent=-0.3703, time=2.4472, value=0.3776}
          {arriveTangent=-0.11, interpMode=RCIM_Cubic, leaveTangent=-0.11, time=3.6302, value=0.7671}
          {arriveTangent=0.2907, interpMode=RCIM_Cubic, leaveTangent=0.2907, time=5.034, value=0.0931}
          {arriveTangent=0.3058, interpMode=RCIM_Cubic, leaveTangent=0.3058, time=6.8459, value=1.702}
          <at 8 1>
         ]}}}
      CurveIndex=$Emitter.NormalizedLoopAge
     )
    )
    Use Spawn Probability=false
   ) {module=SpawnRate001}
  ]
  particleSpawnScript [
   /SetVariables(Particles.SpriteRotationRate=&RandomRangeFloat(Minimum=-1, Maximum=1, Evaluation Type=::NewEnumerator1)) {isSetParameters=true, module=SetVariables_#9}
   /Niagara/Modules/Spawn/Initialization/V2/InitializeParticle(
    Lifetime Max=3
    Lifetime Min=2
    Lifetime Mode=::NewEnumerator1
    Sprite Size Mode=::NewEnumerator1
    Uniform Sprite Size Max=11
    UsePositionOffset=false
    Write Material Random=true
   ) {module=InitializeParticle001}
   &Color(Color=&MakeLinearColorFromVectorAndFloat("Vector (RGB)"=<2.6656 15.8408 16>, "Float (Alpha)"=&ClampFloat(Float=1)))
   /Niagara/Modules/Spawn/Location/SphereLocation(
    MaskSpawnGroup=false
    Module.Override Randomness=false
    Sphere Radius=&RandomRangeFloat(Minimum=50, Maximum=170, Evaluation Type=::NewEnumerator1)
    Surface Only=true
   )
   &AddVelocity()
  ]
  particleUpdateScript [
   &ParticleState()
   &Drag(Use Linear Drag=true, Use Rotational Drag=false)
   &CurlNoiseForce(Noise Strength=150, Pan Noise=false)
   &AccelerationForce(Acceleration=&MakeVector(X=0, Y=0, Z=&RandomRangeFloat(Minimum=0, Maximum=150)))
   &SolveForcesAndVelocity(Clamp Velocity=false, Limit Acceleration=false)
   &ScaleSpriteSize(Uniform Curve Sprite Scale=DataInterface() {propertyValues {Curve {keys [
        {interpMode=RCIM_Cubic, tangentMode=RCTM_User, time=0, value=0}
        {interpMode=RCIM_Cubic, tangentMode=RCTM_User, time=0.5, value=1}
        {tangentMode=RCTM_User, time=1, value=0}
       ]}}})
   &ScaleColor(Scale Alpha=&Lerp_Float(A=0.5, B=2, Alpha=$System.DistanceLerp), ScaleA=true, ScaleRGB=true)
  ]
  rendererClasses [&NiagaraSpriteRendererProperties]
  renderers [{rendererClass=&NiagaraSpriteRendererProperties}]
 }
 Embers004 {
  simTarget=GPUComputeSim
  emitterUpdateScript [
   &EmitterState(Life Cycle Mode=::NewEnumerator1, Loop Behavior=::NewEnumerator1, Loop Duration=4, UseLoopDelay=true)
   &SpawnRate(
    Spawn Probability=&ClampFloat(Float=&Lerp_Float(A=-2, B=1, Alpha=$System.DistanceLerp))
    SpawnRate=&FloatFromCurve(
     FloatCurve=DataInterface() {propertyValues {Curve {keys [<at 0 0>,<at 0.3 1>,<at 1 0>]}}}
     CurveIndex=$Emitter.NormalizedLoopAge
     Scale Curve=650
    )
    Use Spawn Probability=true
   )
  ]
  particleSpawnScript [
   /SetVariables(Particles.SpriteRotationRate=&RandomRangeFloat(Minimum=-1, Maximum=1, Evaluation Type=::NewEnumerator1)) {isSetParameters=true, module=SetVariables_#9}
   &InitializeParticle(
    Lifetime Max=1
    Lifetime Min=0.7
    Lifetime Mode=::NewEnumerator1
    Sprite Size Max=<15 15>
    Sprite Size Min=<10 10>
    Sprite Size Mode=::NewEnumerator2
    UsePositionOffset=false
    Write Material Random=true
   ) {module=InitializeParticle001}
   &Color(Color=&MakeLinearColorFromVectorAndFloat("Vector (RGB)"=<2.6656 15.8408 16>, "Float (Alpha)"=&ClampFloat(Float=1)))
   &SphereLocation(
    MaskSpawnGroup=false
    Module.Override Randomness=false
    Offset=<0 0 250>
    Sphere Radius=&RandomRangeFloat(Minimum=50, Maximum=800, Evaluation Type=::NewEnumerator1)
    Surface Only=true
   )
   &AddVelocity()
  ]
  particleUpdateScript [
   &ParticleState()
   &Drag(Drag=-10, Use Linear Drag=true, Use Rotational Drag=false)
   &CurlNoiseForce(Noise Strength=500, Pan Noise=false)
   &PointAttractionForce(
    Attraction Radius=2000
    Attraction Strength=&FloatFromCurve(
     FloatCurve=DataInterface() {propertyValues {Curve {keys [{tangentMode=RCTM_User, time=0, value=0}, {tangentMode=RCTM_User, time=3, value=1}]}}}
     CurveIndex=$Emitter.Age
     Scale Curve=&RandomRangeFloat(Minimum=60, Maximum=100)
    )
    Attractor Position Offset:Vector3f=<0 0 220>
    Kill Within Radius=false
    Module.Use Falloff=true
    UseAttractorOffset=true
   )
   &AccelerationForce(Acceleration=&MakeVector(X=0, Y=0, Z=&RandomRangeFloat(Minimum=0, Maximum=150)))
   &SolveForcesAndVelocity(Clamp Velocity=false, Limit Acceleration=false)
   &ScaleSpriteSize(
    Non-Uniform Curve Scale=<1 1.5>
    Non-Uniform Curve Sprite Scale=DataInterface() {propertyValues {
      XCurve {keys [{tangentMode=RCTM_User, time=0, value=0}, <at 0.3 0.9>, {tangentMode=RCTM_User, time=1, value=0}]}
      YCurve {keys [{tangentMode=RCTM_User, time=0, value=0}, <at 0.5 1.05>, {tangentMode=RCTM_User, time=1, value=0.5}]}
     }}
    Scale Sprite Size Mode=::NewEnumerator3
   )
   &ScaleColor(
    Scale Alpha=&Lerp_Float(A=0.5, B=2, Alpha=$System.DistanceLerp)
    Scale RGB=&VectorFromFloat(Value=&FloatFromCurve(FloatCurve=DataInterface() {propertyValues {Curve {keys [<at 0 1>,<at 1 0>]}}}))
    ScaleA=true
    ScaleRGB=true
   )
  ]
  rendererClasses [&NiagaraSpriteRendererProperties]
  renderers [{rendererClass=&NiagaraSpriteRendererProperties}]
 }
 EmbersUpper {
  simTarget=GPUComputeSim
  emitterUpdateScript [
   &EmitterState(Life Cycle Mode=::NewEnumerator1, Loop Duration=8, UseLoopDelay=true)
   &SpawnRate(
    Spawn Probability=$User.IsSpawnActiveState
    SpawnRate=&Multiply_Float(
     A=40
     B=&FloatFromCurve(
      FloatCurve=DataInterface() {propertyValues {Curve {keys [
          <at 0 1 RCIM_Cubic>
          {arriveTangent=-0.2543, interpMode=RCIM_Cubic, leaveTangent=-0.2543, time=1.2828, value=1.6362}
          {arriveTangent=-0.3703, interpMode=RCIM_Cubic, leaveTangent=-0.3703, time=2.4472, value=0.3776}
          {arriveTangent=-0.11, interpMode=RCIM_Cubic, leaveTangent=-0.11, time=3.6302, value=0.7671}
          {arriveTangent=0.2907, interpMode=RCIM_Cubic, leaveTangent=0.2907, time=5.034, value=0.0931}
          {arriveTangent=0.3058, interpMode=RCIM_Cubic, leaveTangent=0.3058, time=6.8459, value=1.702}
          <at 8 1>
         ]}}}
      CurveIndex=$Emitter.NormalizedLoopAge
     )
    )
    Use Spawn Probability=true
   ) {module=SpawnRate001}
  ]
  particleSpawnScript [
   /SetVariables(Particles.SpriteRotationRate=&RandomRangeFloat(Minimum=-1, Maximum=1, Evaluation Type=::NewEnumerator1)) {isSetParameters=true, module=SetVariables_#9}
   &InitializeParticle(
    Lifetime Max=3
    Lifetime Min=2
    Lifetime Mode=::NewEnumerator1
    Sprite Size Mode=::NewEnumerator1
    Uniform Sprite Size Max=11
    UsePositionOffset=false
    Write Material Random=true
   ) {module=InitializeParticle001}
   &Color(Color=&MakeLinearColorFromVectorAndFloat("Vector (RGB)"=<2.6656 15.8408 16>, "Float (Alpha)"=&ClampFloat(Float=1)))
   &SphereLocation(
    MaskSpawnGroup=false
    Module.Override Randomness=false
    Offset=&MakeVector(X=0, Y=0, Z=$System.FlareHeight)
    Sphere Radius=&RandomRangeFloat(Minimum=50, Maximum=400, Evaluation Type=::NewEnumerator1)
    Surface Only=true
   )
   &AddVelocity()
  ]
  particleUpdateScript [
   &ParticleState()
   &Drag(Use Linear Drag=true, Use Rotational Drag=false)
   &CurlNoiseForce(Noise Strength=150, Pan Noise=false)
   &PointAttractionForce(
    Attraction Radius=1000
    Attraction Strength=10
    Attractor Position Offset:Vector3f=<0 0 250>
    Kill Within Radius=false
    Module.Use Falloff=true
    UseAttractorOffset=true
   )
   &AccelerationForce(Acceleration=&MakeVector(X=0, Y=0, Z=&RandomRangeFloat(Minimum=0, Maximum=150)))
   &SolveForcesAndVelocity(Clamp Velocity=false, Limit Acceleration=false)
   &ScaleSpriteSize(Uniform Curve Sprite Scale=DataInterface() {propertyValues {Curve {keys [
        {interpMode=RCIM_Cubic, tangentMode=RCTM_User, time=0, value=0}
        {interpMode=RCIM_Cubic, tangentMode=RCTM_User, time=0.5, value=1}
        {tangentMode=RCTM_User, time=1, value=0}
       ]}}})
   &ScaleColor(Scale Alpha=&Lerp_Float(A=0.5, B=2, Alpha=$System.DistanceLerp), ScaleA=true, ScaleRGB=true)
  ]
  rendererClasses [&NiagaraSpriteRendererProperties]
  renderers [{rendererClass=&NiagaraSpriteRendererProperties}]
 }
 EnergyMesh {
  emitterSpawnScript [/SetVariables(Emitter.Refraction=1.1) {isSetParameters=true, module=SetVariables_#10}]
  emitterUpdateScript [
   &EmitterState(
    Inactive Response=::NewEnumerator1
    Life Cycle Mode=::NewEnumerator1
    Loop Behavior=::NewEnumerator1
    Loop Duration Mode=::NewEnumerator1
    UseLoopDelay=false
   )
   &SpawnBurst_Instantaneous(
    Spawn Time=&Multiply_Float(A=$System.ActiveSpawnTime, B=$System.SpawnTimeMul)
    Use Spawn Probability=false
    UseLoopCountLimit=false
   )
  ]
  particleSpawnScript [
   &InitialMeshOrientation(Mesh Axis To Orient=<0 0 1>, Mesh Orientation Mode=::NewEnumerator2, Use Rotation Vector=false)
   &InitializeParticle(Color=<rgba 0.3695 0.9202 1 0.6>, Mesh Scale Mode=::NewEnumerator3, UsePositionOffset=false, Write Material Random=true)
   &SystemLocation(MaskSpawnGroup=false, Offset=<0 0 -150>)
  ]
  particleUpdateScript [
   &ParticleState(Kill Particles When Lifetime Has Elapsed=false, Let Infinitely Lived Particles Die When Emitter Deactivates=true)
   &ScaleMeshSize(Scale Factor=<3 1 1.8>)
   &DynamicMaterialParameters(
    Index 0 Param 1=1.8
    Index 0 Param 2=0.2
    Index 0 Param 3=1
    Index 0 Param 4=1
    Module.Param0WriteEnabled=true
    Module.Param1WriteEnabled=true
    Module.Param2WriteEnabled=true
    Module.Param3WriteEnabled=true
   )
   &ScaleColor(Scale Alpha=&Multiply_Float(A=$System.DistanceLerp, B=$User.OverallVisibility), ScaleA=true, ScaleRGB=true)
  ]
  rendererClasses [&NiagaraMeshRendererProperties]
  renderers [{rendererClass=&NiagaraMeshRendererProperties, rendererIndex=0}]
 }
 EneryMeshCam {
  emitterSpawnScript [/SetVariables(Emitter.Refraction=1.1) {isSetParameters=true, module=SetVariables_#10}]
  emitterUpdateScript [
   &EmitterState(
    Inactive Response=::NewEnumerator1
    Life Cycle Mode=::NewEnumerator1
    Loop Behavior=::NewEnumerator1
    Loop Duration Mode=::NewEnumerator1
    UseLoopDelay=false
   )
   &SpawnBurst_Instantaneous(
    Spawn Time=&Multiply_Float(A=$System.ActiveSpawnTime, B=$System.SpawnTimeMul)
    Use Spawn Probability=false
    UseLoopCountLimit=false
   )
  ]
  particleSpawnScript [
   &InitializeParticle(
    Color=<rgba 0.3695 0.9202 1 1>
    Mesh Scale Mode=::NewEnumerator3
    Mesh Uniform Scale=1.6
    UsePositionOffset=false
    Write Material Random=true
   )
   &SystemLocation(MaskSpawnGroup=false, Offset=&MakeVector(X=0, Y=0, Z=$System.FlareHeight))
  ]
  particleUpdateScript [
   &ParticleState(Kill Particles When Lifetime Has Elapsed=false, Let Infinitely Lived Particles Die When Emitter Deactivates=true)
   &ScaleMeshSize(Scale Factor=&Multiply_VectorByFloat(Vector=<-2 1.8 1>, Float=1))
   &DynamicMaterialParameters(
    Index 0 Param 1=0.8
    Index 0 Param 2=0.2
    Module.Param0WriteEnabled=true
    Module.Param1WriteEnabled=true
    Module.Param2WriteEnabled=true
    Module.Param3WriteEnabled=true
   )
   &ScaleColor(
    Scale Alpha=&Multiply_Float(A=&Lerp_Float(A=0.5, B=$User.OverallVisibility, Alpha=$System.DistanceLerp), B=$User.IsSpawnActiveState)
    ScaleA=true
    ScaleRGB=true
   )
  ]
  rendererClasses [&NiagaraMeshRendererProperties]
  renderers [{rendererClass=&NiagaraMeshRendererProperties, rendererIndex=0}]
 }
 Glass {
  simTarget=GPUComputeSim
  emitterUpdateScript [
   &EmitterState(Life Cycle Mode=::NewEnumerator1, Loop Behavior=::NewEnumerator1, Loop Duration=8, UseLoopDelay=false)
   &SpawnBurst_Instantaneous(
    Spawn Count=&RandomRangeInt(Minimum=15, Maximum=50, Evaluation Type=::NewEnumerator1, Random Seed=$Emitter.LoopCount, Fixed Random Seed=true)
    Spawn Probability=&ClampFloat(Float=&Lerp_Float(A=-2, B=1, Alpha=$System.DistanceLerp))
    Spawn Time=&Multiply_Float(A=$System.ActiveSpawnTime, B=$System.SpawnTimeMul)
    Use Spawn Probability=true
   )
   &SpawnRate(
    Spawn Probability=&ClampFloat(Float=&Lerp_Float(A=-2, B=1, Alpha=$System.DistanceLerp))
    SpawnRate=&FloatFromCurve(
     FloatCurve=DataInterface() {propertyValues {Curve {keys [<at 0 0>,<at 0.01 1>,<at 1 0>]}}}
     CurveIndex=&Subtract_Float(A=$Emitter.Age, B=&Multiply_Float(A=$System.ActiveSpawnTime, B=$System.SpawnTimeMul))
     Scale Curve=300
    )
    Use Spawn Probability=true
   )
  ]
  particleSpawnScript [
   /SetVariables(Particles.SpawnGroup=$Emitter.LoopCount) {isSetParameters=true}
   &InitializeParticle(
    Lifetime Max=4
    Lifetime Min=2.5
    Lifetime Mode=::NewEnumerator1
    Mesh Scale Mode=::NewEnumerator1
    Mesh Uniform Scale Max=0.9
    Mesh Uniform Scale Min=0.3
    Sprite Size Mode=::NewEnumerator3
    UsePositionOffset=false
    Write Material Random=true
   )
   /SetVariables(Particles.MeshIndex=&RandomRangeInt(Minimum=0, Maximum=2)) {isSetParameters=true, module=SetVariables_#11}
   &InitialMeshOrientation(Orientation Vector=&RandomVector(Vector Scale=1), Use Orientation Vector=true, Use Rotation Vector=false)
   &SystemLocation(MaskSpawnGroup=false, Offset=&MakeVector(X=0, Y=0, Z=$System.FlareHeight))
   &AddVelocity(
    Module.Use Falloff Distance=false
    Origin Offset=&MakeVector(X=0, Y=0, Z=$System.FlareHeight)
    Override Seed=false
    Velocity Mode=::NewEnumerator1
    Velocity Speed=&Multiply_Float(
     A=&RandomRangeFloat(Minimum=250, Maximum=500)
     B=&FloatFromCurve(
      FloatCurve=DataInterface() {propertyValues {Curve {keys [<at 0 1 RCIM_Cubic>, {arriveTangent=-0.0369, interpMode=RCIM_Cubic, leaveTangent=-0.0369, tangentMode=RCTM_User, time=1.5, value=0}]}}}
      CurveIndex=&Subtract_Float(A=$Emitter.Age, B=&Multiply_Float(A=$System.ActiveSpawnTime, B=$System.SpawnTimeMul))
      Scale Curve=3
     )
    )
   )
  ]
  particleUpdateScript [
   &ParticleState()
   &CurlNoiseForce(
    Noise Frequency=20
    Noise Strength=&FloatFromCurve(
     FloatCurve=DataInterface() {propertyValues {Curve {keys [{arriveTangent=-3.2987, interpMode=RCIM_Cubic, leaveTangent=-3.2987, tangentMode=RCTM_User, time=0, value=1}, <at 1 0>]}}}
     Scale Curve=300
    )
    Pan Noise=false
   ) {module=CurlNoiseForce001}
   &GravityForce(Gravity=<0 0 -10>)
   &Drag(Drag=2, Rotational Drag=0.01, Use Linear Drag=true, Use Rotational Drag=true)
   &SolveForcesAndVelocity(Clamp Velocity=false, Limit Acceleration=false)
   &SolveRotationalForcesAndVelocity()
   &ScaleMeshSize(Scale Factor=&VectorFromFloat(Value=&FloatFromCurve(FloatCurve=DataInterface() {propertyValues {Curve {keys [<at 0 0 RCIM_Cubic>,<at 0.05 1>,<at 1 0 RCIM_Cubic>]}}})))
   &ScaleColor(
    Scale Alpha=&FloatFromCurve(FloatCurve=DataInterface() {propertyValues {Curve {keys [
         {arriveTangent=9.0725, interpMode=RCIM_Cubic, leaveTangent=9.0725, tangentMode=RCTM_User, time=0, value=0}
         {arriveTangent=-2.5249, interpMode=RCIM_Cubic, leaveTangent=-2.5249, tangentMode=RCTM_User, time=0.1408, value=0.9847}
         {arriveTangent=-0.0405, interpMode=RCIM_Cubic, leaveTangent=-0.0405, tangentMode=RCTM_User, time=1, value=0}
        ]}}})
    Scale RGB=&VectorFromFloat(Value=10)
    ScaleA=true
    ScaleRGB=true
   )
   &UpdateMeshOrientation(Rotation Rate=&RandomRangeFloat(Minimum=-1, Maximum=1), Rotation Vector=<1 1 1>)
  ]
  rendererClasses [&NiagaraMeshRendererProperties]
  renderers [{rendererClass=&NiagaraMeshRendererProperties, rendererIndex=0}]
 }
 Glass001 {
  simTarget=GPUComputeSim
  emitterUpdateScript [
   &EmitterState()
   &SpawnRate(SpawnRate=80, Use Spawn Probability=true)
  ]
  particleSpawnScript [
   &InitializeParticle(
    Lifetime Max=4
    Lifetime Min=3
    Lifetime Mode=::NewEnumerator1
    Mesh Scale Mode=::NewEnumerator1
    Mesh Uniform Scale Max=0.9
    Mesh Uniform Scale Min=0.3
    Sprite Size Mode=::NewEnumerator3
    UsePositionOffset=false
    Write Material Random=true
   )
   /SetVariables(Particles.MeshIndex=&RandomRangeInt(Minimum=0, Maximum=2)) {isSetParameters=true, module=SetVariables_#11}
   &InitialMeshOrientation(Orientation Vector=&RandomVector(Vector Scale=1), Use Orientation Vector=true, Use Rotation Vector=false)
   &SphereLocation(
    MaskSpawnGroup=false
    Module.Override Randomness=false
    Offset=&MakeVector(X=0, Y=0, Z=$System.FlareHeight)
    Sphere Radius=&RandomRangeFloat(Minimum=50, Maximum=400, Evaluation Type=::NewEnumerator1)
    Surface Only=true
   )
   &AddVelocity(
    Module.Use Falloff Distance=false
    Origin Offset=&MakeVector(X=0, Y=0, Z=$System.FlareHeight)
    Override Seed=false
    Velocity Mode=::NewEnumerator1
    Velocity Speed=&RandomRangeFloat(Minimum=2, Maximum=100)
   )
  ]
  particleUpdateScript [
   &ParticleState()
   &CurlNoiseForce(
    Noise Frequency=20
    Noise Strength=&FloatFromCurve(
     FloatCurve=DataInterface() {propertyValues {Curve {keys [{arriveTangent=-3.2987, interpMode=RCIM_Cubic, leaveTangent=-3.2987, tangentMode=RCTM_User, time=0, value=1}, <at 1 0>]}}}
     Scale Curve=20
    )
    Pan Noise=false
   ) {module=CurlNoiseForce001}
   &GravityForce(Gravity=<0 0 -10>)
   &PointAttractionForce(
    Attraction Radius=1000
    Attraction Strength=&Lerp_Float(A=5, B=0, Alpha=$User.IsSpawnActiveState)
    Kill Within Radius=false
    Module.Use Falloff=true
    UseAttractorOffset=false
   )
   &Drag(Rotational Drag=0.01, Use Linear Drag=true, Use Rotational Drag=true)
   &SolveForcesAndVelocity(Clamp Velocity=false, Limit Acceleration=false)
   &SolveRotationalForcesAndVelocity()
   &ScaleMeshSize(Scale Factor=&VectorFromFloat(Value=&FloatFromCurve(FloatCurve=DataInterface() {propertyValues {Curve {keys [<at 0 0 RCIM_Cubic>,<at 0.05 1>,<at 1 0 RCIM_Cubic>]}}})))
   &ScaleColor(
    Scale Alpha=&FloatFromCurve(FloatCurve=DataInterface() {propertyValues {Curve {keys [
         {arriveTangent=9.0725, interpMode=RCIM_Cubic, leaveTangent=9.0725, tangentMode=RCTM_User, time=0, value=0}
         {arriveTangent=-2.5249, interpMode=RCIM_Cubic, leaveTangent=-2.5249, tangentMode=RCTM_User, time=0.1408, value=0.9847}
         {arriveTangent=-0.0405, interpMode=RCIM_Cubic, leaveTangent=-0.0405, tangentMode=RCTM_User, time=1, value=0}
        ]}}})
    Scale RGB=&VectorFromFloat(Value=10)
    ScaleA=true
    ScaleRGB=true
   )
   &UpdateMeshOrientation(Rotation Rate=&RandomRangeFloat(Minimum=-0.2, Maximum=0.2), Rotation Vector=<1 1 1>)
  ]
  rendererClasses [&NiagaraMeshRendererProperties]
  renderers [{rendererClass=&NiagaraMeshRendererProperties, rendererIndex=0}]
 }
 Light001 {
  emitterSpawnScript [/SetVariables() {isSetParameters=true, module=SetVariables_#12}]
  emitterUpdateScript [
   &EmitterState(Life Cycle Mode=::NewEnumerator1, Loop Behavior=::NewEnumerator1, Loop Duration Mode=::NewEnumerator1, UseLoopDelay=false)
   &SpawnBurst_Instantaneous(Use Spawn Probability=false, UseLoopCountLimit=false)
   &SpawnBurst_Instantaneous(Spawn Group=1, Use Spawn Probability=false, UseLoopCountLimit=false) {module=SpawnBurst_Instantaneous001}
   /SetVariables(Emitter.SizeAnimCurve:NiagaraFloat=&Multiply_Float(
     A=&FloatFromCurve(
      FloatCurve=$System.IntensityAnimCurve
      CurveIndex=&Subtract_Float(A=&Divide_Float(A=$Emitter.Age, B=$System.SpawnTimeMul), B=0.5)
     )
     B=&RandomRangeFloat(
      Minimum=1
      Maximum=&FloatFromCurve(
       FloatCurve=$System.FlickerAnimCurve
       CurveIndex=&Subtract_Float(A=&Divide_Float(A=$Emitter.Age, B=$System.SpawnTimeMul), B=0.5)
      )
      Evaluation Type=::NewEnumerator1
     )
    )) {isSetParameters=true, module=SetVariables_#13}
   &SpawnBurst_Instantaneous(Spawn Time=1.9937, Use Spawn Probability=false, UseLoopCountLimit=false) {module=SpawnBurst_Instantaneous002}
  ]
  particleSpawnScript [
   &InitializeParticle(
    Mesh Scale=<1 15 25>
    Mesh Scale Mode=::NewEnumerator4
    Sprite Size=<400 1000>
    Sprite Size Mode=::NewEnumerator4
    UsePositionOffset=false
    Write Material Random=true
   )
   /SetVariables(Particles.VisibilityTag=&SpawnGroup()) {isSetParameters=true}
   /SetVariables(Particles.GlowSpriteAlignment:Vector3f=<0 0 1>) {isSetParameters=true}
   &Color(Color=&MakeLinearColorFromVectorAndFloat("Vector (RGB)"=<0.2027 0.7676 1>, "Float (Alpha)"=&ClampFloat(Float=1)))
   &SystemLocation(
    MaskSpawnGroup=false
    Offset=&MakeVector(
     X=0
     Y=0
     Z=&SelectFloatFromArray(Direct Array Index=&SpawnGroup(), Float Selection Array=DataInterface() {propertyValues {FloatData [70,150]}})
    )
   )
  ]
  particleUpdateScript [
   /SetVariables(Particles.LightExponent=3) {isSetParameters=true}
   /SetVariables(Particles.LightRadius=&Lerp_Float(A=1, B=2, Alpha=$Emitter.SizeAnimCurve)) {isSetParameters=true}
   /SetVariables(Particles.LightColorv2:LinearColor=&MultiplyLinearColor(
     Linear Color=<rgba 0.3282 6.4303 10 1>
     Scale Factor=&MakeVector4_FromFloat(Float=&Multiply_Float(A=$System.DistanceLerp, B=&Multiply_Float(A=$Emitter.SizeAnimCurve, B=$User.OverallVisibility)))
    )) {isSetParameters=true}
   &ParticleState(Kill Particles When Lifetime Has Elapsed=false, Let Infinitely Lived Particles Die When Emitter Deactivates=true)
   &ScaleColor(
    Scale Alpha=&Multiply_Float(A=$System.DistanceLerp, B=$Emitter.SizeAnimCurve)
    Scale RGB=&VectorFromFloat(Value=$User.OverallVisibility)
    ScaleA=true
    ScaleRGB=true
   )
  ]
  rendererClasses [
   &NiagaraLightRendererProperties
   &NiagaraMeshRendererProperties
   &NiagaraSpriteRendererProperties
  ]
  renderers [
   {rendererClass=&NiagaraLightRendererProperties, rendererIndex=0}
   {rendererClass=&NiagaraMeshRendererProperties, rendererIndex=1}
   {rendererClass=&NiagaraSpriteRendererProperties, rendererIndex=2}
  ]
 }
 Shimmer {
  emitterUpdateScript [
   &EmitterState(Inactive Response=::NewEnumerator1, Life Cycle Mode=::NewEnumerator1, Loop Duration=1, UseLoopDelay=false)
   &SpawnRate(SpawnRate=&Multiply_Float(A=$User.IsSpawnActiveState, B=$System.BuildInTime), Use Spawn Probability=false)
  ]
  particleSpawnScript [
   &InitializeParticle(
    Color=<rgba 0.3569 0.733 1 0.7>
    Lifetime Max=3
    Lifetime Min=2
    Lifetime Mode=::NewEnumerator1
    Sprite Rotation Mode=::NewEnumerator2
    Sprite Size Mode=::NewEnumerator1
    Sprite UV Mode=::NewEnumerator3
    Uniform Sprite Size Max=250
    Uniform Sprite Size Min=220
    UsePositionOffset=false
    Write Material Random=true
   )
   &SystemLocation(MaskSpawnGroup=false, Offset=&MakeVector(X=0, Y=0, Z=$System.FlareHeight))
   &ShapeLocation(
    MaskSpawnGroup=false
    Override Seed=false
    Sphere Radius=&RandomRangeFloat(Minimum=200, Maximum=250)
    UseOwnerScale=true
   )
   &AddVelocity(
    Module.Use Falloff Distance=false
    Origin Offset=<0 0 250>
    Override Seed=false
    Velocity Mode=::NewEnumerator1
    Velocity Speed=&RandomRangeFloat(Minimum=25, Maximum=50)
   )
  ]
  particleUpdateScript [
   &ParticleState()
   &SolveForcesAndVelocity(Clamp Velocity=false, Limit Acceleration=false)
   &ScaleColor(
    Scale Alpha=&FloatFromCurve(FloatCurve=DataInterface() {propertyValues {Curve {keys [
         {interpMode=RCIM_Cubic, tangentMode=RCTM_User, time=0, value=0}
         {interpMode=RCIM_Cubic, tangentMode=RCTM_User, time=0.5, value=1}
         {tangentMode=RCTM_User, time=1, value=0}
        ]}}})
    ScaleA=true
    ScaleRGB=true
   )
   &DynamicMaterialParameters(
    Index 0 Param 1=1
    Index 0 Param 2=1
    Index 0 Param 3=0.5
    Index 0 Param 4=1
    Module.Param0WriteEnabled=true
    Module.Param1WriteEnabled=true
    Module.Param2WriteEnabled=true
    Module.Param3WriteEnabled=true
   )
   &ScaleSpriteSize(
    Uniform Curve Scale=2
    Uniform Curve Sprite Scale=DataInterface() {propertyValues {Curve {keys [{tangentMode=RCTM_User, time=0, value=0.5}, {tangentMode=RCTM_User, time=1, value=1}]}}}
   )
  ]
  rendererClasses [&NiagaraSpriteRendererProperties]
  renderers [{rendererClass=&NiagaraSpriteRendererProperties}]
 }
 Shimmer001 {
  emitterUpdateScript [
   &EmitterState(Life Cycle Mode=::NewEnumerator1, Loop Behavior=::NewEnumerator1, Loop Duration=4, UseLoopDelay=false)
   &SpawnRate(
    Spawn Probability=&ClampFloat(Float=&Lerp_Float(A=-2, B=1, Alpha=$System.DistanceLerp))
    SpawnRate=&FloatFromCurve(
     FloatCurve=DataInterface() {propertyValues {Curve {keys [<at 0 0>,<at 0.1 1>,<at 1 0>]}}}
     CurveIndex=$Emitter.NormalizedLoopAge
     Scale Curve=10
    )
    Use Spawn Probability=true
   )
  ]
  particleSpawnScript [
   &InitializeParticle(
    Color=<rgba 0.3569 0.733 1 0.3>
    Lifetime Max=1.5
    Lifetime Min=1.5
    Lifetime Mode=::NewEnumerator1
    Sprite Size Max=<100 500>
    Sprite Size Min=<50 400>
    Sprite Size Mode=::NewEnumerator2
    UsePositionOffset=false
    Write Material Random=true
   )
   &SystemLocation(MaskSpawnGroup=false, Offset=<0 0 250>)
   &SphereLocation(
    MaskSpawnGroup=false
    Module.Override Randomness=false
    Offset=<0 0 250>
    Sphere Radius=&RandomRangeFloat(Minimum=50, Maximum=800, Evaluation Type=::NewEnumerator1)
    Surface Only=true
   )
   &AddVelocity(
    Module.Use Falloff Distance=false
    Origin Offset=<0 0 250>
    Override Seed=false
    Velocity Mode=::NewEnumerator1
    Velocity Speed=&RandomRangeFloat(Minimum=25, Maximum=50)
   )
  ]
  particleUpdateScript [
   &ParticleState()
   &ScaleColor(
    Scale Alpha=&FloatFromCurve(FloatCurve=DataInterface() {propertyValues {Curve {keys [
         {interpMode=RCIM_Cubic, tangentMode=RCTM_User, time=0.1, value=0}
         {interpMode=RCIM_Cubic, tangentMode=RCTM_User, time=0.5, value=1}
         {tangentMode=RCTM_User, time=1, value=0}
        ]}}})
    ScaleA=true
    ScaleRGB=true
   )
   &PointAttractionForce(
    Attraction Radius=2000
    Attraction Strength=&FloatFromCurve(
     FloatCurve=DataInterface() {propertyValues {Curve {keys [{tangentMode=RCTM_User, time=0, value=0.5}, {tangentMode=RCTM_User, time=3, value=1}]}}}
     CurveIndex=$Emitter.Age
     Scale Curve=&RandomRangeFloat(Minimum=30, Maximum=40)
    )
    Attractor Position Offset:Vector3f=<0 0 220>
    Kill Within Radius=false
    Module.Use Falloff=true
    UseAttractorOffset=true
   )
   &SolveForcesAndVelocity(Clamp Velocity=false, Limit Acceleration=false) {module=SolveForcesAndVelocity001}
   &DynamicMaterialParameters(
    Index 0 Param 1=1
    Index 0 Param 2=1
    Index 0 Param 3=0.5
    Index 0 Param 4=1
    Module.Param0WriteEnabled=true
    Module.Param1WriteEnabled=true
    Module.Param2WriteEnabled=true
    Module.Param3WriteEnabled=true
   )
   &ScaleSpriteSize(
    Non-Uniform Curve Scale=<2 2>
    Non-Uniform Curve Sprite Scale=DataInterface() {propertyValues {
      XCurve {keys [
        {arriveTangent=3.497, interpMode=RCIM_Cubic, leaveTangent=3.497, tangentMode=RCTM_User, time=0, value=0}
        {tangentMode=RCTM_User, time=1, value=1}
       ]}
      YCurve {keys [{tangentMode=RCTM_User, time=0, value=0}, {tangentMode=RCTM_User, time=1, value=1}]}
     }}
    Scale Sprite Size Mode=::NewEnumerator3
   )
  ]
  rendererClasses [&NiagaraSpriteRendererProperties]
  renderers [{rendererClass=&NiagaraSpriteRendererProperties}]
 }
 ZeroPointFlare {
  emitterSpawnScript [
   /SetVariables(Emitter.DepthFade=200) {isSetParameters=true}
   /SetVariables() {isSetParameters=true, module=SetVariables_#12}
  ]
  emitterUpdateScript [
   &EmitterState(Life Cycle Mode=::NewEnumerator1, Loop Behavior=::NewEnumerator1, Loop Duration Mode=::NewEnumerator1, UseLoopDelay=false)
   &SpawnBurst_Instantaneous(Spawn Time=1, Use Spawn Probability=false, UseLoopCountLimit=false)
   /SetVariables(Emitter.SizeAnimCurve:NiagaraFloat=&Multiply_Float(
     A=&FloatFromCurve(
      FloatCurve=$System.IntensityAnimCurve
      CurveIndex=&Subtract_Float(A=&Divide_Float(A=$Emitter.Age, B=$System.SpawnTimeMul), B=0.5)
     )
     B=&RandomRangeFloat(
      Minimum=1
      Maximum=&FloatFromCurve(
       FloatCurve=$System.FlickerAnimCurve
       CurveIndex=&Subtract_Float(A=&Divide_Float(A=$Emitter.Age, B=$System.SpawnTimeMul), B=0.5)
      )
      Evaluation Type=::NewEnumerator1
     )
    )) {isSetParameters=true, module=SetVariables_#13}
  ]
  particleSpawnScript [
   /Niagara/Modules/Spawn/Initialization/InitializeParticle(
    Lifetime=&Multiply_Float(
     A=&RandomRangeFloat(Minimum=0.5, Maximum=0.8)
     B=&FloatFromCurve(
      FloatCurve=DataInterface() {propertyValues {Curve {keys [{arriveTangent=-1.5125, interpMode=RCIM_Cubic, leaveTangent=-1.5125, tangentMode=RCTM_User, time=0, value=1}, <at 0.9949 0.2248>]}}}
      CurveIndex=$Emitter.NormalizedLoopAge
     )
    )
    Module.Write Color=true
    Module.Write Lifetime=true
    Module.Write Mass=true
    Module.Write Material Random=true
    Module.Write Position=false
    Module.Write Scale=false
    Module.Write SpriteRotation=true
    Module.Write SpriteSize=true
    Sprite Rotation=&RandomRangeFloat(Minimum=0, Maximum=360)
    Sprite Size=<3000 3000>
   )
   &SystemLocation(MaskSpawnGroup=false, Offset=&MakeVector(X=0, Y=0, Z=$System.FlareHeight))
  ]
  particleUpdateScript [
   &ParticleState(Kill Particles When Lifetime Has Elapsed=false, Let Infinitely Lived Particles Die When Emitter Deactivates=true)
   /SetVariables(Particles.Size:NiagaraFloat=&Multiply_Float(
     A=&Multiply_Float(
      A=&Multiply_Float(A=&Lerp_Float(A=-1, B=3, Alpha=$System.DistanceLerp), B=$User.OverallVisibility)
      B=$User.IsSpawnActiveState
     )
     B=$Emitter.SizeAnimCurve
    )) {isSetParameters=true}
   &FN_Sprite_Attributes(
    Camera Offset=-50
    Color=&MultiplyLinearColor(
     Linear Color=<rgba 0.0279 0.0521 0.1094 0.8>
     Scale Factor=&MakeVector4_FromFloat(Float=&FloatFromCurve(
       FloatCurve=DataInterface() {propertyValues {Curve {keys [{tangentMode=RCTM_User, time=0, value=0}, {tangentMode=RCTM_User, time=1, value=1}]}}}
       CurveIndex=&Subtract_Float(A=&Divide_Float(A=$System.Age, B=2), B=1)
       Scale Curve=$System.DistanceLerp
      ))
    )
    Create Camera Offset Attribute=true
    Create Color Attribute=true
    Create SpriteSize Attribute=true
    SpriteSize=&Vector2DFromFloat(Value=&Multiply_Float(A=500, B=$Particles.Size))
    Write to Module Namespace=true
   )
   &FN_Sprite_Attributes(
    Camera Offset=75
    Color=&MultiplyLinearColor(
     Linear Color=<rgba 0.56 1.89 3 1>
     Scale Factor=&MakeVector4_FromFloat(Float=&Multiply_Float(A=$System.BuildInTime, B=&Multiply_Float(A=$System.DistanceLerp, B=0.1)))
    )
    Create Camera Offset Attribute=true
    Create Color Attribute=true
    Create SpriteSize Attribute=true
    SpriteSize=&Vector2DFromFloat(Value=&Multiply_Float(A=&Multiply_Float(A=200, B=1), B=$Particles.Size))
    Write to Module Namespace=true
   ) {module=FN_Sprite_Attributes001}
   &FN_Sprite_Attributes(
    Camera Offset=50
    Color=&MultiplyLinearColor(
     Linear Color=<rgba 0.1248 0.34 0.4375 0.5>
     Scale Factor=&MakeVector4_FromFloat(Float=&Multiply_Float(A=$System.BuildInTime, B=$System.DistanceLerp))
    )
    Create Camera Offset Attribute=true
    Create Color Attribute=true
    Create SpriteSize Attribute=true
    SpriteSize=&Vector2DFromFloat(Value=&Multiply_Float(
      A=&Multiply_Float(A=&Lerp_Float(A=-1, B=500, Alpha=$System.DistanceLerp), B=$User.OverallVisibility)
      B=$User.IsSpawnActiveState
     ))
    Write to Module Namespace=true
   ) {module=FN_Sprite_Attributes002}
   &FN_Sprite_Attributes(
    Alignment:Vector3f=<0 0 1>
    Camera Offset=45
    Color=&MultiplyLinearColor(
     Linear Color=<rgba 0.23 1 1.5 1>
     Scale Factor=&MakeVector4_FromFloat(Float=&Multiply_Float(A=$System.BuildInTime, B=$System.DistanceLerp))
    )
    Create Alignment Attribute=true
    Create Camera Offset Attribute=true
    Create Color Attribute=true
    Create Sprite Rotation Attribute=true
    Create SpriteSize Attribute=true
    SpriteSize=&Multiply_Vector2(
     A=&RandomRangeVector2D(Minimum=<180 300>, Maximum=<200 500>, Evaluation Type=::NewEnumerator1)
     B=&Vector2DFromFloat(Value=&Multiply_Float(
       A=$Emitter.SizeAnimCurve
       B=&Multiply_Float(
        A=&Multiply_Float(A=&Lerp_Float(A=-2, B=6, Alpha=$System.DistanceLerp), B=$User.OverallVisibility)
        B=$User.IsSpawnActiveState
       )
      ))
    )
    Write to Module Namespace=true
   ) {module=FN_Sprite_Attributes004}
  ]
  rendererClasses [&NiagaraSpriteRendererProperties]
  renderers [
   {rendererClass=&NiagaraSpriteRendererProperties}
   {rendererClass=&NiagaraSpriteRendererProperties, rendererIndex=1}
   {rendererClass=&NiagaraSpriteRendererProperties, rendererIndex=2}
   {rendererClass=&NiagaraSpriteRendererProperties, rendererIndex=3}
  ]
 }
}
summary {
 systemName=NS_Encounter_SpawnIdleLoop
 emitters [
  {emitterName=Embers, simTarget=GPUComputeSim, rendererClasses [&NiagaraSpriteRendererProperties]}
  {emitterName=EnergyMesh, rendererClasses [&NiagaraMeshRendererProperties]}
  {
   emitterName=Light001
   rendererClasses [
    &NiagaraLightRendererProperties
    &NiagaraMeshRendererProperties
    &NiagaraSpriteRendererProperties
   ]
  }
  {emitterName=EneryMeshCam, rendererClasses [&NiagaraMeshRendererProperties]}
  {emitterName=ZeroPointFlare, rendererClasses [&NiagaraSpriteRendererProperties]}
  {emitterName=EmbersUpper, simTarget=GPUComputeSim, rendererClasses [&NiagaraSpriteRendererProperties]}
  {emitterName=Shimmer, rendererClasses [&NiagaraSpriteRendererProperties]}
  {emitterName=Embers004, simTarget=GPUComputeSim, rendererClasses [&NiagaraSpriteRendererProperties]}
  {emitterName=Shimmer001, rendererClasses [&NiagaraSpriteRendererProperties]}
  {emitterName=CrackSpline, rendererClasses [&NiagaraMeshRendererProperties]}
  {emitterName=Glass, simTarget=GPUComputeSim, rendererClasses [&NiagaraMeshRendererProperties]}
  {emitterName=Glass001, simTarget=GPUComputeSim, rendererClasses [&NiagaraMeshRendererProperties]}
  {emitterName=CrackSpline001, rendererClasses [&NiagaraMeshRendererProperties]}
 ]
 userVariables [
  User.IsSpawnActiveState {type=NiagaraFloat, defaultValue=1, description=""}
  User.OverallVisibility {type=NiagaraFloat, defaultValue=1, description=""}
 ]
}
